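# WaterAdvanced material function that samples an input texture with smoothed, higher-order interpolation at given UVs.
# By default, custom HLSL does bicubic Lagrange filtering over 4x4 texel loads, clamped to their min/max to prevent overshoot, with bilinear fallback at border cells.
# The UseSimple static switch instead averages four mip-level samples at symmetric UV offsets derived from the fractional texel position, a cheaper smoothing approximation.
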MaterialFunction /WaterAdvanced/Materials/CubicInterpolateTexture
graph {expressions {
  MaterialExpressionAdd_2 {
   class=MaterialExpressionAdd
   inputs [
    A {from=MaterialExpressionMultiply_5}
    B {from=""}
   ]
   properties {constB=0.5}
  }
  MaterialExpressionAdd_3 {
   class=MaterialExpressionAdd
   inputs [
    A {from=MaterialExpressionReroute_1}
    B {from=MaterialExpressionDivide_1}
   ]
  }
  MaterialExpressionAdd_4 {
   class=MaterialExpressionAdd
   inputs [
    A {from=MaterialExpressionTextureSample_1, output=RGBA}
    B {from=MaterialExpressionTextureSample_2, output=RGBA}
   ]
  }
  MaterialExpressionAdd_5 {
   class=MaterialExpressionAdd
   inputs [
    A {from=MaterialExpressionAdd_4}
    B {from=MaterialExpressionTextureSample_3, output=RGBA}
   ]
  }
  MaterialExpressionAdd_6 {
   class=MaterialExpressionAdd
   inputs [
    A {from=MaterialExpressionAdd_5}
    B {from=MaterialExpressionTextureSample_4, output=RGBA}
   ]
  }
  MaterialExpressionConvert_0 {
   class=MaterialExpressionConvert
   inputs [
    None {from=MaterialExpressionMaterialFunctionCall_0, output=R}
    None {from=MaterialExpressionMaterialFunctionCall_0, output=R}
   ]
   properties {
    nodeName=Make Float2
    convertInputs [
     {
      type=Scalar
      expressionInput {
       expression=&self:MaterialExpressionMaterialFunctionCall_0
       inputName=None
       mask=0
       maskA=0
       maskB=0
       maskG=0
       maskR=0
       outputIndex=0
      }
     }
     {
      type=Scalar
      expressionInput {
       expression=&self:MaterialExpressionMaterialFunctionCall_0
       inputName=None
       mask=0
       maskA=0
       maskB=0
       maskG=0
       maskR=0
       outputIndex=1
      }
     }
    ]
    convertOutputs [{type=Vector2}]
   }
  }
  MaterialExpressionConvert_1 {
   class=MaterialExpressionConvert
   inputs [
    None {from=MaterialExpressionMaterialFunctionCall_0, output=R}
    None {from=MaterialExpressionMaterialFunctionCall_1, output=G}
   ]
   properties {
    nodeName=Make Float2
    convertInputs [
     {
      type=Scalar
      expressionInput {
       expression=&self:MaterialExpressionMaterialFunctionCall_0
       inputName=None
       mask=0
       maskA=0
       maskB=0
       maskG=0
       maskR=0
       outputIndex=0
      }
     }
     {
      type=Scalar
      expressionInput {
       expression=&self:MaterialExpressionMaterialFunctionCall_1
       inputName=None
       mask=0
       maskA=0
       maskB=0
       maskG=0
       maskR=0
       outputIndex=1
      }
     }
    ]
    convertOutputs [{type=Vector2}]
   }
  }
  MaterialExpressionConvert_2 {
   class=MaterialExpressionConvert
   inputs [
    None {from=MaterialExpressionMaterialFunctionCall_1, output=R}
    None {from=MaterialExpressionMaterialFunctionCall_1, output=R}
   ]
   properties {
    nodeName=Make Float2
    convertInputs [
     {
      type=Scalar
      expressionInput {
       expression=&self:MaterialExpressionMaterialFunctionCall_1
       inputName=None
       mask=0
       maskA=0
       maskB=0
       maskG=0
       maskR=0
       outputIndex=0
      }
     }
     {
      type=Scalar
      expressionInput {
       expression=&self:MaterialExpressionMaterialFunctionCall_1
       inputName=None
       mask=0
       maskA=0
       maskB=0
       maskG=0
       maskR=0
       outputIndex=1
      }
     }
    ]
    convertOutputs [{type=Vector2}]
   }
  }
  MaterialExpressionConvert_3 {
   class=MaterialExpressionConvert
   inputs [
    None {from=MaterialExpressionMaterialFunctionCall_1, output=R}
    None {from=MaterialExpressionMaterialFunctionCall_0, output=G}
   ]
   properties {
    nodeName=Make Float2
    convertInputs [
     {
      type=Scalar
      expressionInput {
       expression=&self:MaterialExpressionMaterialFunctionCall_1
       inputName=None
       mask=0
       maskA=0
       maskB=0
       maskG=0
       maskR=0
       outputIndex=0
      }
     }
     {
      type=Scalar
      expressionInput {
       expression=&self:MaterialExpressionMaterialFunctionCall_0
       inputName=None
       mask=0
       maskA=0
       maskB=0
       maskG=0
       maskR=0
       outputIndex=1
      }
     }
    ]
    convertOutputs [{type=Vector2}]
   }
  }
  MaterialExpressionCustom_0 {
   class=MaterialExpressionCustom
   inputs [
    UV {from=MaterialExpressionFunctionInput_0, output=""}
    Texture {from=MaterialExpressionFunctionInput_1, output=""}
   ]
   properties {
    code="int2 NumCells;\r\nTexture.GetDimensions(NumCells.x, NumCells.y);\r\n\r\nconst float2 GridPos = UV.xy * NumCells - .5;\r\n\t\t\t\t\t\t\t\t\r\n// identify the lower-left-hand corner of the cell\r\nconst int2 GridCell = floor(GridPos);\r\n\t\t\t\t\t\t\t\t\r\nconst int2 MaxCell = NumCells - int2(2,2);\r\n\r\nif (any(GridCell < int2(1,1)) || any(GridCell >= MaxCell))\r\n{\r\n\t// revert to bilinear hardware sampling at the boundary cells.\r\n\treturn Texture.SampleLevel(View.MaterialTextureBilinearClampedSampler, UV, 0);\r\n}\r\nelse \r\n{\r\n\t// sample point offset from lower left\r\n\tfloat2 delta = GridPos - GridCell;\r\n\t\t\r\n\tfloat4x2 Weights;\r\n\r\n\tWeights[0] = -1./3 * delta + 1./2 * delta * delta - 1./6 * delta * delta * delta;\r\n\tWeights[1] = 1. - delta * delta + 1./2 * (delta * delta * delta - delta);\r\n\tWeights[2] = delta + 1./2 * (delta * delta - delta * delta * delta);\r\n\tWeights[3] = 1./6 * (delta * delta * delta - delta);\t\t\r\n\r\n\tfloat4 minv = 3.402823466e+38;\r\n\tfloat4 maxv = -3.402823466e+38;\r\n\r\n\tfloat4 BiCubicValue = 0;\t\t\r\n\tfor (int j = 0; j < 4; ++j)\r\n\t{\r\n\t\tfloat4 qj = 0;\r\n\t\tfor (int i = 0; i < 4; ++i)\r\n\t\t{\r\n\t\t\tfloat4 Val = Texture.Load(int4(GridCell.x + i - 1, GridCell.y + j - 1, 0, 0)); \r\n\t\t\tqj += Weights[i][0] * Val;\r\n\r\n\t\t\tminv = min(Val, minv);\r\n\t\t\tmaxv = max(Val, maxv);\r\n\t\t}\r\n\r\n\t\tBiCubicValue += Weights[j][1] * qj;\r\n\t}\r\n\r\n    return min(max(BiCubicValue, minv), maxv);\r\n}\r\n\r\n"
    showCode=true
    inputs [
     {
      inputName=UV
      input {
       expression=&self:MaterialExpressionFunctionInput_0
       inputName=None
       mask=0
       maskA=0
       maskB=0
       maskG=0
       maskR=0
       outputIndex=0
      }
     }
     {
      inputName=Texture
      input {
       expression=&self:MaterialExpressionFunctionInput_1
       inputName=None
       mask=0
       maskA=0
       maskB=0
       maskG=0
       maskR=0
       outputIndex=0
      }
     }
    ]
   }
  }
  MaterialExpressionDivide_1 {
   class=MaterialExpressionDivide
   inputs [
    A {from=MaterialExpressionAdd_2}
    B {from=MaterialExpressionTextureProperty_1}
   ]
  }
  MaterialExpressionDivide_2 {
   class=MaterialExpressionDivide
   inputs [
    A {from=MaterialExpressionAdd_6}
    B {from=""}
   ]
   properties {constB=4}
  }
  MaterialExpressionFrac_1 {class=MaterialExpressionFrac, inputs [None {from=MaterialExpressionMultiply_4, output=""}], properties {}}
  MaterialExpressionFunctionInput_0 {class=MaterialExpressionFunctionInput, inputs [Preview {from="", output=""}], properties {inputName=UVs, inputType=FunctionInput_Vector2}}
  MaterialExpressionFunctionInput_1 {
   class=MaterialExpressionFunctionInput
   inputs [Preview {from="", output=""}]
   properties {inputName=Texture, inputType=FunctionInput_Texture2D}
  }
  MaterialExpressionFunctionInput_2 {
   class=MaterialExpressionFunctionInput
   inputs [Preview {from=MaterialExpressionStaticBool_1, output=""}]
   properties {inputName=UseSimple, inputType=FunctionInput_StaticBool}
  }
  MaterialExpressionFunctionOutput_0 {class=MaterialExpressionFunctionOutput, inputs [None {from=MaterialExpressionStaticSwitch_0, output=""}]}
  MaterialExpressionMaterialFunctionCall_0 {
   class=MaterialExpressionMaterialFunctionCall
   inputs [Float2 {from=MaterialExpressionSubtract_3}]
   properties {materialFunction=&BreakOutFloat2Components}
  }
  MaterialExpressionMaterialFunctionCall_1 {
   class=MaterialExpressionMaterialFunctionCall
   inputs [Float2 {from=MaterialExpressionAdd_3}]
   properties {materialFunction=&BreakOutFloat2Components}
  }
  MaterialExpressionMultiply_4 {
   class=MaterialExpressionMultiply
   inputs [
    A {from=MaterialExpressionFunctionInput_0}
    B {from=MaterialExpressionTextureProperty_1}
   ]
  }
  MaterialExpressionMultiply_5 {
   class=MaterialExpressionMultiply
   inputs [
    A {from=MaterialExpressionSubtract_2}
    B {from=MaterialExpressionFrac_1}
   ]
  }
  MaterialExpressionReroute_1 {class=MaterialExpressionReroute, inputs [None {from=MaterialExpressionFunctionInput_0}]}
  MaterialExpressionReroute_2 {class=MaterialExpressionReroute, inputs [None {from=MaterialExpressionFunctionInput_1}]}
  MaterialExpressionStaticBool_1 {class=MaterialExpressionStaticBool}
  MaterialExpressionStaticSwitch_0 {
   class=MaterialExpressionStaticSwitch
   inputs [
    True {from=MaterialExpressionDivide_2, output=""}
    False {from=MaterialExpressionCustom_0, output=""}
    Value {from=MaterialExpressionFunctionInput_2, output=""}
   ]
   properties {}
  }
  MaterialExpressionSubtract_2 {
   class=MaterialExpressionSubtract
   inputs [
    A {from=MaterialExpressionFrac_1}
    B {from=""}
   ]
  }
  MaterialExpressionSubtract_3 {
   class=MaterialExpressionSubtract
   inputs [
    A {from=MaterialExpressionReroute_1}
    B {from=MaterialExpressionDivide_1}
   ]
  }
  MaterialExpressionTextureProperty_1 {class=MaterialExpressionTextureProperty, inputs [Tex {from=MaterialExpressionFunctionInput_1, output=""}], properties {}}
  MaterialExpressionTextureSample_1 {
   class=MaterialExpressionTextureSample
   inputs [
    UVs {from=MaterialExpressionConvert_0}
    Tex {from=MaterialExpressionReroute_2}
    Level {from=""}
    Apply View MipBias {from=""}
   ]
   properties {mipValueMode=TMVM_MipLevel}
  }
  MaterialExpressionTextureSample_2 {
   class=MaterialExpressionTextureSample
   inputs [
    UVs {from=MaterialExpressionConvert_1}
    Tex {from=MaterialExpressionReroute_2}
    Level {from=""}
    Apply View MipBias {from=""}
   ]
   properties {mipValueMode=TMVM_MipLevel}
  }
  MaterialExpressionTextureSample_3 {
   class=MaterialExpressionTextureSample
   inputs [
    UVs {from=MaterialExpressionConvert_2}
    Tex {from=MaterialExpressionReroute_2}
    Level {from=""}
    Apply View MipBias {from=""}
   ]
   properties {mipValueMode=TMVM_MipLevel}
  }
  MaterialExpressionTextureSample_4 {
   class=MaterialExpressionTextureSample
   inputs [
    UVs {from=MaterialExpressionConvert_3}
    Tex {from=MaterialExpressionReroute_2}
    Level {from=""}
    Apply View MipBias {from=""}
   ]
   properties {mipValueMode=TMVM_MipLevel}
  }
 }}
