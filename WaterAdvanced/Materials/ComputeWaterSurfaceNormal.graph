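# Material function that returns a water surface normal for the fluid sim from an input texture and UVs.
# By default, the Precomputed Normals switch returns the texture's RGB directly.
# Otherwise it samples height at center and one texel along U and V, optionally using cubic interpolation.
# Custom HLSL crosses world-space offsets, spaced by MPC_Water FluidSimSize over FluidSimResolution, falling back to up for degenerate results.
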MaterialFunction /WaterAdvanced/Materials/ComputeWaterSurfaceNormal
graph {nodes [
  MaterialExpressionFunctionInput_0 = FunctionInput() {inputName=UVs, inputType=FunctionInput_Vector2}
  MaterialExpressionFunctionInput_1 = FunctionInput() {inputName=Texture, inputType=FunctionInput_Texture2D}
  MaterialExpressionWorldPosition_0 = WorldPosition()
  MaterialExpressionCollectionParameter_1 = CollectionParameter() {collection=/Water/Materials/MPC/MPC_Water, parameterName=FluidSimResolution}
  MaterialExpressionDivide_0 = Divide(
   A=ComponentMask(None=CollectionParameter() {collection=&MPC_Water, parameterName=FluidSimSize}) {r=true}
   B=@MaterialExpressionCollectionParameter_1
  )
  MaterialExpressionStaticSwitchParameter_0 = StaticSwitchParameter(
   True=ComponentMask(None=MaterialFunctionCall!(Texture=@MaterialExpressionFunctionInput_1, UVs=@MaterialExpressionFunctionInput_0) {materialFunction=&/CubicInterpolateTexture}.Result) {r=true}
   False=TextureSample(UVs=@MaterialExpressionFunctionInput_0, Tex=@MaterialExpressionFunctionInput_1).R
  ) {parameterName=Use Cubic}
  n1 = Divide(B=@MaterialExpressionCollectionParameter_1)
  n2 = Add(
   A=@MaterialExpressionFunctionInput_0
   B=MaterialFunctionCall(X=@n1, Y=Constant()) {materialFunction=&MakeFloat2}.Result
  )
  MaterialExpressionStaticSwitchParameter_2 = StaticSwitchParameter(
   True=ComponentMask(None=MaterialFunctionCall(Texture=@MaterialExpressionFunctionInput_1, UVs=@n2) {materialFunction=&CubicInterpolateTexture}.Result) {r=true}
   False=TextureSample(UVs=@n2, Tex=@MaterialExpressionFunctionInput_1).R
  ) {parameterName=Use Cubic}
  n3 = Add(
   A=@MaterialExpressionFunctionInput_0
   B=MaterialFunctionCall(X=Constant(), Y=@n1) {materialFunction=&MakeFloat2}.Result
  )
  MaterialExpressionStaticSwitchParameter_3 = StaticSwitchParameter(
   True=ComponentMask(None=MaterialFunctionCall(Texture=@MaterialExpressionFunctionInput_1, UVs=@n3) {materialFunction=&CubicInterpolateTexture}.Result) {r=true}
   False=TextureSample(UVs=@n3, Tex=@MaterialExpressionFunctionInput_1).R
  ) {parameterName=Use Cubic}
  n4 = Custom!(
   UV=@MaterialExpressionFunctionInput_0
   Texture=@MaterialExpressionFunctionInput_1
   WorldPos=@MaterialExpressionWorldPosition_0.XYZ
   dx=@MaterialExpressionDivide_0
   GridRes=@MaterialExpressionCollectionParameter_1
   h=@MaterialExpressionStaticSwitchParameter_0
   h_right=@MaterialExpressionStaticSwitchParameter_2
   h_up=@MaterialExpressionStaticSwitchParameter_3
  ) {
   code="float3 Normal = float3(0,0,1);\r\n\r\nfloat3 CenterWorldPos = WorldPos;\r\n\r\nCenterWorldPos.z += h;\r\n\r\nfloat3 WorldPosRight = WorldPos;\r\nWorldPosRight.x += dx;\r\nWorldPosRight.z += h_right;\r\n\r\nfloat3 WorldPosUp = WorldPos;\r\nWorldPosUp.y += dx;\r\nWorldPosUp.z += h_up;\r\n\r\nfloat3 CrossProd = cross(WorldPosRight - CenterWorldPos, WorldPosUp - CenterWorldPos);\r\nfloat CrossProdLength = length(CrossProd);\r\n\r\n// default to upward facing normal\r\nif (CrossProdLength < 1e-5)\r\n{\r\n    Normal = float3(0,0,1);\r\n}\r\nelse\r\n{\r\n    Normal = CrossProd / CrossProdLength;\r\n}\r\n\r\nreturn Normal;"
   showCode=true
   inputs [
    {inputName=UV, input {expression=&self:MaterialExpressionFunctionInput_0}}
    {inputName=Texture, input {expression=&self:MaterialExpressionFunctionInput_1}}
    {
     inputName=WorldPos
     input {expression=&self:MaterialExpressionWorldPosition_0, mask=1, maskB=1, maskG=1, maskR=1}
    }
    {inputName=dx, input {expression=&self:MaterialExpressionDivide_0}}
    {inputName=GridRes, input {expression=&self:MaterialExpressionCollectionParameter_1}}
    {inputName=h, input {expression=&self:MaterialExpressionStaticSwitchParameter_0}}
    {inputName=h_right, input {expression=&self:MaterialExpressionStaticSwitchParameter_2}}
    {inputName=h_up, input {expression=&self:MaterialExpressionStaticSwitchParameter_3}}
   ]
  }
  FunctionOutput(None=StaticSwitchParameter(True=TextureSample(UVs=@MaterialExpressionFunctionInput_0, Tex=@MaterialExpressionFunctionInput_1).RGB, False=@n4) {defaultValue=true, parameterName=Precomputed Normals})
 ]}
