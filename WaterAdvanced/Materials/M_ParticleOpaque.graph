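# M_ParticleOpaque is a minimal opaque Default Lit material in WaterAdvanced, named for use on particles.
# Its graph holds a single Particle Color node, with no textures, parameters or math, and shows no link from that node to a material output.
# Pixel depth offset is set to legacy mode.
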Material /WaterAdvanced/Materials/M_ParticleOpaque
graph {properties {pixelDepthOffsetMode=PDOM_Legacy, usedShadingModels=Default Lit}, nodes [ParticleColor()]}
