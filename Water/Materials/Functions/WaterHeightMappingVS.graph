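# Material function that conforms water surfaces vertically to the height defined by the Water system's water bodies.
# It subtracts world position Z from the Water Z Height output of WaterBodyData and applies the difference along the Z axis as a world position offset.
# It exposes that offset directly and also outputs the input material attributes with it added to their existing WorldPositionOffset.
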MaterialFunction /Water/Materials/Functions/WaterHeightMappingVS
graph {
 properties {description=This function conforms Water surfaces to the Water features specified using the Water system.}
 nodes [
  n1 = WorldPosition()
  n2 = Multiply(
   A=Subtract(
    A=MaterialFunctionCall!(Position) {materialFunction=&/WaterBodyData}.Water Z Height
    B=ComponentMask(None=@n1.XYZ) {b=true}
   )
   B=Constant3Vector() {constant=<rgba 0 0 1 0>}
  )
  FunctionOutput(None=@n2) {outputName=WorldPositionOffset, sortPriority=1}
  n3 = GetMaterialAttributes(None=FunctionInput() {inputName=Attributes, inputType=FunctionInput_MaterialAttributes}) {attributeGetTypes [#1]}
  FunctionOutput(None=SetMaterialAttributes!(MaterialAttributes=@n3.MaterialAttributes, WorldPositionOffset=Add(A=@n3.WorldPositionOffset, B=@n2)) {attributeSetTypes [#1]}) {outputName=Attributes}
 ]
}
