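# Unlit water material that previews one channel of a paint render target over the baked water simulation area, judging by its parameter names.
# World XY is remapped into PaintRT UVs using the baked sim location and size, centered at 0.5.
# A Custom node turns PreviewChannel into an RGB mask (all channels beyond index 2), and the masked sample's length gives a grayscale value.
# The graph also includes the Water plugin's WaterHeightMappingVS function.
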Material /WaterAdvanced/Materials/M_SW_PaintPreview
graph {
 properties {bUsedWithWater=true, shadingModel=MSM_Unlit, usedShadingModels=Unlit}
 nodes [
  n1 = ComponentMask(None=WorldPosition().XYZ) {g=true, r=true}
  n2 = ComponentMask(None=VectorParameter() {parameterName=BakedWaterSimLocation}.RGB) {g=true, r=true}
  n3 = VectorParameter() {parameterName=BakedWaterSimSize, defaultValue=<rgba 1 1 1 1>}
  n4 = Add(A=Divide(A=Subtract(A=@n1, B=@n2), B=ComponentMask(None=@n3.RGB) {g=true, r=true}), B=Constant2Vector() {g=0.5, r=0.5})
  MaterialExpressionScalarParameter_0 = ScalarParameter() {parameterName=PreviewChannel}
  n5 = Custom!(a=@MaterialExpressionScalarParameter_0) {
   code="a == 0 ? float3(1,0,0) : a == 1 ? float3(0,1,0) : a == 2 ? float3(0,0,1) : float3(1,1,1);"
   showCode=true
   inputs [{inputName=a, input {expression=&self:MaterialExpressionScalarParameter_0}}]
  }
  Length(None=Multiply(
    A=TextureSampleParameter2D(UVs=@n4) {parameterName=PaintRT, samplerType=SAMPLERTYPE_LinearColor, texture=&/T_PaintPreviewDefault}.RGB
    B=@n5
   )) {is=class MaterialExpressionLength}
  MaterialFunctionCall!(Attributes) {materialFunction=/Water/Materials/Functions/WaterHeightMappingVS}
 ]
}
