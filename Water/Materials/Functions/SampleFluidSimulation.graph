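# Material function that maps the fluid simulation's normal, height and velocity textures into world space and layers them onto water material attributes.
# Height drives clamped vertical displacement, blended with waves through an optional distance field mask near objects.
# Edge-masked velocity and wave offset drive flowmapped foam, enabled by static bool, affecting normal, opacity, roughness and emissive.
# It also outputs foam scattering and water depth, with debug views for bounds and velocity.
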MaterialFunction /Water/Materials/Functions/SampleFluidSimulation
graph {
 properties {
  bExposeToLibrary=true
  description=Maps the Fluid Simulation Normal and Height into World Space for use in water materials in the world.
  libraryCategoriesText [Water]
 }
 nodes [
  n1 = FunctionInput() {inputName=Attributes, inputType=FunctionInput_MaterialAttributes}
  n2 = GetMaterialAttributes(None=@n1) {attributeGetTypes [
    #1
    #2
    #3
    #4
    #5
   ]}
  n3 = ScalarParameter() {group=FluidSimulation, parameterName=Wave over Rock Steepness}
  n4 = Multiply(
   A=LinearInterpolate(B=MaterialFunctionCall(X=@n3, Y=@n3) {materialFunction=&MakeFloat3}.Result) {constA=1}
   B=GetMaterialAttributes(None=@n1) {attributeGetTypes [#2]}.WorldPositionOffset
  )
  n5 = WorldPosition() {worldPositionShaderOffset=WPT_ExcludeAllShaderOffsets}
  n6 = TextureSampleParameter2D(UVs=MaterialFunctionCall!(WorldPosition=Add(A=@n4, B=@n5.XYZ)) {materialFunction=/Water/FluidSimulation/Materials/Functions/Fluidsim_WorldPosition_to_UV}.UV) {
   parameterName=NormalAndHeight
   samplerSource=SSM_Clamp_WorldGroupSettings
   samplerType=SAMPLERTYPE_LinearColor
   texture=/Engine/Functions/Engine_MaterialFunctions02/PivotPainter2/Black_1x1_EXR_Texture
  }
  n7 = AppendVector(
   A=Multiply(
    A=ComponentMask(None=@n6.RGB) {g=true, r=true}
    B=ScalarParameter() {defaultValue=1, group=FluidSimulation, parameterName=Fluidsim Normal Strength}
   )
   B=Constant()
  )
  n8 = MaterialFunctionCall() {materialFunction=&Fluidsim_WorldPosition_to_UV}
  n9 = TextureSampleParameter2D(UVs=@n8.UV) {
   parameterName=Velocity
   samplerSource=SSM_Clamp_WorldGroupSettings
   samplerType=SAMPLERTYPE_LinearColor
   texture=&Black_1x1_EXR_Texture
  }
  n10 = Constant() {r=0.5}
  n11 = ScalarParameter() {defaultValue=0.2, group=FluidSimulation, parameterName=Fluidsim Velocity EdgeFalloff}
  n12 = OneMinus(None=Multiply(A=@n11) {constB=2})
  n13 = Multiply(
   A=@n9.RGB
   B=MaterialFunctionCall!(
    Alpha=MaterialFunctionCall!(A=@n8.UV, B=@n10, Bounds=@n12, Edge Falloff=@n11) {materialFunction=/Engine/Functions/Engine_MaterialFunctions01/Gradient/BoxMask-2D}.Result
    Offset=ScalarParameter() {defaultValue=0.02, group=FluidSimulation, parameterName=Fluidsim Velocity Edgeoffset}
   ) {materialFunction=/Engine/Functions/Engine_MaterialFunctions01/Opacity/AlphaOffset}.Result
  )
  n14 = MaterialFunctionCall(Vector 2=@n13) {materialFunction=/Engine/Functions/Engine_MaterialFunctions02/Utility/VectorLength}
  n15 = Add(
   A=Divide(A=@n14.V2 Length, B=ScalarParameter() {defaultValue=40, group=FluidSimulation, parameterName=Sim Velocity Normal Strength})
   B=ScalarParameter() {defaultValue=1, group=FluidSimulation, parameterName=Sim Foam Normal}
  )
  n16 = CollectionParameter() {collection=/Water/Materials/MPC/MPC_Water, parameterName=SimLocation}
  n17 = Add(
   A=Subtract(A=WorldPosition().XYZ, B=ComponentMask(None=@n16) {b=true, g=true, r=true})
   B=GetMaterialAttributes(None=@n1) {attributeGetTypes [#2]}.WorldPositionOffset
  )
  n18 = CollectionParameter() {collection=&MPC_Water, parameterName=FluidSimSize}
  n19 = ComponentMask(None=Add(A=Divide(A=@n17, B=@n18)) {constB=0.5}) {g=true, r=true}
  n20 = TextureSample(UVs=Saturate(None=@n19)) {samplerSource=SSM_Clamp_WorldGroupSettings, samplerType=SAMPLERTYPE_LinearColor, texture=/Water/FluidSimulation/WaveFoam/WaterDistanceField}
  n21 = StaticSwitchParameter(True=Constant() {r=1}, False=Constant()) {
   desc=This advanced option references a custom Distance Field captured by the BP_FluidSim_01. References will be automatically configured for WaterBody materials using the SampleFluidSimulation material function.
   group=FluidSimulation
   parameterName=Fluidsim Use Custom DistanceField
  }
  n22 = Saturate(None=Multiply(A=@n20.R, B=@n21))
  n23 = ComponentMask(None=@n2.WorldPositionOffset) {g=true, r=true}
  n24 = Divide(A=Multiply(A=Multiply(A=@n22, B=@n23), B=ScalarParameter() {defaultValue=6, group=FluidSimulation, parameterName=WPO Flowmap})) {constB=1000}
  n25 = MaterialFunctionCall(Vector=@n13) {materialFunction=&SafeNormalize}
  MaterialExpressionMultiply_33 = Multiply(
   A=Divide(A=MaterialFunctionCall(Vector 2=@n13) {materialFunction=&VectorLength}.V2 Length) {constB=100}
   B=ScalarParameter() {defaultValue=0.25, group=FluidSimulation, parameterName=Fluidsim Flow Density}
  )
  n26 = Custom!(x=@MaterialExpressionMultiply_33) {
   code="1.0-exp(-x);"
   description=exp
   outputType=CMOT_Float1
   inputs [{inputName=x, input {expression=&self:MaterialExpressionMultiply_33}}]
  }
  n27 = Multiply(
   A=ComponentMask(None=Multiply(A=@n25.Result, B=@n26)) {g=true, r=true}
   B=ScalarParameter() {defaultValue=0.75, group=FluidSimulation, parameterName=Simulation Flowmap Speed}
  )
  n28 = Time()
  n29 = LinearInterpolate(
   A=WorldPosition().XYZ
   B=WorldPosition() {worldPositionShaderOffset=WPT_ExcludeAllShaderOffsets}.XYZ
   Alpha=ScalarParameter() {defaultValue=0.4, group=General Foam, parameterName=Foam Attachment}
  )
  n30 = ScalarParameter() {defaultValue=350, group=FluidSimulation, parameterName=Sim Foam Scale}
  n31 = MaterialFunctionCall!(
   Diffuse=TextureObject() {texture=/Water/Textures/Foam/T_WaterFlow_01_Foam_Tiled}
   Normal Map=TextureObject() {samplerType=SAMPLERTYPE_Normal, texture=/Water/Textures/Foam/T_WaterFlow_01_Foam_Tiled_N}
   "Flow Vector Map (see tooltip)"=Multiply(A=Add(A=@n24, B=@n27), B=Constant2Vector() {g=-1, r=1})
   UV texture 2 offset
   "Time (see tooltip)"=MaterialFunctionCall(
    Time=FunctionInput(Preview=@n28) {bUsePreviewValueAsDefault=true, inputName=Time, inputType=FunctionInput_Scalar, sortPriority=1}
    Speed=ScalarParameter() {defaultValue=1.5, group=FluidSimulation, parameterName=SIm Flowmap Speed}
   ) {materialFunction=&TimeWithSpeedVariable}.Time
   useMipLevel
   UVs=ComponentMask(None=Divide(A=@n29, B=@n30)) {g=true, r=true}
   Texture Size
   Texture Mip Bias=Divide(A=@n14.V2 Length, B=ScalarParameter() {defaultValue=128, group=FluidSimulation, parameterName=Sim Foam Velocity Mip})
  ) {materialFunction=/Engine/Functions/Engine_MaterialFunctions02/Texturing/Flowmaps.FlowMaps}
  n32 = Multiply(
   A=MaterialFunctionCall(X=@n15, Y=@n15, Z=Constant()) {materialFunction=&MakeFloat3}.Result
   B=@n31.Normal
  )
  MaterialExpressionDivide_1 = Divide(A=@n6.A, B=ScalarParameter() {defaultValue=1.5, group=FluidSimulation, parameterName=Sim Flowmap Detection})
  n33 = Custom(x=@MaterialExpressionDivide_1) {
   code="return 1-exp(-x);"
   description=exp
   outputType=CMOT_Float1
   inputs [{inputName=x, input {expression=&self:MaterialExpressionDivide_1}}]
  }
  n34 = MaterialFunctionCall(
   Transition Phase=FunctionOutput(None=Saturate(None=@n33)) {outputName=Foam Scattering, sortPriority=1}
   Height Texture=ComponentMask(None=@n31.Diffuse) {r=true}
   Contrast=ScalarParameter() {group=FluidSimulation, parameterName=SimFoam Contrast}
  ) {materialFunction=&HeightLerp}
  n35 = SceneDepthWithoutWater!(UVs)
  n36 = PixelDepth()
  n37 = Saturate(None=Divide(A=Subtract(A=@n35.R, B=@n36.R), B=ScalarParameter() {defaultValue=8, group=FluidSimulation, parameterName=Sim Foam Dpeth Min}))
  n38 = FunctionInput(Preview=StaticBool()) {
   bUsePreviewValueAsDefault=true
   description="By default, foam is disabled and the fluidsim will noy affect WPO and Normals. Enable this to also add flowmap based foam with Opacity, Roughness and some Emissive Scattering."
   inputName=Enable Foam
   inputType=FunctionInput_StaticBool
   sortPriority=2
  }
  n39 = Multiply(A=Saturate(None=Multiply(A=@n34.Alpha, B=@n37)), B=StaticSwitch(True=Constant() {r=1}, False=Constant(), Value=@n38))
  n40 = Multiply(A=Abs(None=Subtract(A=@n19) {constB=0.5})) {constB=2}
  n41 = MaterialFunctionCall(Float2=Saturate(None=Floor(None=@n40))) {materialFunction=&BreakOutFloat2Components}
  n42 = Multiply(A=Abs(None=Subtract(A=@n19) {constB=0.5})) {constB=1.975}
  n43 = MaterialFunctionCall(Float2=Saturate(None=Floor(None=@n42))) {materialFunction=&BreakOutFloat2Components}
  n44 = Max(A=@n43.R, B=@n43.R)
  n45 = StaticSwitchParameter(True=Multiply(A=Max(A=@n41.R, B=@n41.R), B=OneMinus(None=@n44)), False=Constant()) {group=FluidSimulation, parameterName=Show FluidSim Debug Bounds}
  n46 = VectorParameter() {group=General Foam, parameterName=Foam Emissive, defaultValue=<rgba 0.2874 1 0.1878 1>}
  n47 = Multiply(A=Multiply(A=Multiply(A=@n46.RGB, B=@n46.RGB), B=SkyAtmosphereDistantLightScatteredLuminance()), B=@n39)
  n48 = Multiply(A=Abs(None=@n13), B=ScalarParameter() {group=FluidSimulation, parameterName=Velocity Debug})
  n49 = ScalarParameter() {defaultValue=1, group=General Foam, parameterName=Foam Opacity}
  n50 = ScalarParameter() {defaultValue=0.4, group=General Foam, parameterName=Foam Roughness}
  n51 = ScalarParameter() {group=FluidSimulation, parameterName=Wave over Rock Steepness}
  n52 = Multiply(A=@n2.WorldPositionOffset) {desc=WPO}
  n53 = Multiply(A=@n52, B=MaterialFunctionCall(X=@n51, Y=@n51) {materialFunction=&MakeFloat3}.Result)
  n54 = Multiply(
   A=ScalarParameter() {defaultValue=1, group=FluidSimulation, parameterName=Fluid Displacement Multiplier}
   B=ScalarParameter() {defaultValue=1, group=FluidSimulation, parameterName=ShowSim}
  )
  n55 = ScalarParameter() {defaultValue=1024, group=FluidSimulation, parameterName=Max Fluid Displacement}
  n56 = Multiply(A=Clamp(None=Multiply(A=@n6.B, B=@n54), Min=Multiply(A=@n55) {constB=-1}, Max=@n55), B=Constant3Vector() {constant=<rgba 0 0 1 0>})
  n57 = Multiply(A=Max(A=ComponentMask(None=@n52) {b=true}, B=ComponentMask(None=@n56) {b=true}), B=Constant3Vector() {constant=<rgba 0 0 1 0>})
  n58 = Add(
   A=Multiply(A=@n20.R, B=CollectionParameter() {collection=&MPC_Water, parameterName=FluidSimSize})
   B=ScalarParameter() {defaultValue=-64, group=FluidSimulation, parameterName=Wave DF Mask Bias}
  )
  n59 = ScalarParameter() {defaultValue=256, group=FluidSimulation, parameterName=Wave Over Object Mask Distance}
  n60 = OneMinus(None=Saturate(None=Divide(A=@n58, B=@n59)))
  n61 = SetMaterialAttributes!(
   MaterialAttributes=@n2.MaterialAttributes
   Normal=Add(A=Add(A=@n2.Normal, B=@n7), B=Multiply(A=@n32, B=@n39))
   WorldPositionOffset=LinearInterpolate(A=Add(A=@n52, B=@n56), B=Add(A=@n53, B=@n57), Alpha=Multiply(A=@n60, B=@n21))
   EmissiveColor=Add(A=Add(A=@n45, B=@n47), B=Add(A=@n2.EmissiveColor, B=@n48)) {constB=0, desc=Emissive}
   Opacity=Add(A=Saturate(None=@n2.Opacity), B=Multiply(A=@n49, B=@n39)) {desc=Opacity}
   Roughness=Add(A=@n2.Roughness, B=Multiply(A=@n50, B=@n39)) {desc=Roughness}
  ) {attributeSetTypes [
    #1
    #2
    #3
    #4
    #5
   ]}
  FunctionOutput(None=@n61) {outputName=Attributes}
  n62 = MaterialFunctionCall!(Position) {materialFunction=&/WaterBodyData}
  n63 = Subtract(A=Add(A=@n62.Water Z Height, B=@n6.B), B=@n62.Terrain Z Height)
  FunctionOutput(None=Max(A=@n63) {constB=0}) {outputName=Water Depth, sortPriority=2}
  Saturate(None=Divide(A=@n63, B=ScalarParameter() {defaultValue=100, group=FluidSimulation, parameterName=Water Depth to Mask Waves}) {constB=50})
 ]
}
