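# Material function that makes Single Layer Water see-through where a depth mask fades, zeroing Opacity and Specular and blending Refraction to 1.
# WaterBodyData depth, shifted by an offset parameter, is maxed with an optional additional mask, remapped and saturated, then also output as Mask.
# A Custom node on BaseColor clips hit-proxy shaders using unoffset depth and the same additional mask, trimming the editor selection outline.
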MaterialFunction /Water/Materials/Functions/WaterOpacityMaskFromDepth
graph {
 properties {
  bExposeToLibrary=true
  description="This function can make a SingleLayerWater material appear see-through by setting Opacity, Specular to 0 and Refraction to 1, using a Depth Mask. Also clips the editor selection outline."
  libraryCategoriesText [Water]
 }
 nodes [
  MaterialExpressionGetMaterialAttributes_0 = GetMaterialAttributes(None=FunctionInput() {inputName=Attributes, inputType=FunctionInput_MaterialAttributes}) {attributeGetTypes [
    #1
    #2
    #3
    #4
   ]}
  n1 = MaterialFunctionCall!(Position) {materialFunction=&/WaterBodyData}
  n2 = FunctionInput() {
   bUsePreviewValueAsDefault=true
   description="This input will perform a Max operation with the given depth based opacity. By using the adjusted depth from the SampleFluidSimulation function, this allows the opacity to be correctly clipped when the simulation causes water to appear in new areas, or drain in others."
   inputName=Additional Water Mask
   inputType=FunctionInput_Scalar
   sortPriority=1
  }
  n3 = Saturate(None=@n2)
  MaterialExpressionLinearInterpolate_2 = LinearInterpolate(Alpha=Max(A=Saturate(None=@n1.Water Depth), B=@n3)) {constA=-1}
  n4 = Custom!(Color=@MaterialExpressionGetMaterialAttributes_0.BaseColor, x=@MaterialExpressionLinearInterpolate_2) {
   code="##if HIT_PROXY_SHADER\r\nclip(x);\r\n##endif\r\nreturn Color;"
   description="Hit Proxy Clip()"
   inputs [
    {
     inputName=Color
     input {expression=&self:MaterialExpressionGetMaterialAttributes_0, outputIndex=1}
    }
    {inputName=x, input {expression=&self:MaterialExpressionLinearInterpolate_2}}
   ]
  }
  n5 = ScalarParameter() {defaultValue=-24, group=Opacity Mask, parameterName=Water Opacity Mask Offset}
  n6 = Max(A=Saturate(None=Subtract(A=@n1.Water Depth, B=@n5)), B=@n3)
  n7 = FunctionOutput(None=Saturate(None=LinearInterpolate(Alpha=@n6) {constA=-1})) {outputName=Mask, sortPriority=1}
  n8 = SetMaterialAttributes!(
   MaterialAttributes=@MaterialExpressionGetMaterialAttributes_0.MaterialAttributes
   BaseColor=@n4
   Opacity=Multiply(A=@MaterialExpressionGetMaterialAttributes_0.Opacity, B=@n7)
   Specular=Multiply(A=@MaterialExpressionGetMaterialAttributes_0.Specular, B=@n7)
   Refraction=LinearInterpolate(B=@MaterialExpressionGetMaterialAttributes_0.Refraction, Alpha=@n7) {constA=1}
  ) {attributeSetTypes [
    #1
    #2
    #3
    #4
   ]}
  FunctionOutput(None=@n8)
 ]
}
