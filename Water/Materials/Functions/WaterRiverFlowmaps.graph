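# Material function for the Water system that adds flowmapped normals to rivers.
# WaterBodyData velocity, attenuated by an exponential curve of biased depth, drives world-XY flowmaps, with scrolling noise offsetting time to hide repetition.
# Flattened normals blend over the incoming attribute by velocity length weighted by unbiased depth, fading out in slow or shallow water.
# The flowmap diffuse is weighted likewise but never reaches the output.
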MaterialFunction /Water/Materials/Functions/WaterRiverFlowmaps
graph {
 properties {bExposeToLibrary=true, description=This function provides the basic flowmaps for Rivers in the Water system., libraryCategoriesText [Water]}
 nodes [
  n1 = GetMaterialAttributes(None=FunctionInput() {inputName=MaterialAttributes, inputType=FunctionInput_MaterialAttributes}) {attributeGetTypes [#1]}
  n2 = MaterialFunctionCall!(Position) {materialFunction=&/WaterBodyData}
  n3 = ScalarParameter() {defaultValue=64, group=River Flowmaps, parameterName=River Flowmap Depth Bias}
  MaterialExpressionMultiply_0 = Multiply(
   A=ScalarParameter() {defaultValue=0.003, group=River Flowmaps, parameterName=River Flowmap Depth Density}
   B=AppendVector(A=Add(A=@n2.Water Depth, B=@n3), B=@n2.Water Depth)
  )
  n4 = Custom!(x=@MaterialExpressionMultiply_0) {
   code="return 1-exp(-x);"
   description=exp
   outputType=CMOT_Float2
   inputs [{inputName=x, input {expression=&self:MaterialExpressionMultiply_0}}]
  }
  n5 = MaterialFunctionCall(Float2=@n4) {materialFunction=&BreakOutFloat2Components}
  n6 = Multiply(A=@n5.R, B=@n2.Velocity)
  n7 = ScalarParameter() {defaultValue=0.003, group=River Flowmaps, parameterName=River Flowmap Speed}
  n8 = WorldPosition() {worldPositionShaderOffset=WPT_ExcludeAllShaderOffsets}
  n9 = ScalarParameter() {defaultValue=512, group=River Flowmaps, parameterName=River Foam Scale}
  n10 = ScalarParameter() {defaultValue=2750, group=River Flowmaps, parameterName=River Time Modulation Size}
  n11 = Constant() {r=0.02}
  n12 = MaterialFunctionCall(
   Time=FunctionInput(Preview=Time()) {bUsePreviewValueAsDefault=true, inputName=Time, inputType=FunctionInput_Scalar, sortPriority=1}
   Speed=Constant() {r=0.5}
  ) {materialFunction=&TimeWithSpeedVariable}
  n13 = TextureSample!(
   UVs=Add(
    A=MaterialFunctionCall!(Scale=@n10) {materialFunction=/Engine/Functions/Engine_MaterialFunctions01/Coordinates/WorldPosition-XY}.Result
    B=Multiply(A=@n11, B=@n12.Time)
   )
   Tex
   Level
   Apply View MipBias
  ) {
   desc=noise to break up the repetition
   mipValueMode=TMVM_MipLevel
   samplerType=SAMPLERTYPE_LinearGrayscale
   texture=/Engine/Functions/Engine_MaterialFunctions02/ExampleContent/Textures/LowResBlurredNoise
  }
  n14 = MaterialFunctionCall!(
   Diffuse=TextureObjectParameter() {parameterName=RiverFlowmapDiffuse, texture=/Water/Textures/Flowmaps/RiverFlowmap_Diffuse}
   Normal Map=TextureObjectParameter() {parameterName=RiverFlowmapNormal, samplerType=SAMPLERTYPE_Normal, texture=/Water/Textures/Flowmaps/RiverFlowmap_Normal}
   "Flow Vector Map (see tooltip)"=Multiply(A=Multiply(A=@n6, B=@n7), B=Constant2Vector() {g=-1, r=1})
   UV texture 2 offset
   "Time (see tooltip)"=Add(A=Multiply(A=@n13.R) {constB=0.25}, B=@n12.Time)
   useMipLevel
   UVs=ComponentMask(None=Divide(A=@n8.XYZ, B=@n9)) {g=true, r=true}
   Texture Size
   Texture Mip Bias
  ) {materialFunction=/Engine/Functions/Engine_MaterialFunctions02/Texturing/Flowmaps.FlowMaps}
  n15 = MaterialFunctionCall(Normal=@n14.Normal, Flatness=ScalarParameter() {defaultValue=0.5, group=River Flowmaps, parameterName=River Normal Flatness}) {materialFunction=&FlattenNormal}
  n16 = Multiply(A=@n5.G, B=@n2.Velocity)
  n17 = Saturate(None=Divide(
    A=MaterialFunctionCall(Vector 2=@n16) {materialFunction=/Engine/Functions/Engine_MaterialFunctions02/Utility/VectorLength}.V2 Length
    B=ScalarParameter() {defaultValue=64, group=River Flowmaps, parameterName=River Flowmap Detection Velocity}
   ))
  FunctionOutput(None=SetMaterialAttributes!(MaterialAttributes=@n1.MaterialAttributes, Normal=LinearInterpolate(A=@n1.Normal, B=@n15.Result, Alpha=@n17)) {attributeSetTypes [#1]})
  Multiply(A=@n14.Diffuse, B=@n17)
 ]
}
