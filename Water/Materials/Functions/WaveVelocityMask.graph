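# Material function for Water materials that masks out waves where velocity is detected, handling River-to-Ocean and River-to-Lake transitions.
# It divides the velocity length from WaterBodyData by the "Velocity to Mask waves" scalar parameter and saturates the result into a blend alpha.
# That alpha lerps the incoming wave attributes' Normal toward straight up (+Z) and World Position Offset toward zero, passing other attributes through.
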MaterialFunction /Water/Materials/Functions/WaveVelocityMask
graph {
 properties {
  bExposeToLibrary=true
  description="This function masks out Waves when Velocity is detected. This is how River->Ocean and River->Lake transitions are handled for Water materials. "
  libraryCategoriesText [Water]
 }
 nodes [
  n1 = GetMaterialAttributes(None=FunctionInput() {inputName=WaveAttributes, inputType=FunctionInput_MaterialAttributes}) {attributeGetTypes [#1, #2]}
  n2 = Constant3Vector() {constant=<rgba 0 0 1 0>}
  n3 = MaterialFunctionCall!(Position) {materialFunction=&/WaterBodyData}
  n4 = Saturate(None=Divide(
    A=MaterialFunctionCall(Vector 2=@n3.Velocity) {materialFunction=/Engine/Functions/Engine_MaterialFunctions02/Utility/VectorLength}.V2 Length
    B=ScalarParameter() {defaultValue=100, parameterName=Velocity to Mask waves}
   ))
  FunctionOutput(None=SetMaterialAttributes!(
    MaterialAttributes=@n1.MaterialAttributes
    Normal=LinearInterpolate(A=@n1.Normal, B=@n2, Alpha=@n4)
    WorldPositionOffset=LinearInterpolate(A=@n1.WorldPositionOffset, Alpha=@n4) {constB=0}
   ) {attributeSetTypes [#1, #2]})
 ]
}
