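# Material function that layers beach wave foam onto incoming water material attributes.
# Samples the FoamRT render target around the simulation location, optionally masks it near the fluid sim box, then height-blends it with tiled foam and noise textures.
# Depth-faded foam drives roughness, opacity, heightmap-derived normals and sky-lit emissive.
# Also outputs foam opacity and blurred-mip scattering; a distance-field foam branch is left unconnected.
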MaterialFunction /Water/Materials/Layers/BeachFoam
graph {nodes [
  n1 = GetMaterialAttributes(None=FunctionInput() {inputName=MA, inputType=FunctionInput_MaterialAttributes}) {attributeGetTypes [
    #1
    #2
    #3
    #4
   ]}
  n2 = ScalarParameter() {defaultValue=0.4, parameterName=Foam Roughness}
  n3 = Divide(
   A=WorldPosition() {worldPositionShaderOffset=WPT_ExcludeAllShaderOffsets}.XYZ
   B=Multiply(
    A=CollectionParameter() {collection=/Water/Materials/MPC/MPC_Water, parameterName=FluidSimSize}
    B=ScalarParameter() {defaultValue=0.5, parameterName=Fluid Eraose Ssaler}
   )
  )
  n4 = ScalarParameter() {defaultValue=0.5, parameterName=Wave Foam Erase Near FluidsimSize}
  n5 = MaterialFunctionCall!(
   A=Add(A=ComponentMask(None=@n3) {g=true, r=true}) {constB=0.5}
   B=Constant() {r=0.5}
   Bounds=OneMinus(None=Multiply(A=@n4) {constB=2})
   Edge Falloff=@n4
  ) {materialFunction=/Engine/Functions/Engine_MaterialFunctions01/Gradient/BoxMask-2D}
  n6 = LinearInterpolate(
   A=WorldPosition().XYZ
   B=WorldPosition() {worldPositionShaderOffset=WPT_ExcludeAllShaderOffsets}.XYZ
   Alpha=ScalarParameter() {defaultValue=0.4, parameterName=Foam Attachment}
  )
  n7 = ComponentMask(None=@n6) {g=true, r=true}
  n8 = ComponentMask(None=CollectionParameter() {collection=&MPC_Water, parameterName=SimLocation}) {g=true, r=true}
  n9 = Add(A=Divide(A=Subtract(A=@n7, B=@n8), B=CollectionParameter() {collection=&MPC_Water, parameterName=WaveCaptureSize})) {constB=0.5}
  n10 = TextureSample(UVs=@n9) {samplerSource=SSM_Clamp_WorldGroupSettings, samplerType=SAMPLERTYPE_LinearColor, texture=/Water/FluidSimulation/WaveFoam/FoamRT}
  n11 = Multiply(
   A=Power(Base=@n10.R, Exp=ScalarParameter() {defaultValue=0.2, parameterName=Foam powr})
   B=ScalarParameter() {defaultValue=1, parameterName=Foam Boost}
  )
  n12 = Multiply(
   A=LinearInterpolate(B=OneMinus(None=@n5.Result), Alpha=ScalarParameter() {defaultValue=1, parameterName=Mask Wave Foam near FluidSim}) {constA=1}
   B=@n11
  )
  n13 = Divide(A=@n7, B=ScalarParameter() {defaultValue=1536, parameterName=Foam MacroScale})
  n14 = TextureSample(UVs=@n13) {samplerSource=SSM_Wrap_WorldGroupSettings, texture=/Water/Textures/Foam/T_WaterFlow_01_Foam_Tiled}
  n15 = ScalarParameter() {defaultValue=-1, parameterName=Height Bias}
  n16 = TextureSample(UVs=Divide(A=@n6, B=ScalarParameter() {defaultValue=654, parameterName=Front Foam Scale})) {samplerSource=SSM_Wrap_WorldGroupSettings, samplerType=SAMPLERTYPE_Masks, texture=/Engine/EngineSky/VolumetricClouds/T_VolumeNoiseShape128}
  n17 = Add(A=Multiply(A=@n16.R, B=@n16.R), B=ScalarParameter() {defaultValue=-0.5, parameterName=Front Height Bias})
  n18 = ScalarParameter() {defaultValue=0.4, parameterName=Foam Texture Blend Min}
  n19 = Saturate(None=Divide(A=Subtract(A=@n10.R, B=@n18), B=ScalarParameter() {defaultValue=0.4, parameterName=Foam Texture Blend Width}))
  n20 = ScalarParameter() {defaultValue=1, parameterName=FoamContrast}
  n21 = MaterialFunctionCall(Transition Phase=@n12, Height Texture=LinearInterpolate(A=Add(A=@n14.R, B=@n15), B=@n17, Alpha=@n19), Contrast=@n20) {materialFunction=&HeightLerp}
  n22 = Divide(A=Subtract(A=SceneDepthWithoutWater!(UVs).R, B=PixelDepth().R), B=ScalarParameter() {defaultValue=64, parameterName=Foam Dpeth Min})
  n23 = Saturate(None=Multiply(A=@n21.Alpha, B=Saturate(None=@n22)))
  n24 = Saturate(None=Multiply(A=@n23, B=ScalarParameter() {defaultValue=1, parameterName=Foam Opacity}))
  n25 = ComponentMask(None=@n1.Normal) {g=true, r=true}
  n26 = ScalarParameter() {defaultValue=2, parameterName=Foam Wave normal Z}
  n27 = Normalize(VectorInput=Multiply(A=AppendVector(A=@n25, B=@n26), B=TwoSidedSign()))
  n28 = MaterialFunctionCall!(
   Height Map=TextureObject() {samplerType=SAMPLERTYPE_LinearColor, texture=&FoamRT}
   Normal Map Intensity=ScalarParameter() {defaultValue=0.5, parameterName=Foam Normal I}
   Height Map UV Offset=ScalarParameter() {defaultValue=0.002, parameterName=Foam normal offset}
   Coordinates=@n9
   Height Map Channel Selector
  ) {materialFunction=/Engine/Functions/Engine_MaterialFunctions03/Procedurals/NormalFromHeightmap}
  n29 = Constant3Vector() {constant=<rgba 1 1 0 0>}
  n30 = TextureSample(UVs=@n13) {samplerSource=SSM_Wrap_WorldGroupSettings, samplerType=SAMPLERTYPE_Normal, texture=/Water/Textures/Foam/T_WaterFlow_01_Foam_Tiled_N}
  n31 = VectorParameter() {parameterName=foramnormal, defaultValue=<rgba 0.5 0.5 1 0>}
  n32 = Multiply(A=Constant3Vector() {constant=<rgba 1 1 0 0>}, B=Add(A=Multiply(A=@n28.Result, B=@n29), B=Multiply(A=@n30.RGB, B=@n31.RGB)))
  n33 = VectorParameter() {parameterName=Foam Emissive, defaultValue=<rgba 0.2874 1 0.1878 1>}
  n34 = SetMaterialAttributes!(
   MaterialAttributes=@n1.MaterialAttributes
   Specular=@n1.MaterialAttributes
   Roughness=Add(A=@n1.Roughness, B=Multiply(A=@n2, B=@n23))
   Opacity=Add(A=@n1.Opacity, B=@n24)
   Normal=LinearInterpolate(A=@n1.Normal, B=Add(A=@n27, B=@n32), Alpha=FunctionOutput(None=@n24) {outputName=Foam Opacity, sortPriority=1})
   EmissiveColor=Multiply(A=Multiply(A=@n33.RGB, B=@n24), B=SkyAtmosphereDistantLightScatteredLuminance())
  ) {attributeSetTypes [
    #1
    #2
    #3
    #4
    #5
   ]}
  FunctionOutput(None=@n34) {outputName=Attributes}
  n35 = Add(
   A=TextureSample!(UVs=@n13, Tex, Level=ScalarParameter() {defaultValue=8, parameterName=Foam Scattering Mip}, Apply View MipBias) {mipValueMode=TMVM_MipLevel, samplerSource=SSM_Wrap_WorldGroupSettings, texture=&T_WaterFlow_01_Foam_Tiled}.R
   B=Add(A=@n15, B=ScalarParameter() {defaultValue=0.25, parameterName=Foam Scatter Bias})
  )
  FunctionOutput(None=MaterialFunctionCall(Transition Phase=@n12, Height Texture=LinearInterpolate(A=@n35, B=@n17, Alpha=@n19), Contrast=@n20) {materialFunction=&HeightLerp}.Lerp Alpha No Contrast) {outputName=Foam Scattering, sortPriority=2}
  n36 = Divide(
   A=MaterialFunctionCall!(Position) {materialFunction=/Water/Materials/Functions/WaterBodyData}.Water Depth
   B=ScalarParameter() {defaultValue=64, parameterName=Depth for DF Foam}
  )
  n37 = WorldPosition() {worldPositionShaderOffset=WPT_ExcludeAllShaderOffsets}
  n38 = DistanceToNearestSurface!(World Position=@n37.XYZ)
  n39 = ScalarParameter() {defaultValue=256, parameterName=Foam Distance}
  n40 = OneMinus(None=Saturate(None=Divide(A=@n38, B=@n39)))
  n41 = Multiply(A=ComponentMask(None=@n37.XYZ) {b=true}) {constB=-1}
  n42 = Saturate(None=Divide(A=Add(A=@n41) {constB=16}) {constB=10})
  Max(A=@n11, B=Multiply(A=Saturate(None=@n36), B=Multiply(A=@n40, B=@n42)))
  Multiply(A=ScalarParameter() {parameterName=Foam Phase Mip Bias}, B=@n19)
  Saturate(None=@n21.Alpha)
  Saturate(None=@n10.G)
 ]}
