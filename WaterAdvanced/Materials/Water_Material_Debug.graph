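# Debug single-layer water material that appears built to display the fluid simulation's NormalAndHeight texture on water surfaces.
# It samples that texture by world position and feeds it through EyeAdaptationInverse into emissive for exposure-independent color, with black base color, zero specular and constant scattering and absorption.
# A fuller water graph (Gerstner waves, river flowmaps, beach foam, fluid sim, depth masking, caustics) is also present but does not feed that debug color.
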Material /WaterAdvanced/Materials/Water_Material_Debug
graph {
 properties {
  bUseMaterialAttributes=true
  bUsedWithInstancedStaticMeshes=true
  bUsedWithSplineMeshes=true
  bUsedWithWater=true
  blendMode=BLEND_Masked
  floatPrecisionMode=MFPM_Full_MaterialExpressionOnly
  pixelDepthOffsetMode=PDOM_Legacy
  refractionMethod=RM_PixelNormalOffset
  shadingModel=MSM_SingleLayerWater
  twoSided=true
  usedShadingModels=SingleLayerWater
 }
 nodes [
  n1 = Constant()
  n2 = TextureSampleParameter2D(UVs=MaterialFunctionCall!(WorldPosition=WorldPosition() {worldPositionShaderOffset=WPT_ExcludeAllShaderOffsets}.XYZ) {materialFunction=/Water/FluidSimulation/Materials/Functions/Fluidsim_WorldPosition_to_UV}.UV) {
   parameterName=NormalAndHeight
   samplerSource=SSM_Clamp_WorldGroupSettings
   samplerType=SAMPLERTYPE_LinearColor
   texture=/Engine/Functions/Engine_MaterialFunctions02/PivotPainter2/Black_1x1_EXR_Texture
  }
  MakeMaterialAttributes!(
   BaseColor=@n1
   Metallic
   Specular=@n1
   Roughness
   Anisotropy
   EmissiveColor=EyeAdaptationInverse(LightValueInput=@n2.RGB).EyeAdaptationInverse
   Opacity
   OpacityMask
   Normal
   Tangent
   WorldPositionOffset
   SubsurfaceColor
   ClearCoat
   ClearCoatRoughness
   AmbientOcclusion
   Refraction
   CustomizedUVs_0
   CustomizedUVs_1
   CustomizedUVs_2
   CustomizedUVs_3
   CustomizedUVs_4
   CustomizedUVs_5
   CustomizedUVs_6
   CustomizedUVs_7
   PixelDepthOffset
   ShadingModel
   Displacement
  )
  n3 = MaterialFunctionCall!(
   Water Body Index=StaticSwitchParameter(
    True=ScalarParameter() {parameterName=WaterBodyIndex}
    False=Custom() {code="GetWaterWaveParamIndex(Parameters)", description=WaterBodyIndex From Tile Instance, outputType=CMOT_Float1, inputs []}
   ) {parameterName=UseScalarWaterBodyIndex}
   Override Position
   Time=MaterialFunctionCall() {materialFunction=/Water/Materials/Functions/GetWaterTime}.WaterTime
   Use Depth Attenuation
  ) {materialFunction=/Water/Materials/Functions/ComputeGerstnerWaves}
  n4 = MaterialFunctionCall!(WaveAttributes=@n3.Wave Attributes) {materialFunction=/Water/Materials/Functions/WaveVelocityMask}
  n5 = Constant3Vector() {constant=<rgba 0 0 1 0>}
  n6 = Constant()
  n7 = StaticSwitchParameter(
   True=@n4.Result
   False=StaticSwitchParameter(
    True=@n4.Result
    False=SetMaterialAttributes!(MaterialAttributes, Normal=@n5, World Position Offset=@n6) {attributeSetTypes [#1, #2]}
   ) {parameterName=Enable Ocean Transition}
  ) {parameterName=Enable Lake Transition}
  n8 = ScalarParameter() {defaultValue=25000, parameterName=Water Distant Normal Length}
  n9 = ScalarParameter() {defaultValue=5000, parameterName=Water Distant Normal Offset}
  n10 = MaterialFunctionCall!(
   Material Atttributes=MaterialFunctionCall!(Attributes=MaterialFunctionCall() {materialFunction=/Water/Materials/Layers/WaterAttributes}.Result, Scattering, Absorption) {materialFunction=/Water/Materials/Functions/Water_Underside}.Result
   Wave Attributes=MaterialFunctionCall!(
    Wave Attributes=StaticSwitchParameter(True=@n3.Wave Attributes, False=@n7) {defaultValue=true, parameterName=Enable Waves}
    Time=MaterialFunctionCall() {materialFunction=&GetWaterTime}.WaterTime
    Distant Scale
    Near Scale
    Blend Alpha=MaterialFunctionCall(Fade Length=@n8, Fade Offset=@n9) {materialFunction=&CameraDepthFade}.Result
    Speed
    Far Normal Strength
    Near Normal Strength
    Far Displacement
   ) {materialFunction=/Water/Materials/Functions/WaterTextureSurface}.Wave Attributes
  ) {materialFunction=/Water/Materials/Functions/SetWaveAttributes}
  n11 = StaticSwitchParameter(
   True=MaterialFunctionCall!(MaterialAttributes=@n10.MA, Time=MaterialFunctionCall() {materialFunction=&GetWaterTime}.WaterTime) {materialFunction=/Water/Materials/Functions/WaterRiverFlowmaps}.Result
   False=@n10.MA
  ) {parameterName=Enable River}
  n12 = MaterialFunctionCall!(MA=@n11) {materialFunction=/Water/Materials/Layers/BeachFoam}
  n13 = MaterialFunctionCall!(
   Attributes=StaticSwitchParameter(True=@n12.Attributes, False=@n11) {parameterName=Enable Ocean Foam}
   Time=MaterialFunctionCall() {materialFunction=&GetWaterTime}.WaterTime
   Enable Foam
  ) {materialFunction=/Water/Materials/Functions/SampleFluidSimulation}
  n14 = MaterialFunctionCall!(Attributes=@n13.Attributes, Additional Water Mask=@n13.Attributes) {materialFunction=/Water/Materials/Functions/WaterOpacityMaskFromDepth}
  n15 = VectorParameter() {parameterName=Foam Scattering, defaultValue=<rgba 1 1 1 5>}
  n16 = StaticSwitchParameter(True=Multiply(A=@n12.Foam Scattering, B=Multiply(A=@n15.RGB, B=@n15.RGB)), False=Constant()) {parameterName=Enable Ocean Foam}
  n17 = VectorParameter() {parameterName=Scattering, defaultValue=<rgba 1 1 1 0.5>}
  n18 = Multiply(A=@n17.RGB, B=@n17.RGB)
  n19 = VectorParameter() {parameterName=Absorption, defaultValue=<rgba 10 150 350 8>}
  n20 = MaterialFunctionCall(Scattering=Divide(A=Add(A=@n16, B=@n18)) {constB=1000}, Absorption=Divide(A=Divide(B=@n19.RGB), B=@n19.A)) {materialFunction=&Water_Underside}
  MaterialFunctionCall!(Mask=@n14.Mask, Scattering=@n20.Scattering, Absorption=@n20.Scattering) {materialFunction=/Water/Materials/Functions/WaterCoefficientMask}
  MaterialFunctionCall!(X=Divide(A=@n13.Water Depth) {constB=1000}) {materialFunction=/NiagaraFluids/Materials/Functions/MF_DebugGradient}
  ScalarParameter() {defaultValue=0.1, group=Water Shading, parameterName=Anisotropy}
  SingleLayerWaterMaterialOutput!(ScatteringCoefficients=Constant(), AbsorptionCoefficients=Constant() {r=1}, PhaseG, ColorScaleBehindWater)
  n21 = MaterialFunctionCall!(Attributes=@n14.Result) {materialFunction=/Water/Materials/Functions/ClipDynamicBounds}
  StaticSwitchParameter(
   True=MaterialFunctionCall!(Attributes=@n21.Attributes) {materialFunction=/Water/Materials/Functions/WaterHeightMappingVS}.Attributes
   False=@n21.Attributes
  ) {defaultValue=true, parameterName=Enable Water VS Mapping}
  n22 = MaterialFunctionCall() {materialFunction=&WaterAttributes}
  StaticSwitchParameter(
   True=MaterialFunctionCall!(
    Refraction=GetMaterialAttributes(None=@n22.Result) {attributeGetTypes [#3]}.Refraction
    Caustics Brightness
   ) {materialFunction=/Water/Caustics/Materials/Functions/WaterCaustics_GerstnerWaves}.Caustics
   False=Constant() {r=1}
  ) {parameterName=Caustics}
 ]
}
