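# Simple masked Single Layer Water material with screen-space reflections and Niagara mesh particle support, driven by textures supplied as parameters.
# NormalTex packs XY normal data in RG and a scalar in alpha, which a custom smoothstep over WaterFade turns into faded opacity, thresholded for masking.
# That alpha also forms a local-space Z vector and is subtracted from a second texture's red channel.
# Vector parameters set water scattering and absorption.
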Material /WaterAdvanced/Materials/Water_Material_Simple
graph {
 properties {
  allowTranslucentCustomDepthWrites=true
  bScreenSpaceReflections=true
  bUsedWithNiagaraMeshParticles=true
  blendMode=BLEND_Masked
  pixelDepthOffsetMode=PDOM_Legacy
  shadingModel=MSM_SingleLayerWater
  translucencyLightingMode=TLM_Surface
  usedShadingModels=SingleLayerWater
 }
 nodes [
  n1 = TextureCoordinate()
  MaterialExpressionTextureSample_4 = TextureSample(UVs=@n1, Tex=TextureObjectParameter() {parameterName=NormalTex, texture=&Black})
  n2 = ComponentMask(None=@MaterialExpressionTextureSample_4.RGB) {g=true, r=true}
  AppendVector(A=@n2, B=Constant() {r=1})
  Multiply(A=@n2) {constB=3}
  MaterialExpressionScalarParameter_9 = ScalarParameter() {defaultValue=0.1, parameterName=WaterFade}
  n3 = Custom!(a=@MaterialExpressionTextureSample_4.A, f=@MaterialExpressionScalarParameter_9) {
   code="smoothstep(0, f, a)"
   outputType=CMOT_Float1
   inputs [
    {
     inputName=a
     input {expression=&self:MaterialExpressionTextureSample_4, mask=1, maskA=1, outputIndex=4}
    }
    {inputName=f, input {expression=&self:MaterialExpressionScalarParameter_9}}
   ]
  }
  MaterialExpressionMultiply_2 = Multiply(A=@n3, B=ScalarParameter() {defaultValue=1, parameterName=Opacity}) {constA=1}
  n4 = Custom!(a=@MaterialExpressionMultiply_2) {code="a > 1e-5", inputs [{inputName=a, input {expression=&self:MaterialExpressionMultiply_2}}]}
  Multiply(A=@n4)
  ScalarParameter() {defaultValue=0.2, parameterName=Metallic}
  ScalarParameter() {parameterName=Specular}
  ScalarParameter() {parameterName=Roughness}
  SingleLayerWaterMaterialOutput!(
   ScatteringCoefficients=VectorParameter() {parameterName=Scatterieng}.RGB
   AbsorptionCoefficients=VectorParameter() {parameterName=Absorption, defaultValue=<rgba 0.0156 0.0071 0.0011 0>}.RGB
   PhaseG
   ColorScaleBehindWater
  )
  Subtract(
   A=TextureSample(UVs=@n1, Tex=TextureObjectParameter() {parameterName=Tex, texture=&Black}).R
   B=@MaterialExpressionTextureSample_4.A
  )
  Transform(None=MaterialFunctionCall(Z=@MaterialExpressionTextureSample_4.A) {materialFunction=&MakeFloat3}.Result) {transformSourceType=TRANSFORMSOURCE_Local}
  VectorParameter() {parameterName=Water Color, defaultValue=<rgba 0.0837 0.2424 1 1>}
 ]
}
