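# Material function that blends lake or ocean material attributes into river material attributes according to water flow speed.
# The blend alpha is velocity length divided by a wave-mask threshold, saturated; a static switch picks water body data velocity or the supplied shallow water velocity.
# The same alpha scales in the shallow water scattering output.
# A VisualizeBlend switch tints lake/ocean emissive red and river blue for debugging.
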MaterialFunction /WaterAdvanced/Niagara/Materials/Function/MF_ShallowWaterBlendTransition
graph {nodes [
  n1 = FunctionInput() {inputName=LakeOrOceanMaterialAttributes, inputType=FunctionInput_MaterialAttributes}
  n2 = SetMaterialAttributes!(MaterialAttributes=@n1, EmissiveColor=Constant3Vector() {constant=<rgba 1 0 0 0>}) {attributeSetTypes [#1]}
  n3 = FunctionInput() {inputName=RiverMaterialAttributes, inputType=FunctionInput_MaterialAttributes}
  n4 = SetMaterialAttributes(MaterialAttributes=@n3, EmissiveColor=Constant3Vector() {constant=<rgba 0 0 1 0>}) {attributeSetTypes [#1]}
  n5 = MaterialFunctionCall(Vector 2=MaterialFunctionCall!(Position) {materialFunction=/Water/Materials/Functions/WaterBodyData}.Velocity) {materialFunction=/Engine/Functions/Engine_MaterialFunctions02/Utility/VectorLength}
  n6 = MaterialFunctionCall(Vector 2=FunctionInput() {inputName=Shallow Water Velocity, inputType=FunctionInput_Vector2}) {materialFunction=&VectorLength}
  n7 = Saturate(None=Divide(
    A=StaticSwitchParameter(True=@n5.V2 Length, False=@n6.V2 Length) {defaultValue=true, parameterName=Blend Based On Spline}
    B=ScalarParameter() {defaultValue=100, parameterName=Velocity to Mask waves}
   ))
  FunctionOutput(None=BlendMaterialAttributes(
    A=StaticSwitchParameter(True=@n2, False=@n1) {parameterName=VisualizeBlend}
    B=StaticSwitchParameter(True=@n4, False=@n3) {parameterName=VisualizeBlend}
    Alpha=@n7
   ))
  FunctionOutput(None=LinearInterpolate(B=FunctionInput() {inputName=Shallow Water Scattering}, Alpha=@n7)) {outputName=Shallow Water Scattering}
 ]}
