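# M_DebugOcean is a Single Layer Water debug ocean material, enabled for Niagara sprite and mesh particles.
# World XY divided by the WorldGridSize parameter gives the UVs for sampling NormalTex, whose alpha fills the Z of a float3.
# Scattering and absorption parameters feed the water output, alongside a color and Opacity, Metallic, Specular and Roughness parameters.
# The graph shows no links from the float3 or those surface parameters to main material inputs.
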Material /WaterAdvanced/Niagara/Materials/M_DebugOcean
graph {
 expressions {
  MaterialExpressionComponentMask_0 {class=MaterialExpressionComponentMask, inputs [None {from=MaterialExpressionVectorParameter_1, output=RGB}], properties {g=true, r=true}}
  MaterialExpressionConvert_0 {
   class=MaterialExpressionConvert
   inputs [
    None {from=""}
    None {from=""}
    None {from=MaterialExpressionTextureSample_0, output=A}
   ]
   properties {
    nodeName=Make Float3
    convertInputs [
     {type=Scalar, expressionInput {expression=null, inputName=None, mask=0, maskA=0, maskB=0, maskG=0, maskR=0, outputIndex=0}}
     {type=Scalar, expressionInput {expression=null, inputName=None, mask=0, maskA=0, maskB=0, maskG=0, maskR=0, outputIndex=0}}
     {
      type=Scalar
      expressionInput {
       expression=&self:MaterialExpressionTextureSample_0
       inputName=None
       mask=1
       maskA=1
       maskB=0
       maskG=0
       maskR=0
       outputIndex=4
      }
     }
    ]
    convertOutputs [{type=Vector3}]
   }
  }
  MaterialExpressionDivide_0 {
   class=MaterialExpressionDivide
   inputs [
    A {from=MaterialExpressionWorldPosition_0, output=XY}
    B {from=MaterialExpressionComponentMask_0}
   ]
  }
  MaterialExpressionReroute_0 {class=MaterialExpressionReroute, inputs [None {from=MaterialExpressionDivide_0}]}
  MaterialExpressionScalarParameter_0 {class=MaterialExpressionScalarParameter, properties {defaultValue=0.5, parameterName=Opacity}}
  MaterialExpressionScalarParameter_1 {class=MaterialExpressionScalarParameter, properties {defaultValue=0.4, parameterName=Metallic}}
  MaterialExpressionScalarParameter_2 {class=MaterialExpressionScalarParameter, properties {parameterName=Specular}}
  MaterialExpressionScalarParameter_3 {class=MaterialExpressionScalarParameter, properties {parameterName=Roughness}}
  MaterialExpressionSingleLayerWaterMaterialOutput_1 {
   class=MaterialExpressionSingleLayerWaterMaterialOutput
   inputs [
    ScatteringCoefficients {from=MaterialExpressionVectorParameter_3, output=RGB}
    AbsorptionCoefficients {from=MaterialExpressionVectorParameter_4, output=RGB}
    PhaseG {from="", output=""}
    ColorScaleBehindWater {from="", output=""}
   ]
   properties {}
  }
  MaterialExpressionTextureObjectParameter_1 {
   class=MaterialExpressionTextureObjectParameter
   inputs [
    UVs {from="", output=""}
    Apply View MipBias {from="", output=""}
   ]
   properties {parameterName=NormalTex}
  }
  MaterialExpressionTextureSample_0 {
   class=MaterialExpressionTextureSample
   inputs [
    UVs {from=MaterialExpressionReroute_0}
    Tex {from=MaterialExpressionTextureObjectParameter_1}
    Apply View MipBias {from=""}
   ]
  }
  MaterialExpressionVectorParameter_0 {class=MaterialExpressionVectorParameter, properties {defaultValue=<rgba 0.1083 0.3439 0.6146 1>}}
  MaterialExpressionVectorParameter_1 {class=MaterialExpressionVectorParameter, properties {parameterName=WorldGridSize}}
  MaterialExpressionVectorParameter_3 {class=MaterialExpressionVectorParameter, properties {parameterName=Scatterieng}}
  MaterialExpressionVectorParameter_4 {class=MaterialExpressionVectorParameter, properties {parameterName=Absorption, defaultValue=<rgba 0.0156 0.0071 0.0011 0>}}
  MaterialExpressionWorldPosition_0 {class=MaterialExpressionWorldPosition, inputs [], properties {}}
 }
 properties {
  bUsedWithNiagaraMeshParticles=true
  bUsedWithNiagaraSprites=true
  pixelDepthOffsetMode=PDOM_Legacy
  shadingModel=MSM_SingleLayerWater
  usedShadingModels=SingleLayerWater
 }
}
