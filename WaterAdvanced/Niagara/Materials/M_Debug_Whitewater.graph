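# Debug material for whitewater in the WaterAdvanced content, set up for use on Niagara mesh particles.
# It samples the Tex texture parameter with clamped addressing and shows its RGB as emissive color.
# The texture's alpha plus an Offset parameter drives vertical world position offset, so the mesh is raised by the sampled value.
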Material /WaterAdvanced/Niagara/Materials/M_Debug_Whitewater
graph {
 properties {bUseMaterialAttributes=true, bUsedWithNiagaraMeshParticles=true, pixelDepthOffsetMode=PDOM_Legacy, usedShadingModels=Default Lit}
 nodes [
  n1 = TextureSample(UVs=TextureCoordinate(), Tex=TextureObjectParameter() {parameterName=Tex}) {samplerSource=SSM_Clamp_WorldGroupSettings}
  n2 = Constant()
  n3 = ScalarParameter() {defaultValue=5, parameterName=Offset}
  MakeMaterialAttributes!(
   BaseColor
   Metallic
   Specular
   Roughness
   Anisotropy
   EmissiveColor=@n1.RGB
   Opacity
   OpacityMask
   Normal
   Tangent
   WorldPositionOffset=MaterialFunctionCall(X=@n2, Y=@n2, Z=Add(A=@n1.A, B=@n3)) {materialFunction=&MakeFloat3}.Result
   SubsurfaceColor
   ClearCoat
   ClearCoatRoughness
   AmbientOcclusion
   Refraction
   CustomizedUVs_0
   CustomizedUVs_1
   CustomizedUVs_2
   CustomizedUVs_3
   CustomizedUVs_4
   CustomizedUVs_5
   CustomizedUVs_6
   CustomizedUVs_7
   PixelDepthOffset
   ShadingModel
   Displacement
  )
 ]
}
