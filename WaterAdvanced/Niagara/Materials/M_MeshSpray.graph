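# Translucent subsurface material for Niagara mesh particles, rendering water spray with volumetric directional lighting.
# A tiled foam texture's red channel, particle alpha and an Opacity parameter are multiplied, saturated, then scaled by inverse Fresnel to soften grazing edges.
# A matching foam normal map shares the DetailTiling UVs, alongside parameters for pale base color and blue subsurface color.
# It outputs translucent velocity and uses pixel normal offset refraction.
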Material /WaterAdvanced/Niagara/Materials/M_MeshSpray
graph {
 properties {
  bOutputTranslucentVelocity=true
  bUsedWithNiagaraMeshParticles=true
  blendMode=BLEND_Translucent
  refractionMethod=RM_PixelNormalOffset
  shadingModel=MSM_Subsurface
  translucencyLightingMode=TLM_VolumetricDirectional
  usedShadingModels=Subsurface
 }
 nodes [
  n1 = Multiply(A=TextureCoordinate(), B=ScalarParameter() {defaultValue=4, parameterName=DetailTiling})
  n2 = Multiply(A=ParticleColor().A, B=TextureSample(UVs=@n1) {texture=/Water/Textures/Foam/T_WaterFlow_01_Foam_Tiled}.R)
  n3 = ScalarParameter() {defaultValue=1, parameterName=Opacity}
  Multiply(A=Saturate(None=Multiply(A=@n2, B=@n3)), B=OneMinus(None=Fresnel() {exponent=0.4}))
  TextureSample(UVs=@n1) {samplerType=SAMPLERTYPE_Normal, texture=/Water/Textures/Foam/T_WaterFlow_01_Foam_Tiled_N}
  VectorParameter() {parameterName=BaseColor, defaultValue=<rgba 0.8086 0.8225 0.8438 1>}
  VectorParameter() {parameterName=SubsurfaceColor, defaultValue=<rgba 0.4 0.5 1 1>}
 ]
}
