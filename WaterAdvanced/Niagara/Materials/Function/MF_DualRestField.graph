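# A material function in WaterAdvanced's Niagara materials that samples a texture using a dual rest field.
# It runs WorldAlignedTexture twice, once per input world position, with shared texture, size, world normal, projection contrast and float4 export settings.
# It linearly blends the two triplanar XYZ results by DualRestWeight, so the caller can crossfade between two rest-position sets.
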MaterialFunction /WaterAdvanced/Niagara/Materials/Function/MF_DualRestField
graph {nodes [
  n1 = FunctionInput() {inputName=Tex, inputType=FunctionInput_Texture2D, sortPriority=-666}
  n2 = FunctionInput() {bUsePreviewValueAsDefault=true, inputName=TextureSize, previewValue=<100 100 100 1>}
  n3 = FunctionInput() {inputName=WorldPosition0, previewValue=<0 0 1 1>}
  n4 = FunctionInput(Preview=StaticBool()) {bUsePreviewValueAsDefault=true, description="Default = False", inputName=Export Float 4, inputType=FunctionInput_StaticBool}
  n5 = FunctionInput() {inputName=NormalWS, previewValue=<0 0 1 1>}
  n6 = FunctionInput() {bUsePreviewValueAsDefault=true, inputName=ProjectionTransitionContrast, inputType=FunctionInput_Scalar, previewValue=<1 0 0 1>}
  n7 = FunctionInput() {inputName=WorldPosition1, previewValue=<0 0 1 1>}
  FunctionOutput(None=LinearInterpolate(
    A=MaterialFunctionCall(TextureObject=@n1, TextureSize=@n2, WorldPosition=@n3, Export Float 4=@n4, World Space Normal=@n5, ProjectionTransitionContrast=@n6) {materialFunction=&WorldAlignedTexture}.XYZ Texture
    B=MaterialFunctionCall(TextureObject=@n1, TextureSize=@n2, WorldPosition=@n7, Export Float 4=@n4, World Space Normal=@n5, ProjectionTransitionContrast=@n6) {materialFunction=&WorldAlignedTexture}.XYZ Texture
    Alpha=FunctionInput() {inputName=DualRestWeight, inputType=FunctionInput_Scalar}
   ))
 ]}
