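# Translucent, lit Niagara sprite material that renders water surface foam particles.
# A world-position-sampled texture builds an XY curl offset, paired with a Z offset toward the dynamic-parameter WaterHeight plus a small lift.
# Particle-relative positions, shifted by two dynamic parameters, feed MF_DualRestField to blend foam color and normal textures.
# Opacity math combines particle alpha, a soft circular mask and S-curved foam texture.
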Material /WaterAdvanced/Niagara/Materials/M_SW_Foam
graph {
 properties {
  bUsedWithNiagaraSprites=true
  blendMode=BLEND_Translucent
  pixelDepthOffsetMode=PDOM_Legacy
  translucencyLightingMode=TLM_SurfacePerPixelLighting
  usedShadingModels=Default Lit
 }
 nodes [
  n1 = TextureSample(UVs=Divide(
    A=WorldPosition() {worldPositionShaderOffset=WPT_ExcludeAllShaderOffsets}.XY
    B=ScalarParameter() {defaultValue=100, parameterName=CurlWPOSize}
   )) {samplerType=SAMPLERTYPE_LinearColor}
  n2 = Constant3Vector() {constant=<rgba 1 1 0 0>}
  n3 = Constant()
  n4 = Subtract(A=DynamicParameter() {paramNames [WaterHeight,Param2,Param3,Param4]}.WaterHeight, B=WorldPosition().Z)
  n5 = ScalarParameter() {defaultValue=5, parameterName=AboveWaterInCM}
  Add(
   A=Multiply(A=Multiply(A=@n1.RGB, B=@n2), B=ScalarParameter() {defaultValue=5, parameterName=CurlWPOIntensity})
   B=MaterialFunctionCall(X=@n3, Y=@n3, Z=Add(A=@n4, B=@n5)) {materialFunction=&MakeFloat3}.Result
  )
  n6 = DynamicParameter() {parameterIndex=2}
  Frac(None=MaterialFunctionCall(X=@n6.R, Y=@n6.R, Z=@n6.R) {materialFunction=&MakeFloat3}.Result)
  n7 = ParticleColor()
  n8 = Constant() {r=0.5}
  n9 = SphereMask(A=@n8, B=TextureCoordinate(), Radius=@n8) {hardnessPercent=0}
  n10 = ScalarParameter() {defaultValue=1, parameterName=Opacity, sliderMax=1}
  n11 = Multiply(A=@n7.A, B=Saturate(None=Multiply(A=@n9, B=@n10)))
  MaterialFunctionCall(Alpha Threshold=@n11) {materialFunction=&DitherTemporalAA}
  MaterialFunctionCall(Flatness=ScalarParameter() {defaultValue=0.5032}) {materialFunction=&FlattenNormal}
  n12 = Constant3Vector() {constant=<rgba 0 0 1 0>}
  n13 = ScalarParameter() {defaultValue=200, parameterName=sTextureSize}
  n14 = Subtract(A=WorldPosition().XYZ, B=ParticlePositionWS())
  n15 = DynamicParameter() {parameterIndex=2}
  n16 = Add(
   A=@n14
   B=MaterialFunctionCall(X=@n15.R, Y=@n15.R, Z=@n15.R) {materialFunction=&MakeFloat3}.Result
  )
  n17 = DynamicParameter() {parameterIndex=3}
  n18 = Add(
   A=@n14
   B=MaterialFunctionCall(X=@n17.R, Y=@n17.R, Z=@n17.R) {materialFunction=&MakeFloat3}.Result
  )
  n19 = ScalarParameter() {parameterName=DualRestWeight}
  MaterialFunctionCall(
   Normal=MaterialFunctionCall!(
    Tex=TextureObject() {samplerType=SAMPLERTYPE_Normal, texture=/Water/Textures/Foam/T_WaterFlow_01_Foam_Tiled_N}
    NormalWS=@n12
    TextureSize=@n13
    WorldPosition0=@n16
    WorldPosition1=@n18
    ProjectionTransitionContrast
    Export Float 4
    DualRestWeight=@n19
   ) {materialFunction=/WaterAdvanced/Niagara/Materials/Function/MF_DualRestField}.Result
   Flatness=ScalarParameter() {defaultValue=0.3, parameterName="Flatness (S)"}
  ) {materialFunction=&FlattenNormal}
  n20 = MaterialFunctionCall(
   Tex=TextureObject() {texture=/Water/Textures/Foam/T_WaterFlow_01_Foam_Tiled}
   NormalWS=@n12
   TextureSize=@n13
   WorldPosition0=@n16
   WorldPosition1=@n18
   DualRestWeight=@n19
  ) {materialFunction=&MF_DualRestField}
  Multiply(A=@n7.RGB, B=@n20.Result)
  Multiply(A=VectorParameter() {parameterName=Base Color, defaultValue=<rgba 0.1914 0.3685 0.6562 1>}.RGB, B=@n7.RGB)
  Multiply(
   A=@n11
   B=MaterialFunctionCall(In=@n20.Result, Power=ScalarParameter() {defaultValue=2, parameterName=SCurve}) {materialFunction=&SCurve}.Result
  )
 ]
}
