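# Material function for water materials that adds interactive fluid simulation ripples, mapped into world space, to incoming material attributes.
# World position converts to simulation UVs through Fluidsim_WorldPosition_to_UV to sample the NormalAndHeight texture parameter, which defaults to black.
# Scaled red and green channels offset the normal's XY.
# The scaled, symmetrically clamped blue channel adds vertical World Position Offset, while emissive, opacity and roughness pass through unchanged.
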MaterialFunction /WaterAdvanced/Niagara/Materials/Function/MF_SampleInteractiveSim
graph {
 properties {
  bExposeToLibrary=true
  description=Maps the Fluid Simulation Normal and Height into World Space for use in water materials in the world.
  libraryCategoriesText [Water]
 }
 nodes [
  n1 = FunctionInput() {inputName=Attributes, inputType=FunctionInput_MaterialAttributes}
  n2 = WorldPosition() {worldPositionShaderOffset=WPT_ExcludeAllShaderOffsets}
  Add(A=GetMaterialAttributes(None=@n1) {attributeGetTypes [#1]}.WorldPositionOffset, B=@n2.XYZ)
  n3 = GetMaterialAttributes(None=@n1) {attributeGetTypes [
    #2
    #1
    #3
    #4
    #5
   ]}
  n4 = TextureSampleParameter2D(UVs=MaterialFunctionCall!(WorldPosition=@n2.XYZ) {materialFunction=/Water/FluidSimulation/Materials/Functions/Fluidsim_WorldPosition_to_UV}.UV) {
   group=Ripple Simulation
   parameterName=NormalAndHeight
   samplerSource=SSM_Clamp_WorldGroupSettings
   samplerType=SAMPLERTYPE_LinearColor
   texture=/Engine/Functions/Engine_MaterialFunctions02/PivotPainter2/Black_1x1_EXR_Texture
  }
  n5 = Multiply(
   A=ComponentMask(None=@n4.RGB) {g=true, r=true}
   B=ScalarParameter() {defaultValue=1, group=Ripple Simulation, parameterName="Ripple  Normal Strength"}
  )
  n6 = Constant()
  n7 = ScalarParameter() {defaultValue=1024, group=Ripple Simulation, parameterName=Max Ripple Displacement}
  n8 = Clamp(
   None=Multiply(A=@n4.B, B=ScalarParameter() {defaultValue=1, group=Ripple Simulation, parameterName=Ripple Displacement Multiplier})
   Min=Multiply(A=@n7) {constB=-1}
   Max=@n7
  )
  n9 = Constant3Vector() {constant=<rgba 0 0 1 0>}
  n10 = SetMaterialAttributes!(
   MaterialAttributes=@n3.MaterialAttributes
   Normal=Add(A=@n3.Normal, B=AppendVector(A=@n5, B=@n6))
   WorldPositionOffset=Add(A=@n3.WorldPositionOffset, B=Multiply(A=@n8, B=@n9))
   EmissiveColor=@n3.EmissiveColor
   Opacity=@n3.Opacity
   Roughness=@n3.Roughness
  ) {attributeSetTypes [
    #2
    #1
    #3
    #4
    #5
   ]}
  FunctionOutput(None=@n10) {outputName=Attributes}
 ]
}
