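# MF_BlendNormals is a material function in the WaterAdvanced Niagara materials that layers a Detail Normal onto a Base Normal.
# A single Custom HLSL node builds an orthonormal basis around the base normal and rotates the detail normal into it, following the Self Shadow "Blending in Detail" article.
# A branch swaps in fixed axes when the base normal points almost straight down, avoiding the division singularity.
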MaterialFunction /WaterAdvanced/Niagara/Materials/Function/MF_BlendNormals
graph {nodes [
  MaterialExpressionFunctionInput_0 = FunctionInput() {inputName=Base Normal}
  MaterialExpressionFunctionInput_1 = FunctionInput() {inputName=Detail Normal, sortPriority=10}
  n1 = Custom!(n1=@MaterialExpressionFunctionInput_0, n2=@MaterialExpressionFunctionInput_1) {
   code="// https://blog.selfshadow.com/publications/blending-in-detail/\r\n\r\nfloat a = 1/(1 + n1.z);\r\nfloat b = -n1.x*n1.y*a;\r\n\r\n// Form a basis\r\nfloat3 b1 = float3(1 - n1.x*n1.x*a, b, -n1.x);\r\nfloat3 b2 = float3(b, 1 - n1.y*n1.y*a, -n1.y);\r\nfloat3 b3 = n1;\r\n\r\nif (n1.z < -0.9999999) // Handle the singularity\r\n{\r\n    b1 = float3( 0, -1, 0);\r\n    b2 = float3(-1,  0, 0);\r\n}\r\n\r\n// Rotate n2 via the basis\r\nreturn n2.x*b1 + n2.y*b2 + n2.z*b3;"
   showCode=true
   inputs [
    {inputName=n1, input {expression=&self:MaterialExpressionFunctionInput_0}}
    {inputName=n2, input {expression=&self:MaterialExpressionFunctionInput_1}}
   ]
  }
  FunctionOutput(None=@n1) {outputName=Custom}
 ]}
