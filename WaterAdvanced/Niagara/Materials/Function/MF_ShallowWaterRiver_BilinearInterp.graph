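# Material function for shallow water river materials that performs manual bilinear interpolation of a 2D texture at given UVs.
# A Use Sample static switch selects between two paths.
# When true, custom HLSL computes four texel-center UVs and weights for four texture samples and nested lerps.
# When false, one custom HLSL node loads and blends four texels, passing Mip Level as a fourth Load coordinate while sizing from mip 0.
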MaterialFunction /WaterAdvanced/Niagara/Materials/Function/MF_ShallowWaterRiver_BilinearInterp
graph {nodes [
  MaterialExpressionFunctionInput_0 = FunctionInput() {inputName=Uvs, inputType=FunctionInput_Vector2}
  MaterialExpressionFunctionInput_1 = FunctionInput() {inputName=Tex, inputType=FunctionInput_Texture2D, sortPriority=10}
  MaterialExpressionTextureProperty_0 = TextureProperty(Tex=@MaterialExpressionFunctionInput_1)
  MaterialExpressionFunctionInput_2 = FunctionInput() {bUsePreviewValueAsDefault=true, inputName=Mip Level, inputType=FunctionInput_Scalar}
  n1 = Custom!(UVs=@MaterialExpressionFunctionInput_0, Tex=@MaterialExpressionTextureProperty_0, TheMipLevel=@MaterialExpressionFunctionInput_2) {
   code="uint2 NumCells = Tex;\r\n\r\nconst float2 GridPos = UVs.xy * NumCells - .5;\r\n\t\t\t\t\t\t\t\r\n// identify the lower-left-hand corner of the cell\r\nconst int2 GridCell = floor(GridPos);\r\n\t\t\r\n\t\t\t\t\r\n\r\na = GridCell;\r\na = (a + .5) / NumCells;\r\n\r\nb = GridCell + uint2(1,0);\r\nb = (b + .5) / NumCells;\r\n\r\nc = GridCell + uint2(0,1);\r\nc = (c + .5) / NumCells;\r\n\r\nd = GridCell + uint2(1,1);\r\nd = (d + .5) / NumCells;\r\n\r\nreturn GridPos - GridCell;"
   outputType=CMOT_Float2
   showCode=true
   additionalOutputs [
    {outputName=a, outputType=CMOT_Float2}
    {outputName=b, outputType=CMOT_Float2}
    {outputName=c, outputType=CMOT_Float2}
    {outputName=d, outputType=CMOT_Float2}
   ]
   inputs [
    {
     inputName=UVs
     input {expression=&self:MaterialExpressionFunctionInput_0}
    }
    {
     inputName=Tex
     input {expression=&self:MaterialExpressionTextureProperty_0}
    }
    {
     inputName=TheMipLevel
     input {expression=&self:MaterialExpressionFunctionInput_2}
    }
   ]
  }
  n2 = MaterialFunctionCall(Float2=@n1.return) {materialFunction=&BreakOutFloat2Components}
  n3 = LinearInterpolate(
   A=TextureSample(UVs=@n1.a, Tex=@MaterialExpressionFunctionInput_1).RGBA
   B=TextureSample(UVs=@n1.b, Tex=@MaterialExpressionFunctionInput_1).RGBA
   Alpha=@n2.R
  )
  n4 = LinearInterpolate(
   A=TextureSample(UVs=@n1.c, Tex=@MaterialExpressionFunctionInput_1).RGBA
   B=TextureSample(UVs=@n1.d, Tex=@MaterialExpressionFunctionInput_1).RGBA
   Alpha=@n2.R
  )
  n5 = Custom(UVs=@MaterialExpressionFunctionInput_0, Tex=@MaterialExpressionFunctionInput_1, TheMipLevel=@MaterialExpressionFunctionInput_2) {
   code="uint2 NumCells;\r\nuint Levels;\r\nTex.GetDimensions(0, NumCells.x, NumCells.y, Levels);\r\nconst float2 GridPos = UVs.xy * NumCells - .5;\r\n\t\t\t\t\t\t\t\r\n// identify the lower-left-hand corner of the cell\r\n\tconst int2 GridCell = floor(GridPos);\r\nconst float2 InterpV = GridPos - GridCell;\t\t\t\t\r\n\t\t\t\t\r\nfloat4 Value00 = Tex.Load(float4(GridCell, 0, TheMipLevel));\r\nfloat4 Value10 = Tex.Load(float4(GridCell + uint2(1,0), 0, TheMipLevel));\r\nfloat4 Value01 = Tex.Load(float4(GridCell + uint2(0,1), 0, TheMipLevel));\r\nfloat4 Value11 = Tex.Load(float4(GridCell + uint2(1,1), 0, TheMipLevel));\r\n\r\nfloat4 LerpX0 = lerp(Value00, Value10, InterpV.x);\r\nfloat4 LerpX1 = lerp(Value01, Value11, InterpV.x);\r\n\r\nreturn lerp(LerpX0, LerpX1, InterpV.y);\r\n"
   outputType=CMOT_Float4
   showCode=true
   inputs [
    {
     inputName=UVs
     input {expression=&self:MaterialExpressionFunctionInput_0}
    }
    {
     inputName=Tex
     input {expression=&self:MaterialExpressionFunctionInput_1}
    }
    {
     inputName=TheMipLevel
     input {expression=&self:MaterialExpressionFunctionInput_2}
    }
   ]
  }
  n6 = StaticBool()
  FunctionOutput(None=StaticSwitch(
    True=LinearInterpolate(A=@n3, B=@n4, Alpha=@n2.G)
    False=@n5
    Value=FunctionInput(Preview=@n6) {inputName=Use Sample, inputType=FunctionInput_StaticBool}
   )) {outputName=Value}
 ]}
