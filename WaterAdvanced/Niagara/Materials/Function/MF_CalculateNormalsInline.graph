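# Material function that computes a surface normal inline from a heightmap texture for Niagara-based shallow water in WaterAdvanced.
# Samples height one texel away in +X and +Y through MF_ShallowWaterRiver_BilinearInterp, with a Use Sample static switch parameter.
# Builds two edge vectors against the supplied center height, using texel size times WorldSpaceSize for spacing and HeightScale for height.
# Outputs their normalized cross product, a forward-difference normal needing only two texture fetches.
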MaterialFunction /WaterAdvanced/Niagara/Materials/Function/MF_CalculateNormalsInline
graph {nodes [
  n1 = FunctionInput() {inputName=Tex, inputType=FunctionInput_Texture2D}
  n2 = TextureProperty(Tex=@n1) {property=TMTM_TexelSize}
  n3 = MaterialFunctionCall(Float2=Multiply(A=@n2, B=FunctionInput() {inputName=WorldSpaceSize, inputType=FunctionInput_Vector2, sortPriority=10})) {materialFunction=&BreakOutFloat2Components}
  n4 = StaticBoolParameter() {defaultValue=true, parameterName=Use Sample}
  n5 = FunctionInput() {inputName=UV, inputType=FunctionInput_Vector2}
  n6 = MaterialFunctionCall(Float2=@n2) {materialFunction=&BreakOutFloat2Components}
  n7 = MaterialFunctionCall(X=@n6.R) {materialFunction=&MakeFloat2}
  n8 = ComponentMask(None=MaterialFunctionCall!(Use Sample=@n4, Mip Level, Uvs=Add(A=@n5, B=@n7.Result), Tex=@n1) {materialFunction=&/MF_ShallowWaterRiver_BilinearInterp}.Value) {r=true}
  n9 = FunctionInput() {bUsePreviewValueAsDefault=true, inputName=HeightScale, inputType=FunctionInput_Scalar, previewValue=<1 1 1 1>}
  n10 = MaterialFunctionCall(Z=Multiply(A=FunctionInput() {inputName=CenterHeight, inputType=FunctionInput_Scalar}, B=@n9)) {materialFunction=&MakeFloat3}
  n11 = Subtract(
   A=MaterialFunctionCall(X=@n3.R, Z=Multiply(A=@n8, B=@n9)) {materialFunction=&MakeFloat3}.Result
   B=@n10.Result
  )
  n12 = MaterialFunctionCall(Y=@n6.G) {materialFunction=&MakeFloat2}
  n13 = ComponentMask(None=MaterialFunctionCall(Use Sample=@n4, Uvs=Add(A=@n5, B=@n12.Result), Tex=@n1) {materialFunction=&MF_ShallowWaterRiver_BilinearInterp}.Value) {r=true}
  n14 = Subtract(
   A=MaterialFunctionCall(Y=@n3.G, Z=Multiply(A=@n13, B=@n9)) {materialFunction=&MakeFloat3}.Result
   B=@n10.Result
  )
  FunctionOutput(None=Normalize(VectorInput=CrossProduct(A=@n11, B=@n14)))
 ]}
