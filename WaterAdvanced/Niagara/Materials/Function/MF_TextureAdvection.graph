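# Material function in WaterAdvanced that advects an input texture along a simulation velocity, using world position and time.
# Static switch parameters Bracketing and ThreePhase select one of four WaterAdvanced.ush routines: two or three phase, plain or bracketed by angle and magnitude deltas.
# RawOutput returns that sample directly.
# With Normal Output on, Output optionally decodes a compressed normal, then rotates it to the flow when velocity alignment is enabled.
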MaterialFunction /WaterAdvanced/Niagara/Materials/Function/MF_TextureAdvection
graph {expressions {
  MaterialExpressionComponentMask_1 {class=MaterialExpressionComponentMask, inputs [None {from=MaterialExpressionStaticSwitchParameter_0}], properties {b=true, g=true, r=true}}
  MaterialExpressionComponentMask_11 {class=MaterialExpressionComponentMask, inputs [None {from=MaterialExpressionStaticSwitchParameter_0}], properties {g=true, r=true}}
  MaterialExpressionConstant_0 {class=MaterialExpressionConstant, properties {r=1}}
  MaterialExpressionConvert_0 {
   class=MaterialExpressionConvert
   inputs [
    None {from=""}
    None {from=""}
   ]
   properties {
    nodeName=Make Vector2
    convertInputs [
     {
      type=Scalar
      defaultValue=<rgba 1 0 0 1>
      expressionInput {expression=null, inputName=None, mask=0, maskA=0, maskB=0, maskG=0, maskR=0, outputIndex=0}
     }
     {type=Scalar, expressionInput {expression=null, inputName=None, mask=0, maskA=0, maskB=0, maskG=0, maskR=0, outputIndex=0}}
    ]
    convertOutputs [{type=Vector2}]
   }
  }
  MaterialExpressionCustom_0 {
   class=MaterialExpressionCustom
   inputs [
    Tex {from=MaterialExpressionNamedRerouteUsage_22, output=""}
    WorldPosition {from=MaterialExpressionNamedRerouteUsage_26, output=""}
    Time {from=MaterialExpressionTime_4, output=""}
    Speed {from=MaterialExpressionNamedRerouteUsage_25, output=""}
    Velocity {from=MaterialExpressionNamedRerouteUsage_23, output=""}
    AdvectedTextureScale {from=MaterialExpressionNamedRerouteUsage_24, output=""}
    PulseLength {from=MaterialExpressionNamedRerouteUsage_31, output=""}
    TheMipBias {from=MaterialExpressionNamedRerouteUsage_27, output=""}
    BracketAngleDelta {from=MaterialExpressionNamedRerouteUsage_28, output=""}
    BracketMagnitudeDelta {from=MaterialExpressionNamedRerouteUsage_29, output=""}
    SimNormal {from=MaterialExpressionNamedRerouteUsage_30, output=""}
    TextureAxis {from=MaterialExpressionNamedRerouteUsage_15, output=""}
    AlignAdvectedTextureToVelocity {from=MaterialExpressionNamedRerouteUsage_12, output=""}
   ]
   properties {
    code="return AdvectTexture_3Phase_Bracketed(WorldPosition, Velocity.xy, Speed, SimNormal, BracketAngleDelta, BracketMagnitudeDelta, PulseLength, Time, AdvectedTextureScale, TheMipBias, Tex, TexSampler, ((int) AlignAdvectedTextureToVelocity) == 1, TextureAxis);"
    outputType=CMOT_Float4
    showCode=true
    includeFilePaths [/Plugin/Experimental/WaterAdvanced/WaterAdvanced.ush]
    inputs [
     {
      inputName=Tex
      input {
       expression=&self:MaterialExpressionNamedRerouteUsage_22
       inputName=None
       mask=0
       maskA=0
       maskB=0
       maskG=0
       maskR=0
       outputIndex=0
      }
     }
     {
      inputName=WorldPosition
      input {
       expression=&self:MaterialExpressionNamedRerouteUsage_26
       inputName=None
       mask=0
       maskA=0
       maskB=0
       maskG=0
       maskR=0
       outputIndex=0
      }
     }
     {
      inputName=Time
      input {
       expression=&self:MaterialExpressionTime_4
       inputName=None
       mask=0
       maskA=0
       maskB=0
       maskG=0
       maskR=0
       outputIndex=0
      }
     }
     {
      inputName=Speed
      input {
       expression=&self:MaterialExpressionNamedRerouteUsage_25
       inputName=None
       mask=0
       maskA=0
       maskB=0
       maskG=0
       maskR=0
       outputIndex=0
      }
     }
     {
      inputName=Velocity
      input {
       expression=&self:MaterialExpressionNamedRerouteUsage_23
       inputName=None
       mask=0
       maskA=0
       maskB=0
       maskG=0
       maskR=0
       outputIndex=0
      }
     }
     {
      inputName=AdvectedTextureScale
      input {
       expression=&self:MaterialExpressionNamedRerouteUsage_24
       inputName=None
       mask=0
       maskA=0
       maskB=0
       maskG=0
       maskR=0
       outputIndex=0
      }
     }
     {
      inputName=PulseLength
      input {
       expression=&self:MaterialExpressionNamedRerouteUsage_31
       inputName=None
       mask=0
       maskA=0
       maskB=0
       maskG=0
       maskR=0
       outputIndex=0
      }
     }
     {
      inputName=TheMipBias
      input {
       expression=&self:MaterialExpressionNamedRerouteUsage_27
       inputName=None
       mask=0
       maskA=0
       maskB=0
       maskG=0
       maskR=0
       outputIndex=0
      }
     }
     {
      inputName=BracketAngleDelta
      input {
       expression=&self:MaterialExpressionNamedRerouteUsage_28
       inputName=None
       mask=0
       maskA=0
       maskB=0
       maskG=0
       maskR=0
       outputIndex=0
      }
     }
     {
      inputName=BracketMagnitudeDelta
      input {
       expression=&self:MaterialExpressionNamedRerouteUsage_29
       inputName=None
       mask=0
       maskA=0
       maskB=0
       maskG=0
       maskR=0
       outputIndex=0
      }
     }
     {
      inputName=SimNormal
      input {
       expression=&self:MaterialExpressionNamedRerouteUsage_30
       inputName=None
       mask=0
       maskA=0
       maskB=0
       maskG=0
       maskR=0
       outputIndex=0
      }
     }
     {
      inputName=TextureAxis
      input {
       expression=&self:MaterialExpressionNamedRerouteUsage_15
       inputName=None
       mask=0
       maskA=0
       maskB=0
       maskG=0
       maskR=0
       outputIndex=0
      }
     }
     {
      inputName=AlignAdvectedTextureToVelocity
      input {
       expression=&self:MaterialExpressionNamedRerouteUsage_12
       inputName=None
       mask=0
       maskA=0
       maskB=0
       maskG=0
       maskR=0
       outputIndex=0
      }
     }
    ]
   }
  }
  MaterialExpressionCustom_1 {
   class=MaterialExpressionCustom
   inputs [
    Tex {from=MaterialExpressionNamedRerouteUsage_0, output=""}
    WorldPosition {from=MaterialExpressionNamedRerouteUsage_5, output=""}
    Time {from=MaterialExpressionTime_1, output=""}
    Speed {from=MaterialExpressionNamedRerouteUsage_4, output=""}
    Velocity {from=MaterialExpressionNamedRerouteUsage_1, output=""}
    AdvectedTextureScale {from=MaterialExpressionNamedRerouteUsage_2, output=""}
    PulseLength {from=MaterialExpressionNamedRerouteUsage_3, output=""}
    TextureAxis {from=MaterialExpressionNamedRerouteUsage_44, output=""}
    AlignAdvectedTextureToVelocity {from=MaterialExpressionNamedRerouteUsage_16, output=""}
   ]
   properties {
    code="return AdvectTexture_2Phase(WorldPosition, Velocity.xy, Speed, Time, PulseLength, AdvectedTextureScale, Tex, TexSampler, ((int) AlignAdvectedTextureToVelocity) == 1, TextureAxis);"
    outputType=CMOT_Float4
    showCode=true
    includeFilePaths [&ush]
    inputs [
     {
      inputName=Tex
      input {
       expression=&self:MaterialExpressionNamedRerouteUsage_0
       inputName=None
       mask=0
       maskA=0
       maskB=0
       maskG=0
       maskR=0
       outputIndex=0
      }
     }
     {
      inputName=WorldPosition
      input {
       expression=&self:MaterialExpressionNamedRerouteUsage_5
       inputName=None
       mask=0
       maskA=0
       maskB=0
       maskG=0
       maskR=0
       outputIndex=0
      }
     }
     {
      inputName=Time
      input {
       expression=&self:MaterialExpressionTime_1
       inputName=None
       mask=0
       maskA=0
       maskB=0
       maskG=0
       maskR=0
       outputIndex=0
      }
     }
     {
      inputName=Speed
      input {
       expression=&self:MaterialExpressionNamedRerouteUsage_4
       inputName=None
       mask=0
       maskA=0
       maskB=0
       maskG=0
       maskR=0
       outputIndex=0
      }
     }
     {
      inputName=Velocity
      input {
       expression=&self:MaterialExpressionNamedRerouteUsage_1
       inputName=None
       mask=0
       maskA=0
       maskB=0
       maskG=0
       maskR=0
       outputIndex=0
      }
     }
     {
      inputName=AdvectedTextureScale
      input {
       expression=&self:MaterialExpressionNamedRerouteUsage_2
       inputName=None
       mask=0
       maskA=0
       maskB=0
       maskG=0
       maskR=0
       outputIndex=0
      }
     }
     {
      inputName=PulseLength
      input {
       expression=&self:MaterialExpressionNamedRerouteUsage_3
       inputName=None
       mask=0
       maskA=0
       maskB=0
       maskG=0
       maskR=0
       outputIndex=0
      }
     }
     {
      inputName=TextureAxis
      input {
       expression=&self:MaterialExpressionNamedRerouteUsage_44
       inputName=None
       mask=0
       maskA=0
       maskB=0
       maskG=0
       maskR=0
       outputIndex=0
      }
     }
     {
      inputName=AlignAdvectedTextureToVelocity
      input {
       expression=&self:MaterialExpressionNamedRerouteUsage_16
       inputName=None
       mask=0
       maskA=0
       maskB=0
       maskG=0
       maskR=0
       outputIndex=0
      }
     }
    ]
   }
  }
  MaterialExpressionCustom_2 {
   class=MaterialExpressionCustom
   inputs [
    Tex {from=MaterialExpressionNamedRerouteUsage_6, output=""}
    WorldPosition {from=MaterialExpressionNamedRerouteUsage_11, output=""}
    Time {from=MaterialExpressionTime_2, output=""}
    Speed {from=MaterialExpressionNamedRerouteUsage_10, output=""}
    Velocity {from=MaterialExpressionNamedRerouteUsage_7, output=""}
    AdvectedTextureScale {from=MaterialExpressionNamedRerouteUsage_8, output=""}
    PulseLength {from=MaterialExpressionNamedRerouteUsage_9, output=""}
    TextureAxis {from=MaterialExpressionNamedRerouteUsage_43, output=""}
    AlignAdvectedTextureToVelocity {from=MaterialExpressionNamedRerouteUsage_14, output=""}
   ]
   properties {
    code="return AdvectTexture_3Phase(WorldPosition, Velocity.xy, Speed, Time, PulseLength, AdvectedTextureScale, Tex, TexSampler, ((int) AlignAdvectedTextureToVelocity) == 1, TextureAxis);"
    outputType=CMOT_Float4
    showCode=true
    includeFilePaths [&ush]
    inputs [
     {
      inputName=Tex
      input {
       expression=&self:MaterialExpressionNamedRerouteUsage_6
       inputName=None
       mask=0
       maskA=0
       maskB=0
       maskG=0
       maskR=0
       outputIndex=0
      }
     }
     {
      inputName=WorldPosition
      input {
       expression=&self:MaterialExpressionNamedRerouteUsage_11
       inputName=None
       mask=0
       maskA=0
       maskB=0
       maskG=0
       maskR=0
       outputIndex=0
      }
     }
     {
      inputName=Time
      input {
       expression=&self:MaterialExpressionTime_2
       inputName=None
       mask=0
       maskA=0
       maskB=0
       maskG=0
       maskR=0
       outputIndex=0
      }
     }
     {
      inputName=Speed
      input {
       expression=&self:MaterialExpressionNamedRerouteUsage_10
       inputName=None
       mask=0
       maskA=0
       maskB=0
       maskG=0
       maskR=0
       outputIndex=0
      }
     }
     {
      inputName=Velocity
      input {
       expression=&self:MaterialExpressionNamedRerouteUsage_7
       inputName=None
       mask=0
       maskA=0
       maskB=0
       maskG=0
       maskR=0
       outputIndex=0
      }
     }
     {
      inputName=AdvectedTextureScale
      input {
       expression=&self:MaterialExpressionNamedRerouteUsage_8
       inputName=None
       mask=0
       maskA=0
       maskB=0
       maskG=0
       maskR=0
       outputIndex=0
      }
     }
     {
      inputName=PulseLength
      input {
       expression=&self:MaterialExpressionNamedRerouteUsage_9
       inputName=None
       mask=0
       maskA=0
       maskB=0
       maskG=0
       maskR=0
       outputIndex=0
      }
     }
     {
      inputName=TextureAxis
      input {
       expression=&self:MaterialExpressionNamedRerouteUsage_43
       inputName=None
       mask=0
       maskA=0
       maskB=0
       maskG=0
       maskR=0
       outputIndex=0
      }
     }
     {
      inputName=AlignAdvectedTextureToVelocity
      input {
       expression=&self:MaterialExpressionNamedRerouteUsage_14
       inputName=None
       mask=0
       maskA=0
       maskB=0
       maskG=0
       maskR=0
       outputIndex=0
      }
     }
    ]
   }
  }
  MaterialExpressionCustom_3 {
   class=MaterialExpressionCustom
   inputs [
    N {from=MaterialExpressionStaticSwitch_2, output=""}
    V {from=MaterialExpressionNamedRerouteUsage_17, output=""}
    Axis {from=MaterialExpressionNamedRerouteUsage_18, output=""}
    AlignAdvectedTextureToVelocity {from=MaterialExpressionNamedRerouteUsage_19, output=""}
   ]
   properties {
    code="bool Align = ((int) AlignAdvectedTextureToVelocity) == 1;\r\n\r\nif (Align)\r\n{\r\n    return RotateNormalToFlow(N, V, Axis);\r\n}\r\nelse\r\n{\r\n    return N;\r\n}"
    showCode=true
    includeFilePaths [&ush]
    inputs [
     {
      inputName=N
      input {
       expression=&self:MaterialExpressionStaticSwitch_2
       inputName=None
       mask=0
       maskA=0
       maskB=0
       maskG=0
       maskR=0
       outputIndex=0
      }
     }
     {
      inputName=V
      input {
       expression=&self:MaterialExpressionNamedRerouteUsage_17
       inputName=None
       mask=0
       maskA=0
       maskB=0
       maskG=0
       maskR=0
       outputIndex=0
      }
     }
     {
      inputName=Axis
      input {
       expression=&self:MaterialExpressionNamedRerouteUsage_18
       inputName=None
       mask=0
       maskA=0
       maskB=0
       maskG=0
       maskR=0
       outputIndex=0
      }
     }
     {
      inputName=AlignAdvectedTextureToVelocity
      input {
       expression=&self:MaterialExpressionNamedRerouteUsage_19
       inputName=None
       mask=0
       maskA=0
       maskB=0
       maskG=0
       maskR=0
       outputIndex=0
      }
     }
    ]
   }
  }
  MaterialExpressionCustom_4 {
   class=MaterialExpressionCustom
   inputs [
    Tex {from=MaterialExpressionNamedRerouteUsage_32, output=""}
    WorldPosition {from=MaterialExpressionNamedRerouteUsage_36, output=""}
    Time {from=MaterialExpressionTime_5, output=""}
    Speed {from=MaterialExpressionNamedRerouteUsage_35, output=""}
    Velocity {from=MaterialExpressionNamedRerouteUsage_33, output=""}
    AdvectedTextureScale {from=MaterialExpressionNamedRerouteUsage_34, output=""}
    PulseLength {from=MaterialExpressionNamedRerouteUsage_41, output=""}
    TheMipBias {from=MaterialExpressionNamedRerouteUsage_37, output=""}
    BracketAngleDelta {from=MaterialExpressionNamedRerouteUsage_38, output=""}
    BracketMagnitudeDelta {from=MaterialExpressionNamedRerouteUsage_39, output=""}
    SimNormal {from=MaterialExpressionNamedRerouteUsage_40, output=""}
    TextureAxis {from=MaterialExpressionNamedRerouteUsage_42, output=""}
    AlignAdvectedTextureToVelocity {from=MaterialExpressionNamedRerouteUsage_13, output=""}
   ]
   properties {
    code="return AdvectTexture_2Phase_Bracketed(WorldPosition, Velocity.xy, Speed, SimNormal, BracketAngleDelta, BracketMagnitudeDelta, PulseLength, Time, AdvectedTextureScale, TheMipBias, Tex, TexSampler, ((int) AlignAdvectedTextureToVelocity) == 1, TextureAxis);"
    outputType=CMOT_Float4
    showCode=true
    includeFilePaths [&ush]
    inputs [
     {
      inputName=Tex
      input {
       expression=&self:MaterialExpressionNamedRerouteUsage_32
       inputName=None
       mask=0
       maskA=0
       maskB=0
       maskG=0
       maskR=0
       outputIndex=0
      }
     }
     {
      inputName=WorldPosition
      input {
       expression=&self:MaterialExpressionNamedRerouteUsage_36
       inputName=None
       mask=0
       maskA=0
       maskB=0
       maskG=0
       maskR=0
       outputIndex=0
      }
     }
     {
      inputName=Time
      input {
       expression=&self:MaterialExpressionTime_5
       inputName=None
       mask=0
       maskA=0
       maskB=0
       maskG=0
       maskR=0
       outputIndex=0
      }
     }
     {
      inputName=Speed
      input {
       expression=&self:MaterialExpressionNamedRerouteUsage_35
       inputName=None
       mask=0
       maskA=0
       maskB=0
       maskG=0
       maskR=0
       outputIndex=0
      }
     }
     {
      inputName=Velocity
      input {
       expression=&self:MaterialExpressionNamedRerouteUsage_33
       inputName=None
       mask=0
       maskA=0
       maskB=0
       maskG=0
       maskR=0
       outputIndex=0
      }
     }
     {
      inputName=AdvectedTextureScale
      input {
       expression=&self:MaterialExpressionNamedRerouteUsage_34
       inputName=None
       mask=0
       maskA=0
       maskB=0
       maskG=0
       maskR=0
       outputIndex=0
      }
     }
     {
      inputName=PulseLength
      input {
       expression=&self:MaterialExpressionNamedRerouteUsage_41
       inputName=None
       mask=0
       maskA=0
       maskB=0
       maskG=0
       maskR=0
       outputIndex=0
      }
     }
     {
      inputName=TheMipBias
      input {
       expression=&self:MaterialExpressionNamedRerouteUsage_37
       inputName=None
       mask=0
       maskA=0
       maskB=0
       maskG=0
       maskR=0
       outputIndex=0
      }
     }
     {
      inputName=BracketAngleDelta
      input {
       expression=&self:MaterialExpressionNamedRerouteUsage_38
       inputName=None
       mask=0
       maskA=0
       maskB=0
       maskG=0
       maskR=0
       outputIndex=0
      }
     }
     {
      inputName=BracketMagnitudeDelta
      input {
       expression=&self:MaterialExpressionNamedRerouteUsage_39
       inputName=None
       mask=0
       maskA=0
       maskB=0
       maskG=0
       maskR=0
       outputIndex=0
      }
     }
     {
      inputName=SimNormal
      input {
       expression=&self:MaterialExpressionNamedRerouteUsage_40
       inputName=None
       mask=0
       maskA=0
       maskB=0
       maskG=0
       maskR=0
       outputIndex=0
      }
     }
     {
      inputName=TextureAxis
      input {
       expression=&self:MaterialExpressionNamedRerouteUsage_42
       inputName=None
       mask=0
       maskA=0
       maskB=0
       maskG=0
       maskR=0
       outputIndex=0
      }
     }
     {
      inputName=AlignAdvectedTextureToVelocity
      input {
       expression=&self:MaterialExpressionNamedRerouteUsage_13
       inputName=None
       mask=0
       maskA=0
       maskB=0
       maskG=0
       maskR=0
       outputIndex=0
      }
     }
    ]
   }
  }
  MaterialExpressionDeriveNormalZ_3 {class=MaterialExpressionDeriveNormalZ, inputs [InXY {from=MaterialExpressionSubtract_9, output=""}], properties {}}
  MaterialExpressionFunctionInput_0 {
   class=MaterialExpressionFunctionInput
   inputs [Preview {from=MaterialExpressionTextureObject_1, output=""}]
   properties {inputName=AdvectedTexture, inputType=FunctionInput_Texture2D}
  }
  MaterialExpressionFunctionInput_1 {
   class=MaterialExpressionFunctionInput
   inputs [Preview {from=MaterialExpressionConvert_0, output=""}]
   properties {bUsePreviewValueAsDefault=true, inputName=TextureAxis, inputType=FunctionInput_Vector2, previewValue=<0 1 0 1>}
  }
  MaterialExpressionFunctionInput_10 {
   class=MaterialExpressionFunctionInput
   inputs [Preview {from="", output=""}]
   properties {inputName=BracketMagnitudeDelta, inputType=FunctionInput_Scalar}
  }
  MaterialExpressionFunctionInput_12 {class=MaterialExpressionFunctionInput, inputs [Preview {from="", output=""}], properties {inputName=Sim Normal}}
  MaterialExpressionFunctionInput_13 {
   class=MaterialExpressionFunctionInput
   inputs [Preview {from="", output=""}]
   properties {inputName=MipLevelBias, inputType=FunctionInput_Scalar}
  }
  MaterialExpressionFunctionInput_14 {class=MaterialExpressionFunctionInput, inputs [Preview {from="", output=""}], properties {inputName=SimVelocity}}
  MaterialExpressionFunctionInput_2 {
   class=MaterialExpressionFunctionInput
   inputs [Preview {from="", output=""}]
   properties {inputName=AdvectedTextureScale, inputType=FunctionInput_Scalar}
  }
  MaterialExpressionFunctionInput_3 {
   class=MaterialExpressionFunctionInput
   inputs [Preview {from=MaterialExpressionConstant_0, output=""}]
   properties {bUsePreviewValueAsDefault=true, inputName=AlignAdvectedTextureToVelocity, inputType=FunctionInput_Scalar}
  }
  MaterialExpressionFunctionInput_4 {
   class=MaterialExpressionFunctionInput
   inputs [Preview {from=MaterialExpressionStaticBool_1, output=""}]
   properties {bUsePreviewValueAsDefault=true, inputName=Normal Output, inputType=FunctionInput_StaticBool}
  }
  MaterialExpressionFunctionInput_5 {
   class=MaterialExpressionFunctionInput
   inputs [Preview {from=MaterialExpressionStaticBool_2, output=""}]
   properties {bUsePreviewValueAsDefault=true, inputName=Compressed Normal, inputType=FunctionInput_StaticBool}
  }
  MaterialExpressionFunctionInput_7 {
   class=MaterialExpressionFunctionInput
   inputs [Preview {from="", output=""}]
   properties {inputName=AdvectionSpeed, inputType=FunctionInput_Scalar}
  }
  MaterialExpressionFunctionInput_8 {
   class=MaterialExpressionFunctionInput
   inputs [Preview {from="", output=""}]
   properties {inputName=PulseLength, inputType=FunctionInput_Scalar}
  }
  MaterialExpressionFunctionInput_9 {
   class=MaterialExpressionFunctionInput
   inputs [Preview {from="", output=""}]
   properties {inputName=BracketAngleDelta, inputType=FunctionInput_Scalar}
  }
  MaterialExpressionFunctionOutput_0 {class=MaterialExpressionFunctionOutput, inputs [None {from=MaterialExpressionStaticSwitch_1, output=""}], properties {outputName=Output}}
  MaterialExpressionFunctionOutput_1 {
   class=MaterialExpressionFunctionOutput
   inputs [None {from=MaterialExpressionStaticSwitchParameter_0, output=""}]
   properties {outputName=RawOutput}
  }
  MaterialExpressionMultiply_12 {
   class=MaterialExpressionMultiply
   inputs [
    A {from=MaterialExpressionComponentMask_11}
    B {from=""}
   ]
   properties {constB=2}
  }
  MaterialExpressionNamedRerouteDeclaration_0 {
   class=MaterialExpressionNamedRerouteDeclaration
   inputs [None {from=MaterialExpressionWorldPosition_0, output=XYZ}]
   properties {name=WorldPosition, nodeColor=<rgba 0 0.5781 1 1>}
  }
  MaterialExpressionNamedRerouteDeclaration_1 {
   class=MaterialExpressionNamedRerouteDeclaration
   inputs [None {from=MaterialExpressionFunctionInput_1}]
   properties {name=TextureAxis, nodeColor=<rgba 0 1 0.5547 1>}
  }
  MaterialExpressionNamedRerouteDeclaration_10 {
   class=MaterialExpressionNamedRerouteDeclaration
   inputs [None {from=MaterialExpressionFunctionInput_14}]
   properties {name=SimVelocity, nodeColor=<rgba 0.2891 1 0 1>}
  }
  MaterialExpressionNamedRerouteDeclaration_11 {
   class=MaterialExpressionNamedRerouteDeclaration
   inputs [None {from=MaterialExpressionFunctionInput_3}]
   properties {name=AlignAdvectedTextureToVelocity, nodeColor=<rgba 0.7812 0 1 1>}
  }
  MaterialExpressionNamedRerouteDeclaration_2 {
   class=MaterialExpressionNamedRerouteDeclaration
   inputs [None {from=MaterialExpressionFunctionInput_0}]
   properties {name=AdvectedTexture, nodeColor=<rgba 0.6641 1 0 1>}
  }
  MaterialExpressionNamedRerouteDeclaration_3 {
   class=MaterialExpressionNamedRerouteDeclaration
   inputs [None {from=MaterialExpressionFunctionInput_2}]
   properties {name=AdvectedTextureScale, nodeColor=<rgba 1 0.375 0 1>}
  }
  MaterialExpressionNamedRerouteDeclaration_4 {
   class=MaterialExpressionNamedRerouteDeclaration
   inputs [None {from=MaterialExpressionFunctionInput_7}]
   properties {name=AdvectionSpeed, nodeColor=<rgba 1 0.7734 0 1>}
  }
  MaterialExpressionNamedRerouteDeclaration_5 {
   class=MaterialExpressionNamedRerouteDeclaration
   inputs [None {from=MaterialExpressionFunctionInput_8}]
   properties {name=PulseLength, nodeColor=<rgba 0 1 0.5312 1>}
  }
  MaterialExpressionNamedRerouteDeclaration_6 {
   class=MaterialExpressionNamedRerouteDeclaration
   inputs [None {from=MaterialExpressionFunctionInput_9}]
   properties {name=BracketAngleDelta, nodeColor=<rgba 1 0 0.3516 1>}
  }
  MaterialExpressionNamedRerouteDeclaration_7 {
   class=MaterialExpressionNamedRerouteDeclaration
   inputs [None {from=MaterialExpressionFunctionInput_10}]
   properties {name=BracketMagnitudeDelta, nodeColor=<rgba 0.125 0 1 1>}
  }
  MaterialExpressionNamedRerouteDeclaration_8 {
   class=MaterialExpressionNamedRerouteDeclaration
   inputs [None {from=MaterialExpressionFunctionInput_12}]
   properties {name=Sim Normal, nodeColor=<rgba 0 0.0391 1 1>}
  }
  MaterialExpressionNamedRerouteDeclaration_9 {
   class=MaterialExpressionNamedRerouteDeclaration
   inputs [None {from=MaterialExpressionFunctionInput_13}]
   properties {name=MipLevelBias, nodeColor=<rgba 1 0 0.8438 1>}
  }
  MaterialExpressionNamedRerouteUsage_0 {class=MaterialExpressionNamedRerouteUsage}
  MaterialExpressionNamedRerouteUsage_1 {class=MaterialExpressionNamedRerouteUsage}
  MaterialExpressionNamedRerouteUsage_10 {class=MaterialExpressionNamedRerouteUsage}
  MaterialExpressionNamedRerouteUsage_11 {class=MaterialExpressionNamedRerouteUsage}
  MaterialExpressionNamedRerouteUsage_12 {class=MaterialExpressionNamedRerouteUsage}
  MaterialExpressionNamedRerouteUsage_13 {class=MaterialExpressionNamedRerouteUsage}
  MaterialExpressionNamedRerouteUsage_14 {class=MaterialExpressionNamedRerouteUsage}
  MaterialExpressionNamedRerouteUsage_15 {class=MaterialExpressionNamedRerouteUsage}
  MaterialExpressionNamedRerouteUsage_16 {class=MaterialExpressionNamedRerouteUsage}
  MaterialExpressionNamedRerouteUsage_17 {class=MaterialExpressionNamedRerouteUsage}
  MaterialExpressionNamedRerouteUsage_18 {class=MaterialExpressionNamedRerouteUsage}
  MaterialExpressionNamedRerouteUsage_19 {class=MaterialExpressionNamedRerouteUsage}
  MaterialExpressionNamedRerouteUsage_2 {class=MaterialExpressionNamedRerouteUsage}
  MaterialExpressionNamedRerouteUsage_22 {class=MaterialExpressionNamedRerouteUsage}
  MaterialExpressionNamedRerouteUsage_23 {class=MaterialExpressionNamedRerouteUsage}
  MaterialExpressionNamedRerouteUsage_24 {class=MaterialExpressionNamedRerouteUsage}
  MaterialExpressionNamedRerouteUsage_25 {class=MaterialExpressionNamedRerouteUsage}
  MaterialExpressionNamedRerouteUsage_26 {class=MaterialExpressionNamedRerouteUsage}
  MaterialExpressionNamedRerouteUsage_27 {class=MaterialExpressionNamedRerouteUsage}
  MaterialExpressionNamedRerouteUsage_28 {class=MaterialExpressionNamedRerouteUsage}
  MaterialExpressionNamedRerouteUsage_29 {class=MaterialExpressionNamedRerouteUsage}
  MaterialExpressionNamedRerouteUsage_3 {class=MaterialExpressionNamedRerouteUsage}
  MaterialExpressionNamedRerouteUsage_30 {class=MaterialExpressionNamedRerouteUsage}
  MaterialExpressionNamedRerouteUsage_31 {class=MaterialExpressionNamedRerouteUsage}
  MaterialExpressionNamedRerouteUsage_32 {class=MaterialExpressionNamedRerouteUsage}
  MaterialExpressionNamedRerouteUsage_33 {class=MaterialExpressionNamedRerouteUsage}
  MaterialExpressionNamedRerouteUsage_34 {class=MaterialExpressionNamedRerouteUsage}
  MaterialExpressionNamedRerouteUsage_35 {class=MaterialExpressionNamedRerouteUsage}
  MaterialExpressionNamedRerouteUsage_36 {class=MaterialExpressionNamedRerouteUsage}
  MaterialExpressionNamedRerouteUsage_37 {class=MaterialExpressionNamedRerouteUsage}
  MaterialExpressionNamedRerouteUsage_38 {class=MaterialExpressionNamedRerouteUsage}
  MaterialExpressionNamedRerouteUsage_39 {class=MaterialExpressionNamedRerouteUsage}
  MaterialExpressionNamedRerouteUsage_4 {class=MaterialExpressionNamedRerouteUsage}
  MaterialExpressionNamedRerouteUsage_40 {class=MaterialExpressionNamedRerouteUsage}
  MaterialExpressionNamedRerouteUsage_41 {class=MaterialExpressionNamedRerouteUsage}
  MaterialExpressionNamedRerouteUsage_42 {class=MaterialExpressionNamedRerouteUsage}
  MaterialExpressionNamedRerouteUsage_43 {class=MaterialExpressionNamedRerouteUsage}
  MaterialExpressionNamedRerouteUsage_44 {class=MaterialExpressionNamedRerouteUsage}
  MaterialExpressionNamedRerouteUsage_5 {class=MaterialExpressionNamedRerouteUsage}
  MaterialExpressionNamedRerouteUsage_6 {class=MaterialExpressionNamedRerouteUsage}
  MaterialExpressionNamedRerouteUsage_7 {class=MaterialExpressionNamedRerouteUsage}
  MaterialExpressionNamedRerouteUsage_8 {class=MaterialExpressionNamedRerouteUsage}
  MaterialExpressionNamedRerouteUsage_9 {class=MaterialExpressionNamedRerouteUsage}
  MaterialExpressionStaticBool_1 {class=MaterialExpressionStaticBool}
  MaterialExpressionStaticBool_2 {class=MaterialExpressionStaticBool}
  MaterialExpressionStaticSwitchParameter_0 {
   class=MaterialExpressionStaticSwitchParameter
   inputs [
    True {from=MaterialExpressionStaticSwitchParameter_1, output=""}
    False {from=MaterialExpressionStaticSwitchParameter_3, output=""}
   ]
   properties {defaultValue=true, group=Advanced, parameterName=Bracketing}
  }
  MaterialExpressionStaticSwitchParameter_1 {
   class=MaterialExpressionStaticSwitchParameter
   inputs [
    True {from=MaterialExpressionCustom_0, output=""}
    False {from=MaterialExpressionCustom_4, output=""}
   ]
   properties {defaultValue=true, group=Advanced, parameterName=ThreePhase}
  }
  MaterialExpressionStaticSwitchParameter_3 {
   class=MaterialExpressionStaticSwitchParameter
   inputs [
    True {from=MaterialExpressionCustom_2, output=""}
    False {from=MaterialExpressionCustom_1, output=""}
   ]
   properties {defaultValue=true, group=Advanced, parameterName=ThreePhase}
  }
  MaterialExpressionStaticSwitch_1 {
   class=MaterialExpressionStaticSwitch
   inputs [
    True {from=MaterialExpressionCustom_3, output=""}
    False {from=MaterialExpressionStaticSwitchParameter_0, output=""}
    Value {from=MaterialExpressionFunctionInput_4, output=""}
   ]
   properties {}
  }
  MaterialExpressionStaticSwitch_2 {
   class=MaterialExpressionStaticSwitch
   inputs [
    True {from=MaterialExpressionDeriveNormalZ_3, output=""}
    False {from=MaterialExpressionComponentMask_1, output=""}
    Value {from=MaterialExpressionFunctionInput_5, output=""}
   ]
   properties {}
  }
  MaterialExpressionSubtract_9 {
   class=MaterialExpressionSubtract
   inputs [
    A {from=MaterialExpressionMultiply_12}
    B {from=""}
   ]
  }
  MaterialExpressionTextureObject_1 {class=MaterialExpressionTextureObject, inputs [], properties {}}
  MaterialExpressionTime_1 {class=MaterialExpressionTime}
  MaterialExpressionTime_2 {class=MaterialExpressionTime}
  MaterialExpressionTime_4 {class=MaterialExpressionTime}
  MaterialExpressionTime_5 {class=MaterialExpressionTime}
  MaterialExpressionWorldPosition_0 {class=MaterialExpressionWorldPosition, inputs [], properties {}}
 }}
