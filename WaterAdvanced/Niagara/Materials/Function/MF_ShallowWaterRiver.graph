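# Material function that renders a baked or live Niagara shallow-water river simulation over incoming water material attributes, also outputting velocity, scattering and base normal.
# Sim depth, height, velocity, foam and normals drive opacity masking, world position offset and foam shading.
# Velocity advects wave detail and two foam layers, with low-speed attenuation.
# Static switches select live, virtual-texture and spline-mesh paths; a previous-frame WPO switch subtracts plane-projected velocity.
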MaterialFunction /WaterAdvanced/Niagara/Materials/Function/MF_ShallowWaterRiver
graph {nodes [
  FunctionOutput(None=NamedRerouteUsage()) {outputName=ShallowWater_BaseNormal, sortPriority=1}
  n1 = NamedRerouteUsage()
  n2 = FunctionInput() {inputType=FunctionInput_MaterialAttributes}
  n3 = BreakMaterialAttributes!(Attr=@n2)
  MaterialExpressionNamedRerouteUsage_20 = NamedRerouteUsage()
  n4 = NamedRerouteUsage()
  n5 = MaterialFunctionCall!(Base Normal=@n4, Detail Normal=NamedRerouteUsage()) {materialFunction=&/MF_BlendNormals}
  n6 = Saturate(None=NamedRerouteUsage())
  n7 = NamedRerouteDeclaration(None=StaticSwitchParameter(True=LinearInterpolate(A=@n4, B=@n5.Custom, Alpha=@n6), False=@n4) {defaultValue=true, parameterName=AdvectWaveHeight}) {name=SW_SimWithWPONormal, nodeColor=<rgba 0.6875 1 0 1>}
  n8 = NamedRerouteUsage()
  n9 = NamedRerouteUsage()
  n10 = NamedRerouteUsage()
  n11 = NamedRerouteUsage()
  MaterialExpressionLinearInterpolate_7 = LinearInterpolate(
   A=LinearInterpolate(
    A=@n7
    B=MaterialFunctionCall(Base Normal=@n7, Detail Normal=@n8) {materialFunction=&MF_BlendNormals}.Custom
    Alpha=Saturate(None=@n9)
   )
   B=Multiply(
    A=MaterialFunctionCall(Base Normal=@n10, Detail Normal=@n11) {materialFunction=&MF_BlendNormals}.Custom
    B=ScalarParameter() {defaultValue=1, group=Foam Shading, parameterName=FoamOpacity}
   )
   Alpha=NamedRerouteUsage()
  )
  n12 = Custom!(OpacityMask=@MaterialExpressionNamedRerouteUsage_20, Normal=@MaterialExpressionLinearInterpolate_7) {
   code="if (OpacityMask < 1e-8)\r\n{\r\n    return float3(0,0,0);\r\n}\r\nelse\r\n{\r\n    return Normal;\r\n}"
   showCode=true
   inputs [
    {
     inputName=OpacityMask
     input {expression=&self:MaterialExpressionNamedRerouteUsage_20}
    }
    {inputName=Normal, input {expression=&self:MaterialExpressionLinearInterpolate_7}}
   ]
  }
  n13 = Add(A=@n3.Normal, B=@n12)
  MaterialExpressionNamedRerouteUsage_11 = NamedRerouteUsage()
  MaterialExpressionConstant_3 = Constant() {r=1}
  MaterialExpressionMaterialFunctionCall_5 = MaterialFunctionCall!(Position) {materialFunction=/Water/Materials/Functions/WaterBodyData}
  n14 = Custom!(
   a=@MaterialExpressionNamedRerouteUsage_11
   b=@MaterialExpressionConstant_3
   SplineWaterDepth=@MaterialExpressionMaterialFunctionCall_5.Water Depth
  ) {
   code="if (SplineWaterDepth > 1e-5)\r\n{\r\n    return b;\r\n}\r\nelse\r\n{\r\n    return a;\r\n}"
   showCode=true
   inputs [
    {inputName=a, input {expression=&self:MaterialExpressionNamedRerouteUsage_11}}
    {inputName=b, input {expression=&self:MaterialExpressionConstant_3}}
    {
     inputName=SplineWaterDepth
     input {expression=&self:MaterialExpressionMaterialFunctionCall_5, outputIndex=2}
    }
   ]
  }
  n15 = Multiply(
   A=@n3.OpacityMask
   B=StaticSwitchParameter(
    True=@MaterialExpressionNamedRerouteUsage_11
    False=StaticSwitchParameter(True=@n14, False=@MaterialExpressionNamedRerouteUsage_11) {defaultValue=true, parameterName=RenderWithSplineMesh}
   ) {defaultValue=true, parameterName=SingleRiverRendering}
  )
  n16 = Constant()
  n17 = If(
   A=PixelDepth().R
   B=ComponentMask(None=SceneTexture() {sceneTextureId=PPI_CustomDepth}.Color) {r=true}
   "A > B"=@n16
   "A == B"=@n16
   "A < B"=Constant() {r=1}
  )
  n18 = SetMaterialAttributes!(
   MaterialAttributes=@n2
   BaseColor=LinearInterpolate(A=@n3.BaseColor, B=VectorParameter() {group=Foam Shading, parameterName=Foam Base Color, defaultValue=<rgba 1 1 1 0>}.RGB, Alpha=@n1)
   Specular=LinearInterpolate(A=@n3.Specular, B=ScalarParameter() {group=Foam Shading, parameterName=Foam Specular}, Alpha=@n1)
   Roughness=LinearInterpolate(A=@n3.Roughness, B=ScalarParameter() {group=Foam Shading, parameterName=Foam Roughness}, Alpha=@n1)
   Normal=MaterialFunctionCall(
    Vector=StaticSwitchParameter(True=@n12, False=@n13) {defaultValue=true, group=Advanced, parameterName=BakedSimNormalsOverrideUpstreamMaterialNormals}
    Default=Constant3Vector() {constant=<rgba 0 0 1 0>}
   ) {materialFunction=&SafeNormalize}.Result
   WorldPositionOffset=Add(A=@n3.WorldPositionOffset, B=NamedRerouteUsage())
   Opacity=Add(A=@n3.Opacity, B=@n1)
   OpacityMask=StaticSwitchParameter(True=Multiply(A=@n15, B=@n17), False=@n15) {defaultValue=true, parameterName=Use Custom Depth Culling}
  ) {attributeSetTypes [
    #1
    #2
    #3
    #4
    #5
    #6
    #7
   ]}
  FunctionOutput(None=StaticSwitchParameter(True=@n18, False=@n2) {defaultValue=true, group=Advanced, parameterName=Use Shallow Water}) {outputName=ShallowWaterMaterialAttributes, sortPriority=-2}
  n19 = Constant3Vector()
  n20 = VectorParameter() {group=Foam Shading, parameterName=Foam Scattering Color, defaultValue=<rgba 0 0 0 1>}
  n21 = Multiply(A=@n20.RGB, B=@n20.RGB) {constB=0.001}
  n22 = Max(
   A=SmoothStep(
    Min=ScalarParameter() {defaultValue=250, group=Velocity Scattering, parameterName=Scattering Min Velocity}
    Max=ScalarParameter() {defaultValue=10000, group=Velocity Scattering, parameterName=Scattering Max Velocity}
    Value=NamedRerouteUsage()
   ) {constMax=10000, constMin=250}
   B=NamedRerouteUsage()
  )
  FunctionOutput(None=StaticSwitchParameter(True=LinearInterpolate(A=@n19, B=@n21, Alpha=@n22), False=Constant3Vector()) {defaultValue=true, parameterName=Use Shallow Water}) {outputName=ShallowWater_SLW_Scattering, sortPriority=-1}
  n23 = ComponentMask(None=NamedRerouteUsage()) {g=true, r=true}
  FunctionOutput(None=@n23) {outputName=ShallowWater_Velocity}
  n24 = NamedRerouteUsage()
  n25 = ScalarParameter() {defaultValue=1, group=Foam, parameterName=Foam Scattering Area Fit 0-1, sortPriority=211}
  n26 = Saturate(None=Multiply(A=@n24, B=@n25))
  n27 = ScalarParameter() {defaultValue=1, group=Foam, parameterName=Foam Scattering Area Power, sortPriority=212}
  NamedRerouteDeclaration(None=Divide(A=Power(Base=@n26, Exp=@n27), B=@n25)) {name=SW_FoamScatteringMask, nodeColor=<rgba 1 0 0.5391 1>}
  NamedRerouteDeclaration(None=MaterialFunctionCall(Vector 2=@n23) {materialFunction=/Engine/Functions/Engine_MaterialFunctions02/Utility/VectorLength}.V2 Length) {name=SW_VelocityMagnitude, nodeColor=<rgba 1 0.1875 0 1>}
  MaterialExpressionNamedRerouteUsage_14 = NamedRerouteUsage()
  MaterialExpressionScalarParameter_1 = ScalarParameter() {defaultValue=0.96, group=Waves, parameterName=LowVelocityWPOWaveScale}
  MaterialExpressionScalarParameter_2 = ScalarParameter() {defaultValue=150, group=Waves, parameterName=WPOWaveAttenMin}
  MaterialExpressionScalarParameter_65 = ScalarParameter() {defaultValue=800, group=Waves, parameterName=WPOWaveAttenMax}
  MaterialExpressionScalarParameter_67 = ScalarParameter() {defaultValue=1, group=Waves, parameterName=WPOWaveHeightMult}
  n28 = Custom!(
   vmag=@MaterialExpressionNamedRerouteUsage_14
   lvws=@MaterialExpressionScalarParameter_1
   WaveAttenuationMin=@MaterialExpressionScalarParameter_2
   WaveAttenuationMax=@MaterialExpressionScalarParameter_65
   WaveScale=@MaterialExpressionScalarParameter_67
  ) {
   code="lerp(1, smoothstep(WaveAttenuationMin, WaveAttenuationMax, vmag), 1. - saturate(lvws));"
   outputType=CMOT_Float1
   showCode=true
   inputs [
    {inputName=vmag, input {expression=&self:MaterialExpressionNamedRerouteUsage_14}}
    {inputName=lvws, input {expression=&self:MaterialExpressionScalarParameter_1}}
    {
     inputName=WaveAttenuationMin
     input {expression=&self:MaterialExpressionScalarParameter_2}
    }
    {
     inputName=WaveAttenuationMax
     input {expression=&self:MaterialExpressionScalarParameter_65}
    }
    {inputName=WaveScale, input {expression=&self:MaterialExpressionScalarParameter_67}}
   ]
  }
  n29 = ScalarParameter() {defaultValue=1, group=Waves, parameterName=LowVelocityWPOWaveAttenuationPower}
  NamedRerouteDeclaration(None=Multiply(A=Power(Base=@n28, Exp=@n29), B=@MaterialExpressionScalarParameter_67)) {name=LowVelocityWPOWaveAttenuation, nodeColor=<rgba 0.2188 1 0 1>}
  MaterialExpressionNamedRerouteUsage_4 = NamedRerouteUsage()
  MaterialExpressionScalarParameter_10 = ScalarParameter() {defaultValue=0.96, group=Waves, parameterName=LowVelocityWaveScale}
  MaterialExpressionScalarParameter_11 = ScalarParameter() {defaultValue=150, group=Waves, parameterName=WaveAttenMin}
  MaterialExpressionScalarParameter_12 = ScalarParameter() {defaultValue=800, group=Waves, parameterName=WaveAttenMax}
  MaterialExpressionScalarParameter_19 = ScalarParameter() {defaultValue=1, group=Waves, parameterName=WaveAmount}
  n30 = Custom(
   vmag=@MaterialExpressionNamedRerouteUsage_4
   lvws=@MaterialExpressionScalarParameter_10
   WaveAttenuationMin=@MaterialExpressionScalarParameter_11
   WaveAttenuationMax=@MaterialExpressionScalarParameter_12
   WaveScale=@MaterialExpressionScalarParameter_19
  ) {
   code="lerp(1, smoothstep(WaveAttenuationMin, WaveAttenuationMax, vmag), 1. - saturate(lvws)) * WaveScale;"
   outputType=CMOT_Float1
   showCode=true
   inputs [
    {inputName=vmag, input {expression=&self:MaterialExpressionNamedRerouteUsage_4}}
    {inputName=lvws, input {expression=&self:MaterialExpressionScalarParameter_10}}
    {
     inputName=WaveAttenuationMin
     input {expression=&self:MaterialExpressionScalarParameter_11}
    }
    {
     inputName=WaveAttenuationMax
     input {expression=&self:MaterialExpressionScalarParameter_12}
    }
    {inputName=WaveScale, input {expression=&self:MaterialExpressionScalarParameter_19}}
   ]
  }
  NamedRerouteDeclaration(None=@n30) {name=LowVelocityWaveAttenuation, nodeColor=<rgba 0.2187 1 0 1>}
  n31 = ScalarParameter() {defaultValue=1, group=Foam, parameterName=Foam Mask Fit 0-1, sortPriority=201}
  n32 = Power(
   Base=Saturate(None=Multiply(A=@n24, B=@n31))
   Exp=ScalarParameter() {defaultValue=1, group=Foam, parameterName=Foam Mask Power, sortPriority=202}
  )
  NamedRerouteDeclaration(None=Multiply(A=Divide(A=@n32, B=@n31), B=NamedRerouteUsage())) {name=SW_FoamMask, nodeColor=<rgba 0 1 0.6719 1>}
  MaterialExpressionNamedRerouteUsage_27 = NamedRerouteUsage()
  MaterialExpressionScalarParameter_42 = ScalarParameter() {group=Foam Attenuation, parameterName=LowVelocityFoamScale}
  MaterialExpressionScalarParameter_44 = ScalarParameter() {defaultValue=10, group=Foam Attenuation, parameterName=FoamAttenMin}
  MaterialExpressionScalarParameter_46 = ScalarParameter() {defaultValue=50, group=Foam Attenuation, parameterName=FoamAttenMax}
  n33 = Custom!(
   vmag=@MaterialExpressionNamedRerouteUsage_27
   lvfs=@MaterialExpressionScalarParameter_42
   AttenuationMin=@MaterialExpressionScalarParameter_44
   AttenuationMax=@MaterialExpressionScalarParameter_46
  ) {
   code="lerp(1, smoothstep(AttenuationMin, AttenuationMax, vmag), 1. - saturate(lvfs));"
   outputType=CMOT_Float1
   showCode=true
   inputs [
    {inputName=vmag, input {expression=&self:MaterialExpressionNamedRerouteUsage_27}}
    {inputName=lvfs, input {expression=&self:MaterialExpressionScalarParameter_42}}
    {
     inputName=AttenuationMin
     input {expression=&self:MaterialExpressionScalarParameter_44}
    }
    {
     inputName=AttenuationMax
     input {expression=&self:MaterialExpressionScalarParameter_46}
    }
   ]
  }
  NamedRerouteDeclaration(None=@n33) {name=LowVelocityFoamAttenuation, nodeColor=<rgba 0.6172 1 0 1>}
  n34 = ScalarParameter() {defaultValue=0.25, group=Foam, parameterName=Foam Render Fit 0-1}
  n35 = NamedRerouteUsage()
  n36 = OneMinus(None=Saturate(None=@n35))
  n37 = NamedRerouteUsage()
  n38 = TextureObjectParameter() {
   group=Foam
   parameterName=FoamWithNormalTexture
   samplerType=SAMPLERTYPE_LinearColor
   sortPriority=5
   texture=/WaterAdvanced/Niagara/Textures/T_ShallowWater_Foam
  }
  n39 = Constant()
  n40 = NamedRerouteUsage()
  n41 = ScalarParameter() {defaultValue=1, group=Foam, parameterName=FoamSpeed, sortPriority=10}
  n42 = ScalarParameter() {defaultValue=5, group=Advection, parameterName=BracketAngleDelta}
  n43 = ScalarParameter() {defaultValue=50, group=Advection, parameterName=BracketMagnitudeDelta}
  n44 = ScalarParameter() {defaultValue=1, group=Advection, parameterName=PulseLength}
  n45 = MaterialFunctionCall!(
   Sim Normal=@n37
   AdvectedTextureScale=ScalarParameter() {defaultValue=15, group=Foam, parameterName=FoamWithNormalTextureScale, sortPriority=6}
   AdvectionSpeed=@n41
   PulseLength=@n44
   BracketAngleDelta=@n42
   BracketMagnitudeDelta=@n43
   AdvectedTexture=@n38
   MipLevelBias=@n39
   SimVelocity=@n40
   TextureAxis
   AlignAdvectedTextureToVelocity=Constant() {r=1}
   Normal Output=StaticBool() {value=true}
   Compressed Normal
  ) {materialFunction=&/MF_TextureAdvection}
  n46 = ComponentMask(None=@n45.RawOutput) {a=true}
  n47 = Saturate(None=@n46)
  n48 = Power(
   Base=LinearInterpolate(
    A=SmoothStep(Min=@n36, Value=@n47)
    B=Multiply(A=@n35, B=@n46)
    Alpha=ScalarParameter() {group=Foam, parameterName=Black Point to Mult Blend}
   )
   Exp=ScalarParameter() {defaultValue=1, group=Foam, parameterName=FoamWithNormalPower, sortPriority=31}
  ) {constExponent=4}
  n49 = ScalarParameter() {defaultValue=1, group=Foam, parameterName=FoamWithNormalGain, sortPriority=30}
  n50 = OneMinus(None=Saturate(None=NamedRerouteUsage()))
  n51 = MaterialFunctionCall(
   Sim Normal=@n37
   AdvectedTextureScale=ScalarParameter() {defaultValue=15, group=Foam, parameterName=SecondaryFoamTextureScale, sortPriority=106}
   AdvectionSpeed=Multiply(A=@n41, B=ScalarParameter() {defaultValue=0.5, group=Foam, parameterName=SecondaryFoamSpeedMult, sortPriority=107})
   PulseLength=Multiply(A=@n44, B=ScalarParameter() {defaultValue=3, group=Advection, parameterName=SecondaryFoamPulseMult})
   BracketAngleDelta=@n42
   BracketMagnitudeDelta=@n43
   AdvectedTexture=TextureObjectParameter() {group=Foam, parameterName=SecondaryFoamTexture, sortPriority=105, texture=/Water/Textures/Foam/T_WaterFlow_01_Foam_Tiled}
   MipLevelBias=@n39
   SimVelocity=@n40
  ) {materialFunction=&MF_TextureAdvection}
  n52 = Saturate(None=ComponentMask(None=NamedRerouteDeclaration(None=@n51.Output) {name=Secondary Advected Foam, nodeColor=<rgba 0 1 0.7656 1>}) {r=true})
  n53 = Multiply(
   A=Power(
    Base=SmoothStep(Min=@n50, Value=@n52)
    Exp=ScalarParameter() {defaultValue=1, group=Foam, parameterName=SecondaryFoamPower, sortPriority=131}
   ) {constExponent=4}
   B=ScalarParameter() {defaultValue=1, group=Foam, parameterName=SecondaryFoamGain, sortPriority=130}
  )
  n54 = Constant()
  n55 = Saturate(None=Max(
    A=Multiply(A=@n48, B=@n49)
    B=StaticSwitchParameter(True=@n53, False=@n54) {defaultValue=true, group=Advanced, parameterName=Use Secondary Foam}
   ))
  n56 = StaticSwitchParameter(
   True=StaticSwitchParameter(True=NamedRerouteUsage(), False=SmoothStep(Max=@n34, Value=@n55) {constMax=0.25}) {group=Foam, parameterName=Debug Foam Mask, sortPriority=200}
   False=Constant3Vector()
  ) {group=Advanced, parameterName=SupportFoam}
  NamedRerouteDeclaration(None=@n56) {name=SW_TotalFoamAmount, nodeColor=<rgba 0 1 1 1>}
  n57 = Constant3Vector()
  n58 = MaterialFunctionCall(Normal=@n45.Output, Flatness=ScalarParameter() {group=Foam, parameterName=Foam Normal Flatness, sortPriority=300}) {materialFunction=&FlattenNormal}
  NamedRerouteDeclaration(None=Normalize(VectorInput=StaticSwitchParameter(True=@n57, False=@n58.Result) {group=Foam, parameterName=Debug Foam Mask, sortPriority=200})) {name=SW_FoamNormal, nodeColor=<rgba 0.9688 0 1 1>}
  n59 = MaterialFunctionCall(
   Sim Normal=NamedRerouteUsage()
   AdvectedTextureScale=ScalarParameter() {defaultValue=1, group=Waves, parameterName=WPOWaveScale}
   AdvectionSpeed=ScalarParameter() {defaultValue=1, group=Waves, parameterName=WaveSpeed}
   PulseLength=ScalarParameter() {defaultValue=1, group=Advection, parameterName=PulseLength}
   BracketAngleDelta=ScalarParameter() {defaultValue=5, group=Advection, parameterName=BracketAngleDelta}
   BracketMagnitudeDelta=ScalarParameter() {defaultValue=50, group=Advection, parameterName=BracketMagnitudeDelta}
   AdvectedTexture=TextureObjectParameter() {group=Set From Niagara, parameterName=NormalDetailTex}
   MipLevelBias=ScalarParameter() {defaultValue=2, group=Waves, parameterName=WPOWaveMIPBias}
   SimVelocity=NamedRerouteUsage()
   AlignAdvectedTextureToVelocity=Constant() {r=1}
   Normal Output=StaticBool() {value=true}
  ) {materialFunction=&MF_TextureAdvection}
  n60 = Multiply(A=ComponentMask(None=@n59.RawOutput) {a=true}, B=NamedRerouteUsage())
  n61 = ScalarParameter() {defaultValue=200, group=Waves, parameterName=AdvectedWPOWaveCameraFalloff}
  n62 = SmoothStep(Min=@n61, Max=Add(A=@n61) {constB=50}, Value=Distance(A=WorldPosition().XYZ, B=CameraPositionWS())) {constMax=300, constMin=200}
  NamedRerouteDeclaration(None=StaticSwitchParameter(True=Multiply(A=@n60, B=@n62), False=Constant()) {parameterName=AdvectWaveHeight}) {name=SW_AdvectedWPOWaveHeight, nodeColor=<rgba 0.4766 1 0 1>}
  NamedRerouteDeclaration(None=Multiply(A=@n56, B=ScalarParameter() {defaultValue=1, group=Foam Shading, parameterName=FoamOpacity})) {name=SW_FoamOpacity, nodeColor=<rgba 1 0 0.7969 1>}
  n63 = ComponentMask(None=WorldPosition().XYZ) {g=true, r=true}
  n64 = DoubleVectorParameter() {parameterName=BakedWaterSimLocationDouble}
  n65 = ComponentMask(None=@n64.XYZ) {g=true, r=true}
  n66 = VectorParameter() {group=Set From Niagara, parameterName=BakedWaterSimSize}
  n67 = Add(A=Divide(A=Subtract(A=@n63, B=@n65), B=ComponentMask(None=@n66.RGB) {g=true, r=true})) {constB=0.5}
  NamedRerouteDeclaration(None=StaticSwitchParameter(True=Clamp(None=@n67), False=TextureCoordinate()) {defaultValue=true, parameterName=RenderWithSplineMesh}) {name=SW_FluidTextureUV, nodeColor=<rgba 0.5938 0 1 1>}
  n68 = NamedRerouteUsage()
  n69 = TextureObjectParameter() {group=Set From Niagara, parameterName=FoamTex, texture=&Black}
  n70 = StaticBool() {value=true}
  n71 = Constant()
  n72 = NamedRerouteUsage()
  n73 = TextureObjectParameter() {
   group=Set From Niagara
   parameterName=BakedWaterSimTexVT
   samplerType=SAMPLERTYPE_VirtualColor
   texture=/WaterAdvanced/Niagara/Textures/BlackVT
  }
  n74 = MaterialFunctionCall!(
   Use Sample=StaticBool()
   Mip Level=@n71
   Uvs=@n72
   Tex=TextureObjectParameter() {group=Set From Niagara, parameterName=BakedWaterSimTex, texture=&Black}
  ) {materialFunction=&/MF_ShallowWaterRiver_BilinearInterp}
  n75 = MaterialFunctionCall(Float4=StaticSwitchParameter(
    True=MaterialFunctionCall(Use Sample=@n70, Mip Level=@n71, Uvs=@n72, Tex=@n73) {materialFunction=&MF_ShallowWaterRiver_BilinearInterp}.Value
    False=@n74.Value
   ) {parameterName=UseVT}) {materialFunction=&BreakOutFloat4Components}
  NamedRerouteDeclaration(None=StaticSwitchParameter(True=TextureSample(UVs=@n68, Tex=@n69) {samplerSource=SSM_Clamp_WorldGroupSettings}.R, False=@n75.A) {parameterName=LiveSim}) {name=SW_BakedFoamMask, nodeColor=<rgba 1 0.5625 0 1>}
  n76 = MaterialFunctionCall(Float4=@n74.Value) {materialFunction=&BreakOutFloat4Components}
  n77 = ScalarParameter() {parameterName=BakedWaterDepthMax}
  NamedRerouteDeclaration(None=StaticSwitchParameter(True=@n76.G, False=Multiply(A=@n75.R, B=@n77)) {parameterName=LiveSim}) {name=SW_BakedWaterDepth, nodeColor=<rgba 0.2422 0 1 1>}
  n78 = TextureObjectParameter() {group=Set From Niagara, parameterName=BakedWaterSimTexVT, samplerType=SAMPLERTYPE_VirtualColor, texture=/Water/Textures/BlackVT4x4}
  n79 = TextureObjectParameter() {group=Set From Niagara, parameterName=BakedWaterSimTex, texture=&Black}
  n80 = StaticSwitchParameter(True=@n78, False=@n79) {parameterName=UseVT}
  NamedRerouteDeclaration(None=TextureProperty(Tex=StaticSwitchParameter(True=@n79, False=@n80) {parameterName=LiveSim})) {name=SW_BakedSimTextureSize, nodeColor=<rgba 0.0781 1 0 1>}
  n81 = NamedRerouteUsage()
  n82 = TextureObjectParameter() {
   group=Set From Niagara
   parameterName=BakedWaterSimNormalTexVT
   samplerType=SAMPLERTYPE_VirtualColor
   texture=&BlackVT
  }
  n83 = TextureObjectParameter() {group=Set From Niagara, parameterName=BakedWaterSimNormalTex, texture=&Black}
  n84 = StaticSwitchParameter(True=@n83, False=StaticSwitchParameter(True=@n82, False=@n83) {parameterName=UseVT}) {parameterName=LiveSim}
  n85 = TextureObjectParameter() {group=Set From Niagara, parameterName=BakedWaterHeightTexVT, samplerType=SAMPLERTYPE_VirtualColor, texture=&BlackVT4x4}
  n86 = StaticSwitchParameter(
   True=TextureObjectParameter() {group=Set From Niagara, parameterName=BakedWaterSimTex, texture=&Black}
   False=StaticSwitchParameter(True=@n85, False=TextureObjectParameter() {group=Set From Niagara, parameterName=BakedWaterHeightTex, texture=&Black}) {parameterName=UseVT}
  ) {parameterName=LiveSim}
  n87 = NamedRerouteUsage()
  n88 = VectorParameter() {group=Set From Niagara, parameterName=BakedWaterSimSize}
  NamedRerouteDeclaration(None=StaticSwitchParameter(
    True=TextureSample!(UVs=@n81, Tex=@n84, Level, Apply View MipBias) {mipValueMode=TMVM_MipLevel}.RGB
    False=MaterialFunctionCall!(HeightScale, Tex=@n86, UV=@n81, CenterHeight=@n87, WorldSpaceSize=@n88.RGB) {materialFunction=&/MF_CalculateNormalsInline}.Result
   ) {defaultValue=true, parameterName=UseBakedSimNormals}) {name=SW_SimNormal, nodeColor=<rgba 1 0 0.4922 1>}
  n89 = Constant()
  n90 = Subtract(A=MaterialFunctionCall(X=@n75.G, Y=@n75.G, Z=Constant() {r=0.5}) {materialFunction=&MakeFloat3}.Result) {constB=0.5}
  n91 = Multiply(A=ScalarParameter() {parameterName=BakedWaterVelocityMaxSpeed}) {constB=2}
  NamedRerouteDeclaration(None=StaticSwitchParameter(
    True=MaterialFunctionCall(X=@n76.B, Y=@n76.B, Z=@n89) {materialFunction=&MakeFloat3}.Result
    False=Multiply(A=@n90, B=@n91) {constB=1212}
   ) {parameterName=LiveSim}) {name=SW_BakedWaterVelocity, nodeColor=<rgba 0.6875 1 0 1>}
  MaterialExpressionNamedRerouteUsage_42 = NamedRerouteUsage()
  n92 = Custom!(a=@MaterialExpressionNamedRerouteUsage_42) {
   code="a > 1e-8;"
   outputType=CMOT_Float1
   inputs [{inputName=a, input {expression=&self:MaterialExpressionNamedRerouteUsage_42}}]
  }
  MaterialExpressionMultiply_2 = Multiply(A=@n92, B=ScalarParameter() {defaultValue=1, group=Water Surface Shading, parameterName=Opacity}) {constA=1}
  NamedRerouteDeclaration(None=@MaterialExpressionMultiply_2) {name=SW_WaterOpacity, nodeColor=<rgba 0 1 0.9062 1>}
  n93 = Constant()
  n94 = NamedRerouteUsage()
  n95 = TextureObjectParameter() {group=Set From Niagara, parameterName=BakedWaterHeightTexVT, samplerType=SAMPLERTYPE_VirtualColor, texture=&BlackVT4x4}
  n96 = MaterialFunctionCall(Use Sample=StaticBool() {value=true}, Mip Level=@n93, Uvs=@n94, Tex=@n95) {materialFunction=&MF_ShallowWaterRiver_BilinearInterp}
  n97 = MaterialFunctionCall(
   Use Sample=StaticBool()
   Mip Level=@n93
   Uvs=@n94
   Tex=TextureObjectParameter() {group=Set From Niagara, parameterName=BakedWaterHeightTex, texture=&Black}
  ) {materialFunction=&MF_ShallowWaterRiver_BilinearInterp}
  NamedRerouteDeclaration(None=StaticSwitchParameter(True=@n76.R, False=StaticSwitchParameter(True=@n96.Value, False=@n97.Value) {parameterName=UseVT}) {parameterName=LiveSim}) {name=SW_BakedWaterHeight, nodeColor=<rgba 0.0547 1 0 1>}
  NamedRerouteDeclaration(None=StaticSwitchParameter(True=ComponentMask(None=@n59.Output) {b=true, g=true, r=true}, False=Constant3Vector()) {defaultValue=true, group=Advanced, parameterName=AdvectWaveHeight}) {name=SW_AdvectedWPOWaveNormal, nodeColor=<rgba 0.8984 1 0 1>}
  NamedRerouteDeclaration(None=MaterialFunctionCall(
    Sim Normal=NamedRerouteUsage()
    AdvectedTextureScale=ScalarParameter() {defaultValue=1, group=Waves, parameterName=NormalTextureScale}
    AdvectionSpeed=ScalarParameter() {defaultValue=1, group=Waves, parameterName=WaveSpeed}
    PulseLength=ScalarParameter() {defaultValue=1, group=Advection, parameterName=PulseLength}
    BracketAngleDelta=ScalarParameter() {defaultValue=5, group=Advection, parameterName=BracketAngleDelta}
    BracketMagnitudeDelta=ScalarParameter() {defaultValue=50, group=Advection, parameterName=BracketMagnitudeDelta}
    AdvectedTexture=TextureObjectParameter() {group=Set From Niagara, parameterName=NormalDetailTex}
    MipLevelBias=Constant()
    SimVelocity=NamedRerouteUsage()
    AlignAdvectedTextureToVelocity=Constant() {r=1}
    Normal Output=StaticBool() {value=true}
   ) {materialFunction=&MF_TextureAdvection}.Output) {name=SW_AdvectedWaveNormal, nodeColor=<rgba 1 0.3281 0 1>}
  n98 = Multiply(A=NamedRerouteUsage(), B=DeltaTime())
  n99 = ScalarParameter() {defaultValue=1, group=Advanced, parameterName=PreviousFrameVelocityMult}
  n100 = NamedRerouteUsage()
  MaterialExpressionMultiply_5 = Multiply(A=Multiply(A=Multiply(A=@n98, B=@n99), B=Saturate(None=@n100)), B=ScalarParameter() {defaultValue=1, group=Waves, parameterName=WaveSpeed})
  MaterialExpressionNamedRerouteUsage_50 = NamedRerouteUsage()
  n101 = Custom!(v=@MaterialExpressionMultiply_5, n=@MaterialExpressionNamedRerouteUsage_50) {
   code="return v - dot(v,n) * n;"
   description=Project Velocity onto plane
   showCode=true
   inputs [
    {inputName=v, input {expression=&self:MaterialExpressionMultiply_5}}
    {inputName=n, input {expression=&self:MaterialExpressionNamedRerouteUsage_50}}
   ]
  }
  n102 = Add(A=NamedRerouteUsage(), B=NamedRerouteUsage())
  n103 = Add(A=@n102, B=@n64.Z)
  n104 = WorldPosition()
  n105 = StaticSwitchParameter(
   True=MaterialFunctionCall(Z=Subtract(A=@n103, B=@n104.Z)) {materialFunction=&MakeFloat3}.Result
   False=Transform(None=MaterialFunctionCall(Z=@n102) {materialFunction=&MakeFloat3}.Result) {transformSourceType=TRANSFORMSOURCE_Local}
  ) {defaultValue=true, parameterName=RenderWithSplineMesh}
  n106 = Subtract(A=@n105, B=@n101)
  NamedRerouteDeclaration(None=StaticSwitchParameter(True=PreviousFrameSwitch(Current Frame=@n105, Previous Frame=@n106), False=@n105) {defaultValue=true, group=Advanced, parameterName=Use Previous Frame Switch}) {name=SW_WPO, nodeColor=<rgba 0.0078 0 1 1>}
  MaterialExpressionCustom_5 = Custom!(a=@MaterialExpressionMultiply_2) {
   code="a > 1e-8"
   description=Mask Opacity
   outputType=CMOT_Float1
   inputs [{inputName=a, input {expression=&self:MaterialExpressionMultiply_2}}]
  }
  MaterialExpressionNamedRerouteUsage_3 = NamedRerouteUsage()
  MaterialExpressionNamedRerouteUsage_43 = NamedRerouteUsage()
  n107 = Custom!(OpacityMask=@MaterialExpressionCustom_5, uv=@MaterialExpressionNamedRerouteUsage_3, sz=@MaterialExpressionNamedRerouteUsage_43) {
   code="float2 UnitDx = 1./sz;\r\n\r\nif (uv.x < UnitDx.x || uv.x > 1.-UnitDx.x || uv.y < UnitDx.y || uv.y > 1.-UnitDx.y)\r\n{\r\n  return 0.;\r\n}\r\nreturn OpacityMask;"
   outputType=CMOT_Float1
   showCode=true
   inputs [
    {inputName=OpacityMask, input {expression=&self:MaterialExpressionCustom_5}}
    {inputName=uv, input {expression=&self:MaterialExpressionNamedRerouteUsage_3}}
    {inputName=sz, input {expression=&self:MaterialExpressionNamedRerouteUsage_43}}
   ]
  }
  NamedRerouteDeclaration(None=@n107) {name=SW_OpacityMask, nodeColor=<rgba 0 0.7188 1 1>}
 ]}
