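# Masked Single Layer Water material for shallow-water rivers, enabled for water and Niagara mesh particles, with world-space normals.
# Base attributes from parameters and constant 1.3 refraction pass through MF_ShallowWaterRiver, then MF_SampleInteractiveSim.
# Water scattering adds squared Scattering color to the river function's scattering output, then divides by 1000.
# Absorption is the reciprocal of Absorption RGB divided by its alpha.
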Material /WaterAdvanced/Niagara/Materials/M_SW_River
graph {
 properties {
  allowTranslucentCustomDepthWrites=true
  bScreenSpaceReflections=true
  bTangentSpaceNormal=false
  bUseMaterialAttributes=true
  bUsedWithNiagaraMeshParticles=true
  bUsedWithWater=true
  blendMode=BLEND_Masked
  pixelDepthOffsetMode=PDOM_Legacy
  refractionMethod=RM_PixelNormalOffset
  shadingModel=MSM_SingleLayerWater
  translucencyLightingMode=TLM_Surface
  usedShadingModels=SingleLayerWater
 }
 nodes [
  n1 = MaterialFunctionCall!(In=MakeMaterialAttributes!(
    BaseColor=VectorParameter() {group=General, parameterName=Water Color, defaultValue=<rgba 1 1 1 0>}.RGB
    Metallic=ScalarParameter() {defaultValue=0.05, group=General, parameterName=Metallic}
    Specular=ScalarParameter() {defaultValue=0.9, group=General, parameterName=Specular}
    Roughness=ScalarParameter() {group=General, parameterName=Roughness}
    Anisotropy
    EmissiveColor
    Opacity
    OpacityMask
    Normal
    Tangent
    WorldPositionOffset
    SubsurfaceColor
    ClearCoat
    ClearCoatRoughness
    AmbientOcclusion
    Refraction=Constant() {r=1.3}
    CustomizedUVs_0
    CustomizedUVs_1
    CustomizedUVs_2
    CustomizedUVs_3
    CustomizedUVs_4
    CustomizedUVs_5
    CustomizedUVs_6
    CustomizedUVs_7
    PixelDepthOffset
    ShadingModel
    Displacement
   )) {materialFunction=/WaterAdvanced/Niagara/Materials/Function/MF_ShallowWaterRiver}
  MaterialFunctionCall!(Attributes=@n1.ShallowWaterMaterialAttributes) {materialFunction=/WaterAdvanced/Niagara/Materials/Function/MF_SampleInteractiveSim}
  n2 = VectorParameter() {parameterName=Scattering, defaultValue=<rgba 1 1 1 0.5>}
  n3 = Multiply(A=@n2.RGB, B=@n2.RGB)
  n4 = VectorParameter() {parameterName=Absorption, defaultValue=<rgba 1 1 1 0>}
  SingleLayerWaterMaterialOutput!(
   ScatteringCoefficients=Divide(A=Add(A=@n3, B=@n1.ShallowWater_SLW_Scattering)) {constB=1000}
   AbsorptionCoefficients=Divide(A=Divide(B=@n4.RGB), B=@n4.A)
   PhaseG=ScalarParameter() {defaultValue=0.1, group=Water Shading, parameterName=Anisotropy}
   ColorScaleBehindWater
  )
 ]
}
