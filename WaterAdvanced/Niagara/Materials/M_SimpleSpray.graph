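# Translucent, two-sided subsurface material for Niagara water spray sprites, using volumetric directional lighting.
# Particle-tinted base color blends from WaterColor by vertical distance to the WaterBodyData surface height, and also feeds a subsurface gain.
# Per-particle randomized noise distorts the UVs of a soft circular opacity mask, scaled by particle alpha.
# A custom HLSL node builds spherical normals, blended with amplified noise normals to fake volume on flat sprites.
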Material /WaterAdvanced/Niagara/Materials/M_SimpleSpray
graph {
 properties {
  bEnableResponsiveAA=true
  bUsedWithNiagaraSprites=true
  blendMode=BLEND_Translucent
  shadingModel=MSM_Subsurface
  translucencyLightingMode=TLM_VolumetricDirectional
  twoSided=true
  usedShadingModels=Subsurface
 }
 nodes [
  n1 = ParticleColor()
  n2 = Multiply(A=VectorParameter() {parameterName=Base Color, defaultValue=<rgba 0.8 0.8 0.8 1>}.RGB, B=@n1.RGB)
  n3 = WorldPosition()
  n4 = Abs(None=Subtract(A=@n3.Z, B=MaterialFunctionCall!(Position=@n3.XY) {materialFunction=/Water/Materials/Functions/WaterBodyData}.Water Z Height))
  n5 = ScalarParameter() {defaultValue=5, parameterName=BlendWithWaterColor}
  LinearInterpolate(
   A=VectorParameter() {parameterName=WaterColor, defaultValue=<rgba 0.5615 0.7812 0.6334 1>}.RGB
   B=@n2
   Alpha=Divide(A=Subtract(A=@n4, B=@n5), B=@n5)
  )
  Multiply(A=@n2, B=VectorParameter() {parameterName=Subsurface Gain, defaultValue=<rgba 1 1 1 1>}.RGB)
  n6 = Add(A=Multiply(A=TextureCoordinate(), B=ScalarParameter() {defaultValue=1, parameterName=Distortion Scale}) {constB=0.25}, B=ParticleRandom())
  n7 = TextureSample(UVs=@n6) {samplerType=SAMPLERTYPE_LinearColor, texture=/WaterAdvanced/Niagara/Textures/Hermite2Noise_CenteredTiling_v2}
  n8 = Add(
   A=TextureCoordinate()
   B=Multiply(A=Subtract(A=@n7.R) {constB=0.7}, B=ScalarParameter() {defaultValue=0.5, parameterName=Distortion Strength}) {constB=0.2}
  )
  n9 = Constant() {r=0.5}
  Multiply(
   A=Multiply(A=SphereMask(A=@n8, B=@n9) {attenuationRadius=0.3, hardnessPercent=50}, B=ScalarParameter() {defaultValue=1, parameterName=Opacity Mult})
   B=@n1.A
  )
  MaterialExpressionTextureCoordinate_0 = TextureCoordinate()
  MaterialExpressionConstant_0 = Constant() {r=1.25}
  n10 = Custom!(TexCoord=@MaterialExpressionTextureCoordinate_0, Radius=@MaterialExpressionConstant_0) {
   code="float X = (TexCoord.x - 0.5) * 2.0;\r\nfloat Y = (TexCoord.y - 0.5) * 2.0;\r\nfloat Z2 = Radius * Radius - X * X - Y * Y;\r\n\r\nif (Z2 < 0)\r\n    return float3(0.0, 0.0, 1.0);\r\nelse\r\n    return float3(X, Y, sqrt(Z2));"
   showCode=true
   inputs [
    {inputName=TexCoord, input {expression=&self:MaterialExpressionTextureCoordinate_0}}
    {inputName=Radius, input {expression=&self:MaterialExpressionConstant_0}}
   ]
  }
  n11 = TextureSample(UVs=@n6) {samplerType=SAMPLERTYPE_Normal, texture=/WaterAdvanced/Niagara/Textures/Hermite2Noise_CenteredTiling_v2_Normals}
  n12 = Constant() {r=-5}
  Normalize(VectorInput=MaterialFunctionCall(
    BaseNormal=@n10
    AdditionalNormal=MaterialFunctionCall(Normal=@n11.RGB, Flatness=@n12) {materialFunction=&FlattenNormal}.Result
   ) {materialFunction=&BlendAngleCorrectedNormals}.Result)
 ]
}
