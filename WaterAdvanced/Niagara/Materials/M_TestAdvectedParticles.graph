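# M_TestAdvectedParticles is a Default Lit test material for Niagara mesh particles in WaterAdvanced, named for testing advected particles.
# Its graph holds a single Particle Color node, with no connection to a material output shown.
# Usage is explicitly enabled for Niagara mesh particles, automatic usage setting in the editor is off, and pixel depth offset uses legacy mode.
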Material /WaterAdvanced/Niagara/Materials/M_TestAdvectedParticles
graph {
 properties {bAutomaticallySetUsageInEditor=false, bUsedWithNiagaraMeshParticles=true, pixelDepthOffsetMode=PDOM_Legacy, usedShadingModels=Default Lit}
 nodes [ParticleColor()]
}
