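# Niagara GPU system for an ocean surface patch on a Grid2D collection, parameterized by wind, gravity, Phillips amplitude, wavenumber limits and choppiness.
# Emitter spawn derives grid resolution and world extent from length and render scales, then sizes the grid and a float16 normal render target exposed as User.OceanNormalRT.
# A scratch module sets up a dynamic grid mesh, gated by RenderDisplacedPatch, alongside sprite and mesh renderers.
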NiagaraSystem /WaterAdvanced/Niagara/Systems/Grid2D_OceanPatch
dependencies {
 usedDataInterfaces [
  /Script/Niagara.NiagaraDataInterfaceGrid2DCollection
  /Script/Niagara.NiagaraDataInterfaceRenderTarget2D
  /Script/Niagara.NiagaraDataInterfaceDynamicMesh
 ]
 usedDynamicInputs []
 usedModules [&SystemState]
}
emitters {OceanSurfaceEmitter {
  simTarget=GPUComputeSim
  emitterSpawnScript [
   /SetVariables(Emitter.gravity=980) {isSetParameters=true}
   /SetVariables(Emitter.windSpeed=55) {isSetParameters=true}
   /SetVariables(Emitter.windAngle=180) {isSetParameters=true}
   /SetVariables() {isSetParameters=true}
   /SetVariables(Emitter.OceanLengthScale=250) {isSetParameters=true}
   /SetVariables(Emitter.TimeScale=1) {isSetParameters=true}
   /SetVariables(Emitter.PhillipsAmplitude=2) {isSetParameters=true}
   /SetVariables(Emitter.Grid2D=DataInterface() {propertyValues {
      ClearBeforeNonIterationStage=false
      NumAttributes=0
      NumCellsMaxAxis=3
      NumCellsX=3
      NumCellsY=3
      OverrideBufferFormat=Float
      PreviewAttribute=None
      SetGridFromMaxAxis=false
      bOverrideFormat=false
      bPreviewGrid=false
      RenderTargetUserParameter {parameter {name=None, typeDefHandle {registeredTypeIndex=4}, varData []}}
      WorldBBoxSize=<100 100>
     }}) {isSetParameters=true}
   /SetVariables(Emitter.ChopHack=5) {isSetParameters=true}
   /SetVariables(Emitter.Slosh=1) {isSetParameters=true}
   /SetVariables(Emitter.RenderSizeScale=1, Emitter.TilingAmount=1) {isSetParameters=true}
   &self:OceanSurfaceEmitter_2.NiagaraScratchPadContainer_44.ScratchModule_0(
    "Max Axis World Extents ":NiagaraFloat=&Multiply_Float(A=$Emitter.OceanLengthScale, B=$Emitter.RenderSizeScale)
    NumCellsX:ENiagaraPowerOfTwoResolution=::NewEnumerator9
    NumCellsY:ENiagaraPowerOfTwoResolution=::NewEnumerator9
   ) {module=ScratchModule_10}
   /NiagaraFluids/Modules/Grid2D/Grid2D_SetResolution(
    Grid:NiagaraDataInterfaceGrid2DCollection=$Emitter.Grid2D
    NumCellsX:NiagaraInt32=$Emitter.NumGridCellsX
    NumCellsY:NiagaraInt32=$Emitter.NumGridCellsY
    WorldGridExtents:Vector2f=$Emitter.WorldGridExtents
   )
   /SetVariables(Emitter.RenderTarget2D=DataInterface() {propertyValues {
      MipMapGeneration=Disabled
      MipMapGenerationType=Linear
      OverrideRenderTargetFilter=TF_Default
      OverrideRenderTargetFormat=RTF_R8
      bInheritUserParameterSettings=false
      bOverrideFormat=false
      bPreviewRenderTarget=false
      RenderTargetUserParameter {parameter {name=None, typeDefHandle {registeredTypeIndex=4}, varData []}}
      Size=<1 1>
     }}) {enabled=false, isSetParameters=true}
   /SetVariables(Emitter.NormalRT=DataInterface() {propertyValues {
      MipMapGeneration=PostSimulate
      MipMapGenerationType=Linear
      OverrideRenderTargetFilter=TF_Default
      OverrideRenderTargetFormat=RTF_RGBA16f
      bInheritUserParameterSettings=false
      bOverrideFormat=true
      bPreviewRenderTarget=false
      RenderTargetUserParameter {parameter {name=User.OceanNormalRT, typeDefHandle {registeredTypeIndex=4}, varData []}}
      Size=<0 0>
     }}) {isSetParameters=true}
   /NiagaraFluids/Modules/Grid2D/Grid2D_InitRT(Grid:NiagaraDataInterfaceGrid2DCollection=$Emitter.Grid2D, SimRT:NiagaraDataInterfaceRenderTarget2D=$Emitter.RenderTarget2D) {enabled=false}
   &Grid2D_InitRT(Grid:NiagaraDataInterfaceGrid2DCollection=$Emitter.Grid2D, SimRT:NiagaraDataInterfaceRenderTarget2D=$Emitter.NormalRT) {module=Grid2D_InitRT001}
   /SetVariables() {isSetParameters=true}
   /SetVariables(Emitter.GridMesh=DataInterface() {propertyValues {
      NumTexCoords=0
      NumVertices=0
      bClearTrianglesPerFrame=false
      bHasColors=false
      bHasTangentBasis=false
      DefaultBounds {isValid=true, max=<100 100 100>, min=<-100 -100 -100>}
      Materials [{
        material=/Engine/EngineMaterials/WorldGridMaterial
        materialUserParamBinding {parameter {name=None, typeDefHandle {registeredTypeIndex=3}, varData []}}
       }]
      Sections []
     }}) {isSetParameters=true}
   &self:OceanSurfaceEmitter_2.NiagaraScratchPadContainer_44.ScratchModule_4(
    Execute:NiagaraBool=$Emitter.RenderDisplacedPatch
    Mesh:NiagaraDataInterfaceDynamicMesh=$Emitter.GridMesh
    NumCellsX:NiagaraInt32=$Emitter.NumGridCellsX
    NumCellsY:NiagaraInt32=$Emitter.NumGridCellsY
   ) {module=ScratchModule004}
   /SetVariables(Emitter.MaxK=50, Emitter.MinK=0.2) {isSetParameters=true}
  ]
  emitterUpdateScript [&EmitterState()]
  rendererClasses [&NiagaraSpriteRendererProperties, &NiagaraMeshRendererProperties]
  renderers [
   {rendererClass=&NiagaraSpriteRendererProperties}
   {rendererClass=&NiagaraMeshRendererProperties, rendererIndex=1}
  ]
 }}
summary {
 systemName=Grid2D_OceanPatch
 emitters [{
   emitterName=OceanSurfaceEmitter
   simTarget=GPUComputeSim
   rendererClasses [&NiagaraSpriteRendererProperties, &NiagaraMeshRendererProperties]
  }]
 userVariables [User.OceanNormalRT {
   type=TextureRenderTarget
   description=""
   defaultValue {struct=NiagaraExt_VariableValue_Object, value {object=null, objectClass=/Script/Engine.TextureRenderTarget}}
  }]
}
