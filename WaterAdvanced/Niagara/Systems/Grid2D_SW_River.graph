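# GPU Niagara system that simulates a shallow-water river on 2D grids, with live or cached modes.
# ShallowWaterRiverEmitter sets up bottom-contour scene captures, mesh colliders, water, velocity, normal and foam grids, render targets and surface rendering.
# SW_River_Secondary_Emitter samples those grids, or baked textures when not simulating live, to drive whitewater sprites.
# Its camera-recycled particles use curvature, vorticity, depth and speed emission criteria, culled around sink boxes.
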NiagaraSystem /WaterAdvanced/Niagara/Systems/Grid2D_SW_River
dependencies {
 usedDataInterfaces [
  /Script/Niagara.NiagaraDataInterfaceGrid2DCollection
  /Script/Niagara.NiagaraDataInterfaceRenderTarget2D
  &NiagaraDataInterfaceParticleRead
  /Script/Niagara.NiagaraDataInterfaceDynamicMesh
  /Script/Niagara.NiagaraDataInterfaceGrid2DCollectionReader
  &NiagaraDataInterfaceDebugDraw
  /Script/Niagara.NiagaraDataInterface2DArrayTexture
  /Script/Niagara.NiagaraDataInterfaceEmitterProperties
  /Script/Niagara.NiagaraDataInterfaceTexture
  /Script/Niagara.NiagaraDataInterfaceVirtualTextureSample
  &NiagaraDataInterfaceCurve
 ]
 usedDynamicInputs []
 usedModules [
  &SystemState
  &self:SystemSpawnScript.SystemScriptSource.SystemScriptGraph.NiagaraNodeAssignment_0.SetVariables_#1
 ]
}
emitters {
 SW_River_Secondary_Emitter {
  simTarget=GPUComputeSim
  emitterSpawnScript [/SetVariables(
    Emitter.NormalGridReader=DataInterface() {propertyValues {
      ClearBeforeNonIterationStage=true
      DIName=NormalGrid
      EmitterName=ShallowWaterRiverEmitter
      NumAttributes=0
      NumCellsMaxAxis=3
      NumCellsX=3
      NumCellsY=3
      OverrideBufferFormat=Float
      PreviewAttribute=None
      SetGridFromMaxAxis=false
      bOverrideFormat=false
      bPreviewGrid=false
      RenderTargetUserParameter {parameter {name=None, typeDefHandle {registeredTypeIndex=4}, varData []}}
      WorldBBoxSize=<100 100>
     }}
    Emitter.VelocityReader=DataInterface() {propertyValues {
      ClearBeforeNonIterationStage=true
      DIName=VelocityGrid
      EmitterName=ShallowWaterRiverEmitter
      NumAttributes=0
      NumCellsMaxAxis=3
      NumCellsX=3
      NumCellsY=3
      OverrideBufferFormat=Float
      PreviewAttribute=None
      SetGridFromMaxAxis=false
      bOverrideFormat=false
      bPreviewGrid=false
      RenderTargetUserParameter {parameter {name=None, typeDefHandle {registeredTypeIndex=4}, varData []}}
      WorldBBoxSize=<100 100>
     }}
    Emitter.WaterReader=DataInterface() {propertyValues {
      ClearBeforeNonIterationStage=true
      DIName=WaterGrid
      EmitterName=ShallowWaterRiverEmitter
      NumAttributes=0
      NumCellsMaxAxis=3
      NumCellsX=3
      NumCellsY=3
      OverrideBufferFormat=Float
      PreviewAttribute=None
      SetGridFromMaxAxis=false
      bOverrideFormat=false
      bPreviewGrid=false
      RenderTargetUserParameter {parameter {name=None, typeDefHandle {registeredTypeIndex=4}, varData []}}
      WorldBBoxSize=<100 100>
     }}
   ) {isSetParameters=true}]
  emitterUpdateScript [
   &EmitterState(Life Cycle Mode=::NewEnumerator1, Loop Behavior=::NewEnumerator1, Loop Duration Mode=::NewEnumerator1, UseLoopDelay=false)
   &SpawnBurst_Instantaneous(Spawn Count=50000, Use Spawn Probability=false, UseLoopCountLimit=false)
   &self:SW_River_Secondary_Emitter.NiagaraScratchPadContainer_8.CellSizeFromReader_0(Extents:Vector2f=$User.WorldGridSize) {module=CellSizeFromReader001}
   /NiagaraFluids/Modules/Grid2D/Grid2D_CreateUnitToWorldTransform(Offset:NiagaraPosition=$Engine.Owner.Position, Rotation:Quat4f=$Engine.Owner.Rotation, WorldGridExtents:Vector2f=$User.WorldGridSize)
   /SetVariables(
    Emitter.LocalToWorld:NiagaraMatrix=$Emitter.Grid2D_CreateUnitToWorldTransform.LocalToWorld
    Emitter.UnitToWorld:NiagaraMatrix=$Emitter.Grid2D_CreateUnitToWorldTransform.UnitToWorld
    Emitter.WorldToLocal:NiagaraMatrix=$Emitter.Grid2D_CreateUnitToWorldTransform.WorldToLocal
    Emitter.WorldToUnit:NiagaraMatrix=$Emitter.Grid2D_CreateUnitToWorldTransform.WorldToUnit
   ) {isSetParameters=true}
   /SetVariables(
    Emitter.TmpPos:NiagaraPosition=$Engine.Owner.Position
    Emitter.Vector:Vector3f=&ConvertPositionToVector(Input Position=$Engine.Owner.Position, Conversion Operation:ENiagara_LWCConvertPosToVec=::NewEnumerator0)
   ) {isSetParameters=true}
   &self:SW_River_Secondary_Emitter.NiagaraScratchPadContainer_8.ScratchModule_2(RenderSecondary:NiagaraBool=$User.RenderSecondary) {module=ScratchModule}
   /NiagaraFluids/Modules/Grid2D/Grid2D_SetBounds(
    EmitterProperties=DataInterface() {propertyValues {EmitterBinding {bindingMode=Self, emitterName=None}}}
    LocalPivot:Vector2f=<0.5 0.5>
    WorldGridExtents:Vector2f=$User.WorldGridSize
   )
  ]
  particleSpawnScript [
   &InitializeParticle(
    Lifetime=0.5
    Sprite Rotation Mode=::NewEnumerator2
    Sprite Size Mode=::NewEnumerator1
    Uniform Sprite Size Max=75
    Uniform Sprite Size Min=30
    UsePositionOffset=false
    Write Material Random=true
   )
   /SetVariables(Particles.MaterialRandom:NiagaraFloat=&RandomRangeFloat(Minimum=0, Maximum=1)) {isSetParameters=true}
   /SetVariables(Particles.MeshIndex=&RandomRangeInt(Minimum=0, Maximum=3)) {isSetParameters=true}
   /SetVariables(Particles.Scale=&RandomRangeVector(Minimum=<1 1 1>, Maximum=<3 3 3>)) {isSetParameters=true}
   &InitialMeshOrientation(Use Rotation Vector=false)
  ]
  particleUpdateScript [
   &ParticleState(Kill Particles When Lifetime Has Elapsed=false, Loop Particles Lifetime=true)
   &self:SW_River_Secondary_Emitter.NiagaraScratchPadContainer_8.RandomizeStartAgeOverLifetime_0() {module=RandomizeStartAgeOverLifetime}
   /Niagara/Modules/Update/Camera/ViewRecycler(
    Maximum FOV Angle=110
    Mesh Orientation Mode:ENiagaraRecycleAttributeMode=::NewEnumerator1
    Mesh Scale Mode:ENiagaraRecycleSpriteSizeMode=::NewEnumerator1
    Near / Far Distance:Vector2f=&RandomRangeVector2D(Minimum=<0 0>, Maximum=<500 10000>)
    Sprite Size Mode:ENiagaraRecycleSpriteSizeMode=::NewEnumerator1
    Velocity Mode:ENiagaraRecycleAttributeMode=::NewEnumerator2
   )
   &GravityForce()
   &SpriteRotationRate(Rotation Rate=&RandomRangeFloat(Minimum=-180, Maximum=180)) {module=SpriteRotationRate001}
   &UpdateMeshOrientation(Rotation Coordinate Space=::NewEnumerator1, Rotation Rate=&RandomRangeFloat(Minimum=-0.5, Maximum=0.5)) {module=UpdateMeshOrientation001}
   &self:SW_River_Secondary_Emitter.NiagaraScratchPadContainer_8.Sample_Water_0(
    BakedHeightGrid=DataInterface() {propertyValues {Texture=null, TextureUserParameter {parameter {name=User.BakedHeightTexture, typeDefHandle {registeredTypeIndex=2}, varData []}}}}
    BakedHeightGridVT=DataInterface() {propertyValues {Texture=null, TextureUserParameter {parameter {name=User.BakedHeightTexture, typeDefHandle {registeredTypeIndex=2}, varData []}}}}
    BakedNormalGrid=DataInterface() {propertyValues {Texture=null, TextureUserParameter {parameter {name=User.BakedWaterSurfaceNormalTexture, typeDefHandle {registeredTypeIndex=2}, varData []}}}}
    BakedNormalGridVT=DataInterface() {propertyValues {Texture=null, TextureUserParameter {parameter {name=User.BakedWaterSurfaceNormalTexture, typeDefHandle {registeredTypeIndex=2}, varData []}}}}
    BakedSimGrid=DataInterface() {propertyValues {Texture=null, TextureUserParameter {parameter {name=User.BakedSimTexture, typeDefHandle {registeredTypeIndex=2}, varData []}}}}
    BakedSimGridVT=DataInterface() {propertyValues {Texture=null, TextureUserParameter {parameter {name=User.BakedSimTexture, typeDefHandle {registeredTypeIndex=2}, varData []}}}}
    BakedWaterDepthMax:NiagaraFloat=$User.BakedWaterDepthMax
    BakedWaterMaxSpeed:NiagaraFloat=$User.BakedWaterVelocityMaxSpeed
    LocalToWorldTranslation:Vector3f=$Emitter.Grid2D_CreateUnitToWorldTransform.LocalToWorldTranslation
    NormalReader:NiagaraDataInterfaceGrid2DCollectionReader=$Emitter.NormalGridReader
    SampleFromGrid:NiagaraBool=$System.RunLiveSim
    SampleFromVT=true
    SimDt:NiagaraFloat=$Engine.DeltaTime
    SimRes:Vector2f=$User.SimRes
    VelocityReader:NiagaraDataInterfaceGrid2DCollectionReader=$Emitter.VelocityReader
    WaterReader:NiagaraDataInterfaceGrid2DCollectionReader=$Emitter.WaterReader
    WorldToUnit:NiagaraMatrix=$Emitter.WorldToUnit
    dx:NiagaraFloat=$Emitter.dx
   ) {module=Sample_Water}
   &self:SW_River_Secondary_Emitter.NiagaraScratchPadContainer_8.Emission_Criterias_0(
    BoxAngleArray:NiagaraDataInterfaceArrayFloat=$User.SinkAngleArray
    BoxPositionArray:NiagaraDataInterfaceArrayPosition=$User.SinkPosArray
    BoxSizeArray:NiagaraDataInterfaceArrayFloat3=$User.SinkSizeArray
    BoxSizeMult:Vector3f=<2 2 1>
    CullFromBoxes=true
    Curl:NiagaraFloat=$Output.Sample_Water.Curl
    Curvature:NiagaraFloat=$Output.Sample_Water.Curvature
    CurvatureEmissionMult=5
    CurvatureEmissionRange:Vector2f=<0.001 0.05>
    DepthDifference:NiagaraFloat=$Output.Sample_Water.DepthDifference
    EmissionMult=1
    Min EmitAmount=0.1
    NormalVelocityMagnitude=100
    RecycleValid:NiagaraBool=$StackContext.RecycleValid
    RestrictEmissionDepthRange:Vector2f=<150 500>
    RestrictEmissionSpeedRange:Vector2f=<50 200>
    SimDt:NiagaraFloat=$Engine.DeltaTime
    Velocity:Vector2f=$Output.Sample_Water.Velocity
    VorticityEmissionRange:Vector2f=<1 3>
    WaterDepth:NiagaraFloat=$Output.Sample_Water.WaterDepth
    WaterHeight:NiagaraFloat=$Output.Sample_Water.WaterHeight
    WaterNormal:Vector3f=$Output.Sample_Water.WaterNormal
    WorldToLocal:NiagaraMatrix=$Emitter.WorldToLocal
    dx:NiagaraFloat=&MakeFloatFromVector2d(Vector 2D=$Emitter.CellSize)
   ) {module=Emission_Criterias}
   &self:SW_River_Secondary_Emitter.NiagaraScratchPadContainer_8.WhiteWaterVelocityAndForce_0(
    Bubble Buoyancy:Vector3f=<0 0 100>
    Position:NiagaraPosition=$Particles.Position
    Velocity:Vector2f=$Output.Sample_Water.Velocity
    WaterHeight:NiagaraFloat=$Output.Sample_Water.WaterHeight
    WaterNormal:Vector3f=$Output.Sample_Water.WaterNormal
   ) {module=WhiteWaterVelocityAndForce}
   &SolveForcesAndVelocity(Clamp Velocity=false, Limit Acceleration=false)
   &ScaleMeshSize(Scale Factor=&VectorFromFloat(Value=&FloatFromCurve(FloatCurve=DataInterface() {propertyValues {Curve {
         defaultValue=1
         keys [
          {arriveTangent=4.1785, interpMode=RCIM_Cubic, leaveTangent=4.1785, tangentMode=RCTM_User, time=0, value=0}
          {arriveTangent=3.3154, interpMode=RCIM_Cubic, leaveTangent=3.3154, tangentMode=RCTM_User, time=0.1106, value=0.4824}
          {arriveTangent=0.0437, interpMode=RCIM_Cubic, leaveTangent=0.0437, tangentMode=RCTM_User, time=0.5, value=1}
          {arriveTangent=-4.1729, interpMode=RCIM_Cubic, leaveTangent=-4.1729, tangentMode=RCTM_User, time=1, value=0}
         ]
        }}}))) {module=ScaleMeshSize001}
   &ScaleSpriteSize(Uniform Curve Sprite Scale=DataInterface() {propertyValues {Curve {
       defaultValue=1
       keys [
        {arriveTangent=4.1785, interpMode=RCIM_Cubic, leaveTangent=4.1785, tangentMode=RCTM_User, time=0, value=0}
        {arriveTangent=3.3154, interpMode=RCIM_Cubic, leaveTangent=3.3154, tangentMode=RCTM_User, time=0.1106, value=0.4824}
        {arriveTangent=0.0437, interpMode=RCIM_Cubic, leaveTangent=0.0437, tangentMode=RCTM_User, time=0.5, value=1}
        {arriveTangent=-4.1729, interpMode=RCIM_Cubic, leaveTangent=-4.1729, tangentMode=RCTM_User, time=1, value=0}
       ]
      }}}) {module=ScaleSpriteSize002}
   &ScaleColor(
    Scale Alpha=&FloatFromCurve(FloatCurve=DataInterface() {propertyValues {Curve {keys [
         {interpMode=RCIM_Cubic, tangentMode=RCTM_User, time=0, value=0}
         <at 0.1 1>
         {arriveTangent=-1.1111, interpMode=RCIM_Cubic, leaveTangent=-1.1111, time=0.8, value=1}
         {tangentMode=RCTM_User, time=1, value=0}
        ]}}})
    ScaleA=true
    ScaleRGB=false
   ) {module=ScaleColor001}
  ]
  rendererClasses [&NiagaraSpriteRendererProperties]
  renderers [
   {rendererClass=&NiagaraSpriteRendererProperties}
   {rendererClass=&NiagaraSpriteRendererProperties, rendererIndex=1}
  ]
 }
 ShallowWaterRiverEmitter {
  simTarget=GPUComputeSim
  emitterSpawnScript [
   /SetVariables(
    Emitter.GridCollection=DataInterface() {propertyValues {
      ClearBeforeNonIterationStage=true
      NumAttributes=0
      NumCellsMaxAxis=3
      NumCellsX=128
      NumCellsY=128
      OverrideBufferFormat=Float
      PreviewAttribute=None
      SetGridFromMaxAxis=false
      bOverrideFormat=false
      bPreviewGrid=false
      RenderTargetUserParameter {parameter {name=None, typeDefHandle {registeredTypeIndex=4}, varData []}}
      WorldBBoxSize=<1000 1000>
     }}
    Emitter.NormalsRT=DataInterface() {propertyValues {
      MipMapGeneration=Disabled
      MipMapGenerationType=Linear
      OverrideRenderTargetFilter=TF_Default
      OverrideRenderTargetFormat=RTF_RGBA16f
      bInheritUserParameterSettings=false
      bOverrideFormat=true
      bPreviewRenderTarget=false
      RenderTargetUserParameter {parameter {name=User.NormalRT, typeDefHandle {registeredTypeIndex=4}, varData []}}
      Size=<1 1>
     }}
    Emitter.SimGridRT=DataInterface() {propertyValues {
      MipMapGeneration=PostSimulate
      MipMapGenerationType=Linear
      OverrideRenderTargetFilter=TF_Default
      OverrideRenderTargetFormat=RTF_RGBA16f
      bInheritUserParameterSettings=false
      bOverrideFormat=true
      bPreviewRenderTarget=false
      RenderTargetUserParameter {parameter {name=User.SimGridRT, typeDefHandle {registeredTypeIndex=4}, varData []}}
      Size=<1 1>
     }}
    Emitter.VelocityGrid=DataInterface() {propertyValues {
      ClearBeforeNonIterationStage=true
      NumAttributes=0
      NumCellsMaxAxis=3
      NumCellsX=3
      NumCellsY=3
      OverrideBufferFormat=Float
      PreviewAttribute=None
      SetGridFromMaxAxis=false
      bOverrideFormat=true
      bPreviewGrid=false
      RenderTargetUserParameter {parameter {name=None, typeDefHandle {registeredTypeIndex=4}, varData []}}
      WorldBBoxSize=<100 100>
     }}
    Emitter.WaterGrid=DataInterface() {propertyValues {
      ClearBeforeNonIterationStage=true
      NumAttributes=0
      NumCellsMaxAxis=3
      NumCellsX=3
      NumCellsY=3
      OverrideBufferFormat=Float
      PreviewAttribute=None
      SetGridFromMaxAxis=false
      bOverrideFormat=true
      bPreviewGrid=false
      RenderTargetUserParameter {parameter {name=None, typeDefHandle {registeredTypeIndex=4}, varData []}}
      WorldBBoxSize=<100 100>
     }}
   ) {isSetParameters=true}
   /NiagaraFluids/Modules/Grid2D/Grid2D_SetResolution(
    Grid:NiagaraDataInterfaceGrid2DCollection=$Emitter.GridCollection
    NumCellsMaxAxis:NiagaraInt32=$User.ResolutionMaxAxis
    Resolution Method:ENiagaraGrid2DResolution=::NewEnumerator1
    WorldGridExtents:Vector2f=$User.WorldGridSize
   )
   &Grid2D_SetResolution(
    Grid:NiagaraDataInterfaceGrid2DCollection=$Emitter.VelocityGrid
    OtherGrid:NiagaraDataInterfaceGrid2DCollection=$Emitter.GridCollection
    Resolution Method:ENiagaraGrid2DResolution=::NewEnumerator3
    WorldGridExtents:Vector2f=$Emitter.Grid2D_SetResolution.WorldGridExtents
   ) {module=Grid2D_SetResolution001}
   &Grid2D_SetResolution(
    Grid:NiagaraDataInterfaceGrid2DCollection=$Emitter.WaterGrid
    OtherGrid:NiagaraDataInterfaceGrid2DCollection=$Emitter.GridCollection
    Resolution Method:ENiagaraGrid2DResolution=::NewEnumerator3
    WorldGridExtents:Vector2f=$Emitter.Grid2D_SetResolution.WorldGridExtents
   ) {module=Grid2D_SetResolution002}
   &Grid2D_SetResolution(
    Grid:NiagaraDataInterfaceGrid2DCollection=$Emitter.NormalGrid
    OtherGrid:NiagaraDataInterfaceGrid2DCollection=$Emitter.GridCollection
    Resolution Method:ENiagaraGrid2DResolution=::NewEnumerator3
    WorldGridExtents:Vector2f=$Emitter.Grid2D_SetResolution.WorldGridExtents
   ) {module=Grid2D_SetResolution003}
   /SetVariables(Emitter.ParticleCollisionReader=DataInterface() {propertyValues {EmitterBinding {bindingMode=Self, emitterName=ShallowWaterRiverEmitter}}}) {enabled=false, isSetParameters=true}
   /SetVariables(
    Emitter.DistributeOverlapVolume=true
    Emitter.LocalSpaceParticles:NiagaraBool=&self:ShallowWaterRiverEmitter_1.NiagaraScratchPadContainer_0.ScratchDynamicInput_0(Reader:NiagaraDataInterfaceParticleRead=$Emitter.ParticleCollisionReader)
   ) {enabled=false, isSetParameters=true}
   /NiagaraFluids/Modules/Grid2D/Grid2D_InitRT(Grid:NiagaraDataInterfaceGrid2DCollection=$Emitter.GridCollection, SimRT:NiagaraDataInterfaceRenderTarget2D=$Emitter.NormalsRT)
   &Grid2D_InitRT(Grid:NiagaraDataInterfaceGrid2DCollection=$Emitter.GridCollection, SimRT:NiagaraDataInterfaceRenderTarget2D=$Emitter.SimGridRT) {module=Grid2D_InitRT001}
   &self:ShallowWaterRiverEmitter_1.NiagaraScratchPadContainer_0.ScratchModule_03_0(
    BottomDepth:NiagaraFloat=$User.BottomDepth
    Debug Render:NiagaraBool=&Boolean_OrOperation(Bool1=$User.DebugRenderBottomContour, Bool2=$User.DebugRenderFoam)
    RenderWaterSurface:NiagaraBool=$User.RenderWaterSurface
   ) {module=ScratchModule_03}
   /SetVariables(
    Emitter.GridMesh=DataInterface() {propertyValues {
      NumTexCoords=0
      NumVertices=0
      bClearTrianglesPerFrame=false
      bHasColors=false
      bHasTangentBasis=false
      DefaultBounds {isValid=true, max=<100 100 100>, min=<-100 -100 -100>}
      Materials [{
        material=/Engine/EngineMaterials/WorldGridMaterial
        materialUserParamBinding {parameter {name=None, typeDefHandle {registeredTypeIndex=3}, varData []}}
       }]
      Sections []
     }}
    Emitter.MeshResMult=1
   ) {isSetParameters=true}
   &self:ShallowWaterRiverEmitter_1.NiagaraScratchPadContainer_0.ScratchModule_04_0(
    Execute:NiagaraBool=$User.RenderWaterSurface
    Mesh:NiagaraDataInterfaceDynamicMesh=$Emitter.GridMesh
    NumCellsX:NiagaraInt32=&MakeIntFromFloat(Float=&Multiply_Float(A=$Emitter.MeshResMult, B=&MakeFloatFromInt(Int=$Emitter.Grid2D_SetResolution.NumCellsX)))
    NumCellsY:NiagaraInt32=&MakeIntFromFloat(Float=&Multiply_Float(A=$Emitter.MeshResMult, B=&MakeFloatFromInt(Int=$Emitter.Grid2D_SetResolution.NumCellsY)))
   ) {module=ScratchModule_04}
   /SetVariables(Emitter.FoamRT=DataInterface() {propertyValues {
      MipMapGeneration=Disabled
      MipMapGenerationType=Linear
      OverrideRenderTargetFilter=TF_Default
      OverrideRenderTargetFormat=RTF_R16f
      bInheritUserParameterSettings=false
      bOverrideFormat=true
      bPreviewRenderTarget=false
      RenderTargetUserParameter {parameter {name=User.FoamRT, typeDefHandle {registeredTypeIndex=4}, varData []}}
      Size=<0 0>
     }}) {isSetParameters=true}
   &Grid2D_SetResolution(
    Grid:NiagaraDataInterfaceGrid2DCollection=$Emitter.FoamGrid
    OtherGrid:NiagaraDataInterfaceGrid2DCollection=$Emitter.GridCollection
    Resolution Method:ENiagaraGrid2DResolution=::NewEnumerator3
    WorldGridExtents:Vector2f=$Emitter.Grid2D_SetResolution.WorldGridExtents
   ) {module=Grid2D_SetResolution005}
   &Grid2D_InitRT(Grid:NiagaraDataInterfaceGrid2DCollection=$Emitter.FoamGrid, SimRT:NiagaraDataInterfaceRenderTarget2D=$Emitter.FoamRT) {module=Grid2D_InitRT003}
   /SetVariables(
    Emitter.NormalReader=DataInterface() {propertyValues {
      ClearBeforeNonIterationStage=true
      DIName=NormalGrid
      EmitterName=ShallowWaterRiverEmitter
      NumAttributes=0
      NumCellsMaxAxis=3
      NumCellsX=3
      NumCellsY=3
      OverrideBufferFormat=Float
      PreviewAttribute=None
      SetGridFromMaxAxis=false
      bOverrideFormat=false
      bPreviewGrid=false
      RenderTargetUserParameter {parameter {name=None, typeDefHandle {registeredTypeIndex=4}, varData []}}
      WorldBBoxSize=<100 100>
     }}
    Emitter.VelocityReader=DataInterface() {propertyValues {
      ClearBeforeNonIterationStage=true
      DIName=VelocityGrid
      EmitterName=ShallowWaterRiverEmitter
      NumAttributes=0
      NumCellsMaxAxis=3
      NumCellsX=3
      NumCellsY=3
      OverrideBufferFormat=Float
      PreviewAttribute=None
      SetGridFromMaxAxis=false
      bOverrideFormat=false
      bPreviewGrid=false
      RenderTargetUserParameter {parameter {name=None, typeDefHandle {registeredTypeIndex=4}, varData []}}
      WorldBBoxSize=<100 100>
     }}
    Emitter.WaterReader=DataInterface() {propertyValues {
      ClearBeforeNonIterationStage=true
      DIName=WaterGrid
      EmitterName=ShallowWaterRiverEmitter
      NumAttributes=0
      NumCellsMaxAxis=3
      NumCellsX=3
      NumCellsY=3
      OverrideBufferFormat=Float
      PreviewAttribute=None
      SetGridFromMaxAxis=false
      bOverrideFormat=false
      bPreviewGrid=false
      RenderTargetUserParameter {parameter {name=None, typeDefHandle {registeredTypeIndex=4}, varData []}}
      WorldBBoxSize=<100 100>
     }}
   ) {isSetParameters=true}
   /SetVariables(Emitter.FoamDx:NiagaraFloat=&MakeFloatFromVector2d(Vector 2D=$Emitter.Grid2D_SetResolution005.WorldCellSize)) {isSetParameters=true}
   &self:ShallowWaterRiverEmitter_1.NiagaraScratchPadContainer_0.SetCacheState_0(ReadCachedSim:NiagaraBool=$User.ReadCachedSim) {module=SetCacheState}
   /SetVariables(Emitter.NumSubsteps:NiagaraInt32=$User.NumSteps) {isSetParameters=true}
   /NiagaraFluids/Modules/Grid2D/ShallowWater/Grid2D_Init_BottomCapture(CaptureBottomContour:NiagaraDataInterfaceSceneCapture2D=$User.BottomCapture, CaptureOffset:NiagaraFloat=$User.CaptureOffset)
   &Grid2D_Init_BottomCapture(
    CaptureBottomContour:NiagaraDataInterfaceSceneCapture2D=$User.BottomCaptureUnder
    CaptureOffset:NiagaraFloat=$User.CaptureOffsetUnder
    TopDown=false
   ) {module=Grid2D_Init_BottomCapture003}
   &Grid2D_Init_BottomCapture(CaptureBottomContour:NiagaraDataInterfaceSceneCapture2D=$User.DilatedBottomCapture, CaptureOffset:NiagaraFloat=$User.DilatedCaptureOffset) {module=Grid2D_Init_BottomCapture001}
   &Grid2D_Init_BottomCapture(
    CaptureBottomContour:NiagaraDataInterfaceSceneCapture2D=$User.DilatedBottomCaptureUnder
    CaptureOffset:NiagaraFloat=$User.DilatedCaptureOffsetUnder
    TopDown=false
   ) {module=Grid2D_Init_BottomCapture004}
   &Grid2D_Init_BottomCapture(
    CaptureBottomContour:NiagaraDataInterfaceSceneCapture2D=$User.LandscapeBottomCapture
    CaptureOffset:NiagaraFloat=$User.LandscapeCaptureOffset
   ) {module=Grid2D_Init_BottomCapture002}
   /SetVariables(Emitter.DebugDraw=DataInterface() {propertyValues {OverrideMaxLineInstances=4294967295}}) {isSetParameters=true}
   /SetVariables(Emitter.WaterInfoSample=DataInterface() {propertyValues {Texture=null, TextureUserParameter {parameter {name=User.WaterInfoTexture, typeDefHandle {registeredTypeIndex=2}, varData []}}}}) {isSetParameters=true}
   /SetVariables(Emitter.GridSize:Vector3f=&MakeVector(
     X=&MakeFloatFromVector2d(Vector 2D=$Emitter.Grid2D_SetResolution.WorldGridExtents)
     Y=&MakeFloatFromVector2d(Vector 2D=$Emitter.Grid2D_SetResolution.WorldGridExtents, Channel=::NewEnumerator1)
     Z=1
    )) {isSetParameters=true}
   /SetVariables(Emitter.NormalSmoothingRadius=1) {isSetParameters=true}
   /SetVariables(Emitter.GridsAllocated=true) {isSetParameters=true}
   /SetVariables(Emitter.SystemPos:NiagaraPosition=$Engine.Owner.Position) {isSetParameters=true}
   /SetVariables(Emitter.CapturePosition:NiagaraPosition=$Output.Grid2D_Init_BottomCapture.CapturePosition) {isSetParameters=true}
   /SetVariables(Emitter.CaptureUnderPosition:NiagaraPosition=$Output.Grid2D_Init_BottomCapture003.CapturePosition) {isSetParameters=true}
   /SetVariables(Emitter.DilatedCapturePosition:NiagaraPosition=$Output.Grid2D_Init_BottomCapture001.CapturePosition) {isSetParameters=true}
   /SetVariables(Emitter.DilatedCaptureUnderPosition:NiagaraPosition=$Output.Grid2D_Init_BottomCapture004.CapturePosition) {isSetParameters=true}
   /SetVariables(Emitter.LandscapeCapturePosition:NiagaraPosition=$Output.Grid2D_Init_BottomCapture002.CapturePosition) {isSetParameters=true}
   /SetVariables(
    Emitter.RenderCached:NiagaraBool=&Boolean_AndOperation(Bool1=$Emitter.BeautyRender, Bool2=$Emitter.ReadCachedSim)
    Emitter.RenderLive:NiagaraBool=&Boolean_AndOperation(Bool1=$Emitter.BeautyRender, Bool2=$Emitter.RunLiveSim)
   ) {isSetParameters=true}
   /SetVariables(Emitter.SimResVector:Vector3f=&MakeVector(
     X=&MakeFloatFromInt(Int=$Emitter.NumCellsX)
     Y=&MakeFloatFromInt(Int=$Emitter.NumCellsY)
     Z=0
    )) {isSetParameters=true}
  ]
  emitterUpdateScript [
   &EmitterState(Life Cycle Mode=::NewEnumerator1, Loop Behavior=::NewEnumerator1, Loop Duration Mode=::NewEnumerator1, UseLoopDelay=false) {module=EmitterLifeCycle}
   &Grid2D_CreateUnitToWorldTransform(
    Offset:NiagaraPosition=$Engine.Owner.Position
    Rotation:Quat4f=$Engine.Owner.Rotation
    WorldGridExtents:Vector2f=$Emitter.Grid2D_SetResolution.WorldGridExtents
   )
   /NiagaraFluids/Modules/Grid2D/ShallowWater/Grid2D_SW_UpdateSimAttributes(
    DeltaTime:NiagaraFloat=/Niagara/DynamicInputs/Math/Min_Float(
     A:NiagaraFloat=&Divide_Float(A=0.0167, B=&MakeFloatFromInt(Int=$Emitter.NumSubsteps))
     B:NiagaraFloat=0.0111
    )
    DtMult:NiagaraFloat=$User.SimSpeed
    WorldCellSize:Vector2f=$Emitter.Grid2D_SetResolution.WorldCellSize
    WorldGridExtents:Vector2f=$Emitter.Grid2D_SetResolution.WorldGridExtents
   )
   &Grid2D_SetBounds(
    EmitterProperties=DataInterface() {propertyValues {EmitterBinding {bindingMode=Self, emitterName=None}}}
    LocalPivot:Vector2f=<0.5 0.5>
    WorldGridExtents:Vector2f=$Emitter.Grid2D_SetResolution.WorldGridExtents
   )
   /NiagaraFluids/Modules/Grid3D/Grid3D_RigidMesh_FindColliders(
    CollisionQuery:NiagaraDataInterfaceRigidMeshCollisionQuery=$User.Collide_Meshes
    GridCenterPosition:NiagaraPosition=$Engine.Owner.Position
    GridRotation:Quat4f=$Engine.Owner.Rotation
    SkipOverlap:NiagaraBool=$User.ReadCachedSim
    WorldGridExtents:Vector3f=&MakeVector(
     X=&MakeFloatFromVector2d(Vector 2D=$Emitter.Grid2D_SetResolution.WorldGridExtents)
     Y=&MakeFloatFromVector2d(Vector 2D=$Emitter.Grid2D_SetResolution.WorldGridExtents, Channel=::NewEnumerator1)
     Z=100000
    )
   )
   /NiagaraFluids/Modules/Grid3D/Grid3D_DrawBoundsOverlay(
    Draw Bounds=false
    Num Cells X:NiagaraInt32=$Emitter.Grid2D_SetResolution.NumCellsX
    Num Cells Y:NiagaraInt32=$Emitter.Grid2D_SetResolution.NumCellsY
    Num Cells Z=1
    UnitToWorld:NiagaraMatrix=$Emitter.Grid2D_CreateUnitToWorldTransform.UnitToWorld
   ) {enabled=false}
   /SetVariables(Emitter.SuccessfulCacheRead:NiagaraBool=$Emitter.EmitterUpdateRun) {isSetParameters=true}
   /SetVariables(Emitter.PerformReadCache:NiagaraBool=&Boolean_AndOperation(Bool1=$Emitter.ReadCachedSim, Bool2=&Boolean_NotOperation(Bool=$Emitter.SuccessfulCacheRead))) {isSetParameters=true}
   &self:ShallowWaterRiverEmitter_1.NiagaraScratchPadContainer_0.ScratchModule_01_3(Deallocate NormalGrid=true, Deallocate VelocityGrid=true, Deallocate WaterGrid=true) {module=ScratchModule_01004}
   /SetVariables(Emitter.EmitterUpdateRun=true) {isSetParameters=true}
  ]
  rendererClasses [&NiagaraMeshRendererProperties]
  renderers [
   {rendererClass=&NiagaraMeshRendererProperties, rendererIndex=0}
   {rendererClass=&NiagaraMeshRendererProperties, rendererIndex=1}
   {rendererClass=&NiagaraMeshRendererProperties, rendererIndex=2}
   {rendererClass=&NiagaraMeshRendererProperties, rendererIndex=3}
   {rendererClass=&NiagaraMeshRendererProperties, rendererIndex=4}
  ]
 }
}
summary {
 systemName=Grid2D_SW_River
 emitters [
  {emitterName=ShallowWaterRiverEmitter, simTarget=GPUComputeSim, rendererClasses [&NiagaraMeshRendererProperties]}
  {emitterName=SW_River_Secondary_Emitter, simTarget=GPUComputeSim, rendererClasses [&NiagaraSpriteRendererProperties]}
 ]
 userVariables [
  User.BakedFoamTexture {type=Texture, description="", defaultValue {struct=NiagaraExt_VariableValue_Object, value {object=null, objectClass=/Script/Engine.Texture}}}
  User.BakedHeightTexture {type=Texture, description="", defaultValue {struct=NiagaraExt_VariableValue_Object, value {object=null, objectClass=&Texture}}}
  User.BakedSimTexture {type=Texture, description="", defaultValue {struct=NiagaraExt_VariableValue_Object, value {object=null, objectClass=&Texture}}}
  User.BakedWaterDepthMax {type=NiagaraFloat, defaultValue=0, description=""}
  User.BakedWaterSurfaceNormalTexture {type=Texture, description="", defaultValue {struct=NiagaraExt_VariableValue_Object, value {object=null, objectClass=&Texture}}}
  User.BakedWaterVelocityMaxSpeed {type=NiagaraFloat, defaultValue=0, description=""}
  User.BottomCapture {
   type=NiagaraDataInterfaceSceneCapture2D
   description=""
   defaultValue {
    struct=NiagaraExt_VariableValue_DataInterface
    value {
     dataInterface=&self:NiagaraDataInterfaceSceneCapture2D_0
     dataInterfaceClass=/Script/Niagara.NiagaraDataInterfaceSceneCapture2D
    }
   }
  }
  User.BottomCaptureUnder {
   type=NiagaraDataInterfaceSceneCapture2D
   description=""
   defaultValue {
    struct=NiagaraExt_VariableValue_DataInterface
    value {
     dataInterface=&self:NiagaraDataInterfaceSceneCapture2D_4
     dataInterfaceClass=&NiagaraDataInterfaceSceneCapture2D
    }
   }
  }
  User.BottomContourBlurRadius {type=NiagaraInt32}
  User.BottomContourCollisionDilation {type=NiagaraFloat, defaultValue=0, description=""}
  User.BottomDepth {type=NiagaraFloat, defaultValue=-10000, description=""}
  User.CaptureOffset {type=NiagaraFloat, defaultValue=0, description=""}
  User.CaptureOffsetUnder {type=NiagaraFloat, defaultValue=0, description=""}
  User.Collide_Meshes {
   type=NiagaraDataInterfaceRigidMeshCollisionQuery
   description=""
   defaultValue {
    struct=NiagaraExt_VariableValue_DataInterface
    value {
     dataInterface=&self:NiagaraDataInterfaceRigidMeshCollisionQuery_0
     dataInterfaceClass=/Script/Niagara.NiagaraDataInterfaceRigidMeshCollisionQuery
    }
   }
  }
  User.DebugRenderBottomContour {type=NiagaraBool}
  User.DebugRenderFoam {type=NiagaraBool}
  User.DilatedBottomCapture {
   type=NiagaraDataInterfaceSceneCapture2D
   description=""
   defaultValue {
    struct=NiagaraExt_VariableValue_DataInterface
    value {
     dataInterface=&self:NiagaraDataInterfaceSceneCapture2D_2
     dataInterfaceClass=&NiagaraDataInterfaceSceneCapture2D
    }
   }
  }
  User.DilatedBottomCaptureUnder {
   type=NiagaraDataInterfaceSceneCapture2D
   description=""
   defaultValue {
    struct=NiagaraExt_VariableValue_DataInterface
    value {
     dataInterface=&self:NiagaraDataInterfaceSceneCapture2D_3
     dataInterfaceClass=&NiagaraDataInterfaceSceneCapture2D
    }
   }
  }
  User.DilatedCaptureOffset {type=NiagaraFloat, defaultValue=0, description=""}
  User.DilatedCaptureOffsetUnder {type=NiagaraFloat, defaultValue=0, description=""}
  User.ExtrapolationHalfWidth {type=NiagaraInt32, defaultValue=5}
  User.FoamDissipationRate {type=NiagaraFloat, defaultValue=2.5, description=""}
  User.FoamRT {
   type=TextureRenderTarget
   description=""
   defaultValue {struct=NiagaraExt_VariableValue_Object, value {object=null, objectClass=/Script/Engine.TextureRenderTarget}}
  }
  User.Landscape {
   type=NiagaraDataInterfaceLandscape
   description=""
   defaultValue {
    struct=NiagaraExt_VariableValue_DataInterface
    value {
     dataInterface=&self:NiagaraDataInterfaceLandscape_0
     dataInterfaceClass=&NiagaraDataInterfaceLandscape
    }
   }
  }
  User.LandscapeBottomCapture {
   type=NiagaraDataInterfaceSceneCapture2D
   description=""
   defaultValue {
    struct=NiagaraExt_VariableValue_DataInterface
    value {
     dataInterface=&self:NiagaraDataInterfaceSceneCapture2D_1
     dataInterfaceClass=&NiagaraDataInterfaceSceneCapture2D
    }
   }
  }
  User.LandscapeCaptureOffset {type=NiagaraFloat, defaultValue=0, description=""}
  User.MatchSpline {type=NiagaraBool, defaultValue=true}
  User.NormalDetailRT {
   type=TextureRenderTarget
   description=""
   defaultValue {struct=NiagaraExt_VariableValue_Object, value {object=null, objectClass=&TextureRenderTarget}}
  }
  User.NormalRT {
   type=TextureRenderTarget
   description=""
   defaultValue {struct=NiagaraExt_VariableValue_Object, value {object=null, objectClass=&TextureRenderTarget}}
  }
  User.NumSteps {type=NiagaraInt32, defaultValue=1}
  User.OceanNormalRT {
   type=TextureRenderTarget
   description=""
   defaultValue {struct=NiagaraExt_VariableValue_Object, value {object=null, objectClass=&TextureRenderTarget}}
  }
  User.Position {type=NiagaraPosition, description="", defaultValue {}}
  User.ReadCachedSim {type=NiagaraBool}
  User.RemoveOutsideSplineAmount {type=NiagaraFloat, defaultValue=1000, description=""}
  User.RenderSecondary {type=NiagaraBool}
  User.RenderWaterSurface {type=NiagaraBool, defaultValue=true}
  User.ResolutionMaxAxis {type=NiagaraInt32, defaultValue=512}
  User.SimGridRT {
   type=TextureRenderTarget
   description=""
   defaultValue {struct=NiagaraExt_VariableValue_Object, value {object=null, objectClass=&TextureRenderTarget}}
  }
  User.SimRes {type=Vector2f, description="", defaultValue {struct=Vector2f, value=<0 0>}}
  User.SimSpeed {type=NiagaraFloat, defaultValue=1, description=""}
  User.SinkAngle {type=NiagaraFloat, defaultValue=0, description=""}
  User.SinkAngleArray {
   type=NiagaraDataInterfaceArrayFloat
   defaultValue {
    struct=NiagaraExt_VariableValue_DataInterface
    value {
     dataInterface=&self:NiagaraDataInterfaceArrayFloat_1
     dataInterfaceClass=&NiagaraDataInterfaceArrayFloat
    }
   }
  }
  User.SinkPos {type=NiagaraPosition, description="", defaultValue {}}
  User.SinkPosArray {
   type=NiagaraDataInterfaceArrayPosition
   description=""
   defaultValue {
    struct=NiagaraExt_VariableValue_DataInterface
    value {
     dataInterface=&self:NiagaraDataInterfaceArrayPosition_1
     dataInterfaceClass=&NiagaraDataInterfaceArrayPosition
    }
   }
  }
  User.SinkSize {type=Vector3f, description="", defaultValue {struct=Vector3f, value=<0 0 0>}}
  User.SinkSizeArray {
   type=NiagaraDataInterfaceArrayFloat3
   description=""
   defaultValue {
    struct=NiagaraExt_VariableValue_DataInterface
    value {
     dataInterface=&self:NiagaraDataInterfaceArrayFloat3_1
     dataInterfaceClass=&NiagaraDataInterfaceArrayFloat3
    }
   }
  }
  User.SmoothingHeightCutoff {type=NiagaraFloat, defaultValue=1500, description=""}
  User.SourceAngle {type=NiagaraFloat, defaultValue=0, description=""}
  User.SourceAngleArray {
   type=NiagaraDataInterfaceArrayFloat
   defaultValue {
    struct=NiagaraExt_VariableValue_DataInterface
    value {
     dataInterface=&self:NiagaraDataInterfaceArrayFloat_0
     dataInterfaceClass=&NiagaraDataInterfaceArrayFloat
    }
   }
  }
  User.SourcePos {type=NiagaraPosition, description="", defaultValue {}}
  User.SourcePosArray {
   type=NiagaraDataInterfaceArrayPosition
   description=""
   defaultValue {
    struct=NiagaraExt_VariableValue_DataInterface
    value {
     dataInterface=&self:NiagaraDataInterfaceArrayPosition_0
     dataInterfaceClass=&NiagaraDataInterfaceArrayPosition
    }
   }
  }
  User.SourceSize {type=Vector3f, description="", defaultValue {struct=Vector3f, value=<0 0 0>}}
  User.SourceSizeArray {
   type=NiagaraDataInterfaceArrayFloat3
   description=""
   defaultValue {
    struct=NiagaraExt_VariableValue_DataInterface
    value {
     dataInterface=&self:NiagaraDataInterfaceArrayFloat3_0
     dataInterfaceClass=&NiagaraDataInterfaceArrayFloat3
    }
   }
  }
  User.SplineHeightMatchingAmount {type=NiagaraFloat, defaultValue=5, description=""}
  User.WaterInfoTexture {type=Texture, description="", defaultValue {struct=NiagaraExt_VariableValue_Object, value {object=null, objectClass=&Texture}}}
  User.WaterZoneExtent {type=Vector2f, description="", defaultValue {struct=Vector2f, value=<0 0>}}
  User.WaterZoneIdx {type=NiagaraInt32}
  User.WaterZoneLocation {type=Vector2f, description="", defaultValue {struct=Vector2f, value=<0 0>}}
  User.WorldGridSize {type=Vector2f, description="", defaultValue {struct=Vector2f, value=<1000 1000>}}
 ]
}
