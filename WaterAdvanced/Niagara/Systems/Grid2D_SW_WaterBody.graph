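# Niagara GPU system that runs a 2D shallow-water simulation over water bodies, in a moving window around the owner.
# ShallowWaterEmitter sizes grid collections for water, velocity and foam, substeps the timestep, and sets up a normals render target and grid mesh.
# It takes in rigid-mesh colliders, impact arrays, data channel particles, water info and baked sim textures.
# Collision-particle and secondary sprite emitters, linked by particle and grid readers, are disabled.
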NiagaraSystem /WaterAdvanced/Niagara/Systems/Grid2D_SW_WaterBody
dependencies {
 usedDataInterfaces [
  /Script/Niagara.NiagaraDataInterfaceGrid2DCollection
  /Script/Niagara.NiagaraDataInterfaceRenderTarget2D
  &NiagaraDataInterfaceParticleRead
  /Script/Niagara.NiagaraDataInterfaceDynamicMesh
  /Script/Niagara.NiagaraDataInterface2DArrayTexture
  /Script/Niagara.NiagaraDataInterfaceTexture
  /Script/Niagara.NiagaraDataInterfaceVirtualTextureSample
  /Script/Niagara.NiagaraDataInterfaceEmitterProperties
  /Script/Niagara.NiagaraDataInterfaceGrid2DCollectionReader
  /Script/Niagara.NiagaraDataInterfaceNeighborGrid3D
  &NiagaraDataInterfaceCurve
  /Script/Niagara.NiagaraDataInterfaceDataChannelRead
 ]
 usedDynamicInputs []
 usedModules [
  &self:SystemSpawnScript.SystemScriptSource.SystemScriptGraph.NiagaraNodeAssignment_0.SetVariables_#1
  &self:SystemSpawnScript.SystemScriptSource.SystemScriptGraph.NiagaraNodeAssignment_1.SetVariables_#2
  &SystemState
  &self:ScratchModule_08
  &self:ScratchModule_09
 ]
}
emitters {
 ShallowWaterCollisionParticles {
  bEnabled=false
  simTarget=GPUComputeSim
  emitterUpdateScript [
   &EmitterState(
    Life Cycle Mode=::NewEnumerator1
    Loop Behavior=::NewEnumerator1
    Loop Duration Mode=::NewEnumerator1
    Scalability Mode=::NewEnumerator1
    UseLoopDelay=false
   )
   &SpawnRate(Use Spawn Probability=false)
   &SpawnBurst_Instantaneous(Spawn Count=1, Use Spawn Probability=false) {enabled=false}
  ]
  particleSpawnScript [
   /SetVariables(Particles.ParticleRadius=100) {isSetParameters=true}
   &AddVelocity(Velocity=<1000 0 -1000>)
   &Color(Color=<rgba 1 0 0 1>)
   /SetVariables(
    Particles.Radius:NiagaraFloat=$Particles.ParticleRadius
    Particles.RadiusSquared:NiagaraFloat=&Multiply_Float(A=$Particles.ParticleRadius, B=$Particles.ParticleRadius)
    Particles.Scale=&Divide_Vector(
     A:Vector3f=&VectorFromFloat(Value=$Particles.ParticleRadius)
     B:Vector3f=&VectorFromFloat(Value=50)
    )
   ) {isSetParameters=true}
   /SetVariables(Particles.Position=&Add_Vector(
     A=&ConvertPositionToVector(Input Position=$Engine.Owner.Position, Convert To Absolute Large World Vector=false)
     B=&MakeVector(X=0, Y=0, Z=0)
    )) {enabled=false, isSetParameters=true}
   &ShapeLocation(
    MaskSpawnGroup=false
    Non Uniform Scale=<1 1 0.1>
    Offset=<0 0 100>
    Offset Mode=::NewEnumerator0
    Override Seed=false
    Sphere Radius=300
    UseOwnerScale=true
   )
  ]
  particleUpdateScript [
   &GravityForce() {module=GravityForce001}
   &SolveForcesAndVelocity(Clamp Velocity=false, Limit Acceleration=false)
  ]
  rendererClasses [&NiagaraMeshRendererProperties]
  renderers [{rendererClass=&NiagaraMeshRendererProperties, rendererIndex=0}]
 }
 ShallowWaterEmitter {
  simTarget=GPUComputeSim
  emitterSpawnScript [
   /SetVariables(
    Emitter.FoamGrid=DataInterface() {propertyValues {
      ClearBeforeNonIterationStage=true
      NumAttributes=0
      NumCellsMaxAxis=3
      NumCellsX=3
      NumCellsY=3
      OverrideBufferFormat=HalfFloat
      PreviewAttribute=None
      SetGridFromMaxAxis=false
      bOverrideFormat=true
      bPreviewGrid=false
      RenderTargetUserParameter {parameter {name=None, typeDefHandle {registeredTypeIndex=4}, varData []}}
      WorldBBoxSize=<100 100>
     }}
    Emitter.GridCollection=DataInterface() {propertyValues {
      ClearBeforeNonIterationStage=true
      NumAttributes=0
      NumCellsMaxAxis=3
      NumCellsX=128
      NumCellsY=128
      OverrideBufferFormat=Float
      PreviewAttribute=IsCollider
      SetGridFromMaxAxis=false
      bOverrideFormat=false
      bPreviewGrid=false
      RenderTargetUserParameter {parameter {name=None, typeDefHandle {registeredTypeIndex=4}, varData []}}
      WorldBBoxSize=<1000 1000>
     }}
    Emitter.NormalsRT=DataInterface() {propertyValues {
      MipMapGeneration=Disabled
      MipMapGenerationType=Linear
      OverrideRenderTargetFilter=TF_Default
      OverrideRenderTargetFormat=RTF_RGBA16f
      bInheritUserParameterSettings=false
      bOverrideFormat=true
      bPreviewRenderTarget=false
      RenderTargetUserParameter {parameter {name=User.NormalRT, typeDefHandle {registeredTypeIndex=4}, varData []}}
      Size=<1 1>
     }}
    Emitter.SimGridRT=DataInterface() {propertyValues {
      MipMapGeneration=Disabled
      MipMapGenerationType=Linear
      OverrideRenderTargetFilter=TF_Default
      OverrideRenderTargetFormat=RTF_RGBA16f
      bInheritUserParameterSettings=false
      bOverrideFormat=true
      bPreviewRenderTarget=false
      RenderTargetUserParameter {parameter {name=User.SimRT, typeDefHandle {registeredTypeIndex=4}, varData []}}
      Size=<1 1>
     }}
    Emitter.VelocityGrid=DataInterface() {propertyValues {
      ClearBeforeNonIterationStage=true
      NumAttributes=0
      NumCellsMaxAxis=3
      NumCellsX=3
      NumCellsY=3
      OverrideBufferFormat=Float
      PreviewAttribute=None
      SetGridFromMaxAxis=false
      bOverrideFormat=true
      bPreviewGrid=false
      RenderTargetUserParameter {parameter {name=None, typeDefHandle {registeredTypeIndex=4}, varData []}}
      WorldBBoxSize=<100 100>
     }}
    Emitter.WaterGrid=DataInterface() {propertyValues {
      ClearBeforeNonIterationStage=true
      NumAttributes=0
      NumCellsMaxAxis=3
      NumCellsX=3
      NumCellsY=3
      OverrideBufferFormat=Float
      PreviewAttribute=None
      SetGridFromMaxAxis=false
      bOverrideFormat=true
      bPreviewGrid=false
      RenderTargetUserParameter {parameter {name=None, typeDefHandle {registeredTypeIndex=4}, varData []}}
      WorldBBoxSize=<100 100>
     }}
   ) {isSetParameters=true}
   /NiagaraFluids/Modules/Grid2D/Grid2D_SetResolution(
    Grid:NiagaraDataInterfaceGrid2DCollection=$Emitter.GridCollection
    NumCellsMaxAxis:NiagaraInt32=$User.ResolutionMaxAxis
    Resolution Method:ENiagaraGrid2DResolution=::NewEnumerator1
    WorldGridExtents:Vector2f=$User.WorldGridSize
   )
   &Grid2D_SetResolution(
    Grid:NiagaraDataInterfaceGrid2DCollection=$Emitter.VelocityGrid
    OtherGrid:NiagaraDataInterfaceGrid2DCollection=$Emitter.GridCollection
    Resolution Method:ENiagaraGrid2DResolution=::NewEnumerator3
    WorldGridExtents:Vector2f=$Emitter.Grid2D_SetResolution.WorldGridExtents
   ) {module=Grid2D_SetResolution001}
   &Grid2D_SetResolution(
    Grid:NiagaraDataInterfaceGrid2DCollection=$Emitter.WaterGrid
    OtherGrid:NiagaraDataInterfaceGrid2DCollection=$Emitter.GridCollection
    Resolution Method:ENiagaraGrid2DResolution=::NewEnumerator3
    WorldGridExtents:Vector2f=$Emitter.Grid2D_SetResolution.WorldGridExtents
   ) {module=Grid2D_SetResolution002}
   &Grid2D_SetResolution(
    Grid:NiagaraDataInterfaceGrid2DCollection=$Emitter.FoamGrid
    OtherGrid:NiagaraDataInterfaceGrid2DCollection=$Emitter.GridCollection
    Resolution Method:ENiagaraGrid2DResolution=::NewEnumerator3
    WorldGridExtents:Vector2f=$Emitter.Grid2D_SetResolution.WorldGridExtents
   ) {module=Grid2D_SetResolution003}
   /SetVariables(Emitter.ParticleCollisionReader=DataInterface() {propertyValues {EmitterBinding {bindingMode=Other, emitterName=ShallowWaterCollisionParticles}}}) {isSetParameters=true}
   /SetVariables(Emitter.LocalSpaceParticles:NiagaraBool=&self:ShallowWaterEmitter.NiagaraScratchPadContainer_0.ScratchDynamicInput_0(Reader:NiagaraDataInterfaceParticleRead=$Emitter.ParticleCollisionReader)) {isSetParameters=true}
   /NiagaraFluids/Modules/Grid2D/Grid2D_InitRT(Grid:NiagaraDataInterfaceGrid2DCollection=$Emitter.GridCollection, SimRT:NiagaraDataInterfaceRenderTarget2D=$Emitter.NormalsRT)
   &Grid2D_InitRT(Grid:NiagaraDataInterfaceGrid2DCollection=$Emitter.GridCollection, SimRT:NiagaraDataInterfaceRenderTarget2D=$Emitter.SimGridRT) {enabled=false, module=Grid2D_InitRT001}
   &self:ShallowWaterEmitter.NiagaraScratchPadContainer_0.ScratchModule_03_0(BottomDepth:NiagaraFloat=$User.BottomDepth, Debug Render:NiagaraBool=$User.UseDebugRender, RenderWaterSurface=true) {module=ScratchModule_03}
   &self:ScratchModule_05(RenderDebug:NiagaraBool=$Emitter.DebugRender) {module=ScratchModule_05001}
   &self:ScratchModule_13(Render Target 2D:NiagaraDataInterfaceRenderTarget2D=$Emitter.NormalsRT) {enabled=false, module=ScratchModule001}
   /SetVariables(
    Emitter.GridMesh=DataInterface() {propertyValues {
      NumTexCoords=0
      NumVertices=0
      bClearTrianglesPerFrame=false
      bHasColors=false
      bHasTangentBasis=false
      DefaultBounds {isValid=true, max=<100 100 100>, min=<-100 -100 -100>}
      Materials [{
        material=/Engine/EngineMaterials/WorldGridMaterial
        materialUserParamBinding {parameter {name=None, typeDefHandle {registeredTypeIndex=3}, varData []}}
       }]
      Sections []
     }}
    Emitter.MeshResMult=1
   ) {isSetParameters=true}
   &self:ShallowWaterEmitter.NiagaraScratchPadContainer_0.ScratchModule_04_0(
    Execute:NiagaraBool=$Emitter.EnableBeautyRender
    Mesh:NiagaraDataInterfaceDynamicMesh=$Emitter.GridMesh
    NumCellsX:NiagaraInt32=&MakeIntFromFloat(Float=&Multiply_Float(A=$Emitter.MeshResMult, B=&MakeFloatFromInt(Int=$Emitter.Grid2D_SetResolution.NumCellsX)))
    NumCellsY:NiagaraInt32=&MakeIntFromFloat(Float=&Multiply_Float(A=$Emitter.MeshResMult, B=&MakeFloatFromInt(Int=$Emitter.Grid2D_SetResolution.NumCellsY)))
   ) {module=ScratchModule_04}
   /SetVariables(Emitter.WaterInfoTextureArray=DataInterface() {propertyValues {Texture=null, TextureUserParameter {parameter {name=User.WaterInfoTexture, typeDefHandle {registeredTypeIndex=2}, varData []}}}}) {isSetParameters=true}
   /SetVariables(
    Emitter.BakedWaterSimHeightVirtualTextureSample=DataInterface() {propertyValues {Texture=null, TextureUserParameter {parameter {name=User.BakedWaterHeightTexture, typeDefHandle {registeredTypeIndex=2}, varData []}}}}
    Emitter.BakedWaterSimTextureSample=DataInterface() {propertyValues {Texture=null, TextureUserParameter {parameter {name=User.BakedWaterSimTexture, typeDefHandle {registeredTypeIndex=2}, varData []}}}}
    Emitter.BakedWaterSimVirtualTextureSample=DataInterface() {propertyValues {Texture=null, TextureUserParameter {parameter {name=User.BakedWaterSimTexture, typeDefHandle {registeredTypeIndex=2}, varData []}}}}
   ) {isSetParameters=true}
   &self:ScratchModule_02(
    MaxImpacts:NiagaraInt32=$User.MaxImpacts
    PositionArray:NiagaraDataInterfaceArrayPosition=$User.ImpactPositionArray
    RadiusArray:NiagaraDataInterfaceArrayFloat=$User.ImpactRadiusArray
    VelocityArray:NiagaraDataInterfaceArrayFloat3=$User.ImpactVelocityArray
   ) {module=ScratchModule_02001}
   /SetVariables(Emitter.NumSubsteps=2) {isSetParameters=true}
   /SetVariables(Emitter.DataChannelReader:NiagaraDataInterfaceDataChannelRead=$System.DataChannelReader) {isSetParameters=true}
  ]
  emitterUpdateScript [
   &EmitterState(
    Life Cycle Mode=::NewEnumerator1
    Loop Behavior=::NewEnumerator1
    Loop Duration Mode=::NewEnumerator1
    Scalability Mode=::NewEnumerator1
    UseLoopDelay=false
   ) {module=EmitterLifeCycle}
   /NiagaraFluids/Modules/Grid2D/Grid2D_CreateUnitToWorldTransform(
    Offset:NiagaraPosition=$Engine.Owner.Position
    Rotation:Quat4f=$Engine.Owner.Rotation
    WorldGridExtents:Vector2f=$Emitter.Grid2D_SetResolution.WorldGridExtents
   )
   /NiagaraFluids/Modules/Grid2D/ShallowWater/Grid2D_SW_UpdateSimAttributes(
    DeltaTime:NiagaraFloat=/Niagara/DynamicInputs/Math/Min_Float(
     A:NiagaraFloat=&Divide_Float(A=$Engine.DeltaTime, B=&MakeFloatFromInt(Int=$Emitter.NumSubsteps))
     B:NiagaraFloat=0.0111
    )
    MaxDepthMult=2
    MaxVelocityMult=2
    WorldCellSize:Vector2f=$Emitter.Grid2D_SetResolution.WorldCellSize
    WorldGridExtents:Vector2f=$Emitter.Grid2D_SetResolution.WorldGridExtents
   )
   /NiagaraFluids/Modules/Grid2D/Grid2D_SetBounds(
    EmitterProperties=DataInterface() {propertyValues {EmitterBinding {bindingMode=Self, emitterName=None}}}
    LocalPivot:Vector2f=<0.5 0.5>
    WorldGridExtents:Vector2f=$Emitter.Grid2D_SetResolution.WorldGridExtents
   ) {enabled=false}
   &self:ScratchModule_17(
    EmitterProperties=DataInterface() {propertyValues {EmitterBinding {bindingMode=Self, emitterName=None}}}
    GridExtents:Vector3f=&MakeVector(
     X=&MakeFloatFromVector2d(Vector 2D=$Emitter.Grid2D_SetResolution.WorldGridExtents)
     Y=&MakeFloatFromVector2d(Vector 2D=$Emitter.Grid2D_SetResolution.WorldGridExtents, Channel=::NewEnumerator1)
     Z=500
    )
    LocalPivot:Vector3f=&MakeVector(X=0, Y=0, Z=$User.WaterHeightAtParent)
   ) {module=ScratchModule002}
   /NiagaraFluids/Modules/Grid3D/Grid3D_RigidMesh_FindColliders(
    CollisionQuery:NiagaraDataInterfaceRigidMeshCollisionQuery=$User.Collide_Meshes
    FindColliders=true
    GridCenterPosition:NiagaraPosition=&AddVectorToPosition(Position=$Engine.Owner.Position, Vector=&MakeVector(X=0, Y=0, Z=$User.WaterHeightAtParent))
    GridRotation:Quat4f=$Engine.Owner.Rotation
    SkipOverlap:NiagaraBool=&Boolean_NotOperation(Bool=$User.PlayerIsInWater)
    TraceChannel=ECC_WorldDynamic
    WorldGridExtents:Vector3f=&MakeVector(
     X=&MakeFloatFromVector2d(Vector 2D=$Emitter.Grid2D_SetResolution.WorldGridExtents)
     Y=&MakeFloatFromVector2d(Vector 2D=$Emitter.Grid2D_SetResolution.WorldGridExtents, Channel=::NewEnumerator1)
     Z=&Multiply_Float(A=2, B=$Emitter.dx)
    )
   )
   /NiagaraFluids/Modules/Grid2D/ShallowWater/Grid2D_SW_MovingWindow_GetDelta(WorldToUnit:NiagaraMatrix=$Emitter.Grid2D_CreateUnitToWorldTransform.WorldToUnit)
   /SetVariables(Emitter.NumDataChannelParticles:NiagaraInt32=&self:ScratchDynamicInput(DataChannelReader:NiagaraDataInterfaceDataChannelRead=$Emitter.DataChannelReader)) {isSetParameters=true}
   /SetVariables(Emitter.DebugPlanePosition:NiagaraPosition=&AddVectorToPosition(Position=$Engine.Owner.Position, Vector=&MakeVector(X=0, Y=0, Z=$User.WaterHeightAtParent))) {isSetParameters=true}
  ]
  rendererClasses [&NiagaraMeshRendererProperties]
  renderers [
   {rendererClass=&NiagaraMeshRendererProperties, rendererIndex=0}
   {rendererClass=&NiagaraMeshRendererProperties, rendererIndex=1}
  ]
 }
 ShallowWater_Secondary_Emitter {
  bEnabled=false
  simTarget=GPUComputeSim
  emitterSpawnScript [
   /SetVariables(
    Emitter.VelocityReader=DataInterface() {propertyValues {
      ClearBeforeNonIterationStage=true
      DIName=VelocityGrid
      EmitterName=ShallowWaterEmitter
      NumAttributes=0
      NumCellsMaxAxis=3
      NumCellsX=3
      NumCellsY=3
      OverrideBufferFormat=Float
      PreviewAttribute=None
      SetGridFromMaxAxis=false
      bOverrideFormat=false
      bPreviewGrid=false
      RenderTargetUserParameter {parameter {name=None, typeDefHandle {registeredTypeIndex=4}, varData []}}
      WorldBBoxSize=<100 100>
     }}
    Emitter.WaterReader=DataInterface() {propertyValues {
      ClearBeforeNonIterationStage=true
      DIName=WaterGrid
      EmitterName=ShallowWaterEmitter
      NumAttributes=0
      NumCellsMaxAxis=3
      NumCellsX=3
      NumCellsY=3
      OverrideBufferFormat=Float
      PreviewAttribute=None
      SetGridFromMaxAxis=false
      bOverrideFormat=false
      bPreviewGrid=false
      RenderTargetUserParameter {parameter {name=None, typeDefHandle {registeredTypeIndex=4}, varData []}}
      WorldBBoxSize=<100 100>
     }}
   ) {isSetParameters=true}
   /SetVariables(Emitter.MaxSecondaryParticlesPerFrame=64) {isSetParameters=true}
   /SetVariables(Emitter.EmissionPositions=DataInterface() {propertyValues {
      ClearBeforeNonIterationStage=true
      NumAttributes=2
      NumCellsMaxAxis=3
      NumCellsX=1
      NumCellsY=1
      OverrideBufferFormat=Float
      PreviewAttribute=None
      SetGridFromMaxAxis=false
      bOverrideFormat=false
      bPreviewGrid=false
      RenderTargetUserParameter {parameter {name=None, typeDefHandle {registeredTypeIndex=4}, varData []}}
      WorldBBoxSize=<100 100>
     }}) {isSetParameters=true}
   &Grid2D_SetResolution(
    Grid:NiagaraDataInterfaceGrid2DCollection=$Emitter.EmissionPositions
    NumCellsX:NiagaraInt32=$Emitter.MaxSecondaryParticlesPerFrame
    NumCellsY=1
   )
   /SetVariables(Emitter.EmissionCounter=DataInterface() {propertyValues {
      CellSize=1
      ClearBeforeNonIterationStage=true
      MaxNeighborsPerCell=1
      NumCellsMaxAxis=10
      SetResolutionMethod=Independent
      NumCells=<2 2 2>
      WorldBBoxSize=<100 100 100>
     }}) {isSetParameters=true}
   &self:ShallowWater_Secondary_Emitter.NiagaraScratchPadContainer_2.ScratchModule_01_5() {module=ScratchModule_01}
   /SetVariables(Emitter.WaterInfoTexture=DataInterface() {propertyValues {Texture=null, TextureUserParameter {parameter {name=User.WaterInfoTexture, typeDefHandle {registeredTypeIndex=2}, varData []}}}}) {isSetParameters=true}
  ]
  emitterUpdateScript [
   &EmitterState()
   &self:ShallowWater_Secondary_Emitter.NiagaraScratchPadContainer_2.ScratchModule_06_1(Extents:Vector2f=$User.WorldGridSize, Read:NiagaraDataInterfaceGrid2DCollectionReader=$Emitter.WaterReader) {module=ScratchModule_06}
   &Grid2D_CreateUnitToWorldTransform(Offset:NiagaraPosition=$Engine.Owner.Position, Rotation:Quat4f=$Engine.Owner.Rotation, WorldGridExtents:Vector2f=$User.WorldGridSize)
   /SetVariables(
    Emitter.LocalToWorld:NiagaraMatrix=$Emitter.Grid2D_CreateUnitToWorldTransform.LocalToWorld
    Emitter.UnitToWorld:NiagaraMatrix=$Emitter.Grid2D_CreateUnitToWorldTransform.UnitToWorld
    Emitter.WorldToLocal:NiagaraMatrix=$Emitter.Grid2D_CreateUnitToWorldTransform.WorldToLocal
    Emitter.WorldToUnit:NiagaraMatrix=$Emitter.Grid2D_CreateUnitToWorldTransform.WorldToUnit
   ) {isSetParameters=true}
   /SetVariables(Emitter.DeltaTime:NiagaraFloat=$Engine.DeltaTime) {isSetParameters=true}
   &SpawnPerFrame(Spawn Count:NiagaraInt32=$Emitter.MaxSecondaryParticlesPerFrame, Use Spawn Probability=false)
   /WaterAdvanced/Niagara/Modules/DualRestTimeline(DualRestPeriod=0.5)
  ]
  particleSpawnScript [
   &self:ShallowWater_Secondary_Emitter.NiagaraScratchPadContainer_2.ScratchModule_0(
    Lifetime:NiagaraFloat=&RandomRangeFloat(Minimum=1, Maximum=3)
    SpriteSize:Vector2f=&Multiply_Vector2D_ByFloat(
     Float=&MakeFloatFromVector2d(Vector 2D=&Vector2DFromFloat(Value=&FloatFromCurve(
        FloatCurve=DataInterface() {propertyValues {Curve {keys [<at 0 0.5>,<at 1 2>]}}}
        CurveIndex=&RandomRangeFloat(Minimum=0, Maximum=1)
        Scale Curve=15
       )))
     Vector2D=<1 1>
    )
   ) {module=ScratchModule_4}
   /SetVariables(Particles.SpriteFacing=<0 0 1>) {isSetParameters=true}
   /SetVariables(Particles.SmoothlyRotateVectorAroundAxis.SmoothedVectorDirection:Vector3f=<1 0 0>) {isSetParameters=true}
  ]
  particleUpdateScript [
   &ParticleState()
   &ScaleSpriteSize(Uniform Curve Sprite Scale=DataInterface() {propertyValues {Curve {keys [{tangentMode=RCTM_User, time=0, value=1}, <at 1 2 RCIM_Cubic>]}}}) {enabled=false}
   /WaterAdvanced/Niagara/Modules/SmearSpriteBySpeed(SpeedRange:Vector2f=<0 150>, StretchRange:Vector2f=<1 1.5>)
   /WaterAdvanced/Niagara/Modules/DualRestCapturePosition(ShouldCaptureThisFrame:NiagaraBool=$Emitter.DualRestTimeline.ShouldCaptureRestPosition, WriteToDynamicParameters2And3=true)
   /WaterAdvanced/Niagara/Modules/SmoothlyRotateVectorAroundAxis(Alpha=0.3, MaxRotateSpeed=1, TargetVector:Vector3f=$Particles.Velocity)
   /SetVariables() {isSetParameters=true}
   /WaterAdvanced/Niagara/Modules/WaterBodyData(
    UseCustomWorldPosition2D=true
    WaterInfoTexture:NiagaraDataInterfaceTexture=$Emitter.WaterInfoTexture
    WorldPosition2D:Vector2f=&MakeVector2D(
     X=&MakeFloatFromVector(Vector=Linked() {linkedVariable {name=Particles.Position, type=NiagaraPosition}}, Channel=::NewEnumerator0)
     Y=&MakeFloatFromVector(Vector=Linked() {linkedVariable {name=Particles.Position, type=NiagaraPosition}}, Channel=::NewEnumerator1)
    )
   )
   &self:SetAboveWater(WaterHeight:NiagaraFloat=$Output.WaterBodyData.WaterHeight) {enabled=false, module=SetAboveWater}
   &DynamicMaterialParameters(
    Index 0 Param 1=$Output.WaterBodyData.WaterHeight
    Module.Index0Param0Write=true
    Module.Index0Param1Write=false
    Module.Index0Param2Write=false
    Module.Index0Param3Write=false
   )
  ]
  rendererClasses [&NiagaraSpriteRendererProperties]
  renderers [{rendererClass=&NiagaraSpriteRendererProperties}]
 }
}
summary {
 systemName=Grid2D_SW_WaterBody
 emitters [
  {emitterName=ShallowWaterEmitter, simTarget=GPUComputeSim, rendererClasses [&NiagaraMeshRendererProperties]}
  {
   bEnabled=false
   emitterName=ShallowWaterCollisionParticles
   simTarget=GPUComputeSim
   rendererClasses [&NiagaraMeshRendererProperties]
  }
  {
   bEnabled=false
   emitterName=ShallowWater_Secondary_Emitter
   simTarget=GPUComputeSim
   rendererClasses [&NiagaraSpriteRendererProperties]
  }
 ]
 userVariables [
  User.BakedWaterDepthMax {type=NiagaraFloat, defaultValue=0, description=""}
  User.BakedWaterHeightTexture {type=Texture, description="", defaultValue {struct=NiagaraExt_VariableValue_Object, value {object=null, objectClass=/Script/Engine.Texture}}}
  User.BakedWaterSimLocation {type=Vector3f, description="", defaultValue {struct=Vector3f, value=<0 0 0>}}
  User.BakedWaterSimRes {type=Vector2f, description="", defaultValue {struct=Vector2f, value=<0 0>}}
  User.BakedWaterSimSize {type=Vector2f, description="", defaultValue {struct=Vector2f, value=<0 0>}}
  User.BakedWaterSimTexture {type=Texture, description="", defaultValue {struct=NiagaraExt_VariableValue_Object, value {object=null, objectClass=&Texture}}}
  User.BakedWaterVelocityMaxSpeed {type=NiagaraFloat, defaultValue=0, description=""}
  User.BottomDepth {type=NiagaraFloat, defaultValue=-40, description=""}
  User.Collide_Meshes {
   type=NiagaraDataInterfaceRigidMeshCollisionQuery
   description=""
   defaultValue {
    struct=NiagaraExt_VariableValue_DataInterface
    value {
     dataInterface=&self:NiagaraDataInterfaceRigidMeshCollisionQuery_0
     dataInterfaceClass=/Script/Niagara.NiagaraDataInterfaceRigidMeshCollisionQuery
    }
   }
  }
  User.ImpactPositionArray {
   type=NiagaraDataInterfaceArrayPosition
   description=""
   defaultValue {
    struct=NiagaraExt_VariableValue_DataInterface
    value {
     dataInterface=&self:NiagaraDataInterfaceArrayPosition_1
     dataInterfaceClass=&NiagaraDataInterfaceArrayPosition
    }
   }
  }
  User.ImpactRadiusArray {
   type=NiagaraDataInterfaceArrayFloat
   defaultValue {
    struct=NiagaraExt_VariableValue_DataInterface
    value {
     dataInterface=&self:NiagaraDataInterfaceArrayFloat_1
     dataInterfaceClass=&NiagaraDataInterfaceArrayFloat
    }
   }
  }
  User.ImpactVelocityArray {
   type=NiagaraDataInterfaceArrayFloat3
   description=""
   defaultValue {
    struct=NiagaraExt_VariableValue_DataInterface
    value {
     dataInterface=&self:NiagaraDataInterfaceArrayFloat3_1
     dataInterfaceClass=&NiagaraDataInterfaceArrayFloat3
    }
   }
  }
  User.Landscape {
   type=NiagaraDataInterfaceLandscape
   description=""
   defaultValue {
    struct=NiagaraExt_VariableValue_DataInterface
    value {
     dataInterface=&self:NiagaraDataInterfaceLandscape_0
     dataInterfaceClass=&NiagaraDataInterfaceLandscape
    }
   }
  }
  User.MaxImpacts {type=NiagaraInt32}
  User.NormalRT {
   type=TextureRenderTarget
   description=""
   defaultValue {struct=NiagaraExt_VariableValue_Object, value {object=null, objectClass=/Script/Engine.TextureRenderTarget}}
  }
  User.PlayerIsInWater {type=NiagaraBool, defaultValue=true}
  User.ResolutionMaxAxis {type=NiagaraInt32, defaultValue=100}
  User.SecondsUntilDestroyed {type=NiagaraFloat, defaultValue=100, description=""}
  User.UseBakedSim {type=NiagaraBool}
  User.UseDebugRender {type=NiagaraBool}
  User.UseWaterInfoTexture {type=NiagaraBool}
  User.WaterHeightAtParent {type=NiagaraFloat, defaultValue=0, description=""}
  User.WaterInfoTexture {type=Texture, description="", defaultValue {struct=NiagaraExt_VariableValue_Object, value {object=null, objectClass=&Texture}}}
  User.WaterZoneExtent {type=Vector2f, description="", defaultValue {struct=Vector2f, value=<0 0>}}
  User.WaterZoneIdx {type=NiagaraInt32}
  User.WaterZoneIndex {type=NiagaraFloat, defaultValue=0, description=""}
  User.WaterZoneLocation {type=Vector2f, description="", defaultValue {struct=Vector2f, value=<0 0>}}
  User.WorldGridSize {type=Vector2f, description="", defaultValue {struct=Vector2f, value=<1000 1000>}}
 ]
}
