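# Niagara demo system that sprays a fountain of particles to feed collision data into shallow water through a Data Channel.
# The Fountain emitter launches sprite/mesh particles in an upward cone with gravity, drag and collision, then slow-kills them on impact.
# Each update writes position, scale-derived radius and velocity to NDC_ShallowWater for CPU and GPU readers.
# A disabled SingleLoopingParticle emitter follows the owner, writing waveform-driven wiggle velocity to the same channel.
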NiagaraSystem /WaterAdvanced/Niagara/Systems/NDCCollision_FountainDemo
dependencies {
 usedDataInterfaces [/Script/Niagara.NiagaraDataInterfaceDataChannelWrite, &NiagaraDataInterfaceCurve]
 usedDynamicInputs []
 usedModules [&SystemState]
}
emitters {
 Fountain {
  emitterUpdateScript [
   &EmitterState(Life Cycle Mode=::NewEnumerator1, Loop Duration=2, UseLoopDelay=false)
   &SpawnRate(SpawnRate=5, Use Spawn Probability=false)
  ]
  particleSpawnScript [
   &InitializeParticle(
    AdjustAlpha=false
    AdjustHue=true
    AdjustSaturation=false
    AdjustValue=false
    Color=<rgba 1 0 0 1>
    Color Mode=::NewEnumerator3
    Hue Shift Range:Vector2f=<0 1>
    Lifetime=5
    Mesh Scale Mode=::NewEnumerator3
    Mesh Uniform Scale=&RandomRangeFloat(Minimum=0.02, Maximum=0.2)
    Sprite Rotation Mode=::NewEnumerator2
    UsePositionOffset=false
    Write Material Random=true
   )
   &ShapeLocation(MaskSpawnGroup=false, Override Seed=false, Sphere Radius=1, UseOwnerScale=true)
   &AddVelocity(
    Cone Angle=60
    Cone Axis=<0 0 1>
    Distribution Along Cone Axis=0.25
    Module.UseConeFalloff=true
    Module.UseVelDistribution=true
    Override Seed=false
    Speed Falloff From Cone Axis=0.5
    Velocity Mode=::NewEnumerator2
    Velocity Speed=&RandomRangeFloat(Minimum=400, Maximum=1550)
   )
  ]
  particleUpdateScript [
   &ParticleState()
   &GravityForce()
   &Drag(Drag=0.25, Use Linear Drag=true, Use Rotational Drag=false)
   &Collision(
    Age Colliding Particles=false
    CPU Collision Trace Channel=ECC_GameTraceChannel15
    KillOnCollision=false
    Module.Control Rest State Via Surface Normal=false
    Module.EnableMaxCollisionCount=false
    Module.Randomize Collision Normal Vector=false
    Restitution=0
   )
   &ScaleColor(
    Scale Alpha=&FloatFromCurve(FloatCurve=DataInterface() {propertyValues {Curve {keys [<at -0.0046 1.0113 RCIM_Cubic>,<at 1.0011 0.002 RCIM_Cubic>]}}})
    ScaleA=true
    ScaleRGB=true
   )
   &SolveForcesAndVelocity(Clamp Velocity=false, Limit Acceleration=false)
   /WaterAdvanced/Niagara/Modules/WriteNDC_SW_CollisionParticles(
    DataChannelWriter=DataInterface() {propertyValues {
      AllocationCount=10
      AllocationMode=Static
      Channel=/WaterAdvanced/Niagara/DataChannels/NDC_ShallowWater
      bOnlyWriteOnceOnSubticks=true
      bPublishToCPU=true
      bPublishToGPU=true
      bPublishToGame=false
      bUpdateDestinationDataEveryTick=true
      AccessContext {accessContext {}}
     }}
    Position:NiagaraPosition=$Particles.Position
    Radius:NiagaraFloat=&Multiply_Float(A=&MakeFloatFromVector(Vector=$Particles.Scale, Channel=::NewEnumerator0), B=400)
    Velocity:Vector3f=&Multiply_VectorByFloat(Vector=$Particles.Velocity, Float=1)
   )
   &self:SlowKill(ShouldSlowKill:NiagaraBool=$Transient.CollisionValid, SlowKillTimer=0.1) {module=SlowKill}
  ]
  rendererClasses [&NiagaraSpriteRendererProperties, &NiagaraMeshRendererProperties]
  renderers [
   {rendererClass=&NiagaraSpriteRendererProperties}
   {rendererClass=&NiagaraMeshRendererProperties, rendererIndex=1}
  ]
 }
 SingleLoopingParticle {
  bEnabled=false
  emitterUpdateScript [
   &EmitterState()
   &SpawnBurst_Instantaneous(Use Spawn Probability=false, UseLoopCountLimit=false)
  ]
  particleSpawnScript [&InitializeParticle(
    Color Mode=::NewEnumerator0
    Lifetime=$Emitter.CurrentLoopDuration
    Mesh Scale Mode=::NewEnumerator3
    Mesh Uniform Scale=0.1
    Sprite Size Mode=::NewEnumerator3
    Uniform Sprite Size=5
    UsePositionOffset=false
    Write Material Random=true
   )]
  particleUpdateScript [
   &ParticleState()
   /SetVariables(Particles.Position=$Engine.Owner.Position) {isSetParameters=true}
   &self:SetWaterSurfacePosition(CustomWaterZ:NiagaraFloat=$User.CustomWaterZ, bShouldUseCustomWaterZ:NiagaraBool=$User.bShouldUseCustomWaterZ) {module=SetWaterSurfacePosition}
   &VectorNoiseForce(Module.Use Random Noise Time Interval=true)
   /WaterAdvanced/Niagara/Modules/NDCCollision/NDCCollision_SphereWiggle(
    AmpHigh:NiagaraFloat=$User.WaveAmpHigh
    AmpLow:NiagaraFloat=$User.WaveAmpLow
    FreqHigh:NiagaraFloat=$User.WaveFreqHigh
    FreqLow:NiagaraFloat=$User.WaveFreqLow
    SpeedThresholdToUseHigh=1000000000
   )
   /Niagara/Noise/Waveform(
    EnableExponent1=false
    EnableExponent2=false
    EnableExponent3=false
    Global Phase:NiagaraFloat=$StackContext.Phase
    GlobalClampMax=false
    GlobalClampMin=false
    UseInput1=false
    UseInput3=false
    UsePhase1=false
    UsePhase2=false
    UsePhase3=false
    Waveform Count:ENiagara_WaveformCount=::NewEnumerator2
    "[1] Amplitude Min/Max":Vector2f=<-1 1>
    "[2] Amplitude Min/Max":Vector2f=<-1 1>
    "[2] Frequency"=1.7
    "[2] Phase Scale"=0.8
    "[3] Amplitude Min/Max":Vector2f=<-1 1>
    "[3] Frequency"=2.9
    "[3] Phase Scale"=0.5
   )
   &Waveform(
    EnableExponent1=false
    EnableExponent2=false
    EnableExponent3=false
    Global Input Position=1234.5601
    Global Phase:NiagaraFloat=$StackContext.Phase
    GlobalClampMax=false
    GlobalClampMin=false
    UseInput1=false
    UseInput3=false
    UsePhase1=false
    UsePhase2=false
    UsePhase3=false
    Waveform Count:ENiagara_WaveformCount=::NewEnumerator2
    "[1] Amplitude Min/Max":Vector2f=<-1 1>
    "[2] Amplitude Min/Max":Vector2f=<-1 1>
    "[2] Frequency"=1.7
    "[2] Phase Scale"=0.8
    "[3] Amplitude Min/Max":Vector2f=<-1 1>
    "[3] Frequency"=2.9
    "[3] Phase Scale"=0.5
   ) {module=Waveform001}
   &Waveform(
    EnableExponent1=false
    EnableExponent2=false
    EnableExponent3=false
    Global Input Position=66666.6641
    Global Phase:NiagaraFloat=$StackContext.Phase
    GlobalClampMax=false
    GlobalClampMin=false
    UseInput1=false
    UseInput3=false
    UsePhase1=false
    UsePhase2=false
    UsePhase3=false
    Waveform Count:ENiagara_WaveformCount=::NewEnumerator2
    "[1] Amplitude Min/Max":Vector2f=<-1 1>
    "[2] Amplitude Min/Max":Vector2f=<-1 1>
    "[2] Frequency"=1.7
    "[2] Phase Scale"=0.8
    "[3] Amplitude Min/Max":Vector2f=<-1 1>
    "[3] Frequency"=2.9
    "[3] Phase Scale"=0.5
   ) {module=Waveform002}
   /SetVariables(Particles.WaveVector:Vector3f=&MakeVector(X=$Particles.Waveform.Waveform, Y=$Particles.Waveform001.Waveform, Z=$Particles.Waveform002.Waveform)) {isSetParameters=true}
   /SetVariables(Particles.CollisionVelocity:Vector3f=&Multiply_VectorByFloat(Vector=$Particles.WaveVector, Float=$StackContext.Amp)) {isSetParameters=true}
   &ScaleMeshSize(Scale Factor=&VectorFromFloat(Value=/Niagara/DynamicInputs/Math/Min_Float(A:NiagaraFloat=&VectorLength(Vector=$Particles.CollisionVelocity), B:NiagaraFloat=100)))
   &WriteNDC_SW_CollisionParticles(
    DataChannelWriter=DataInterface() {propertyValues {
      AllocationCount=1
      AllocationMode=Static
      Channel=&NDC_ShallowWater
      bOnlyWriteOnceOnSubticks=true
      bPublishToCPU=true
      bPublishToGPU=true
      bPublishToGame=false
      bUpdateDestinationDataEveryTick=true
      AccessContext {accessContext {}}
     }}
    Position:NiagaraPosition=$Particles.Position
    Radius:NiagaraFloat=$User.Radius
    Velocity:Vector3f=$Particles.CollisionVelocity
   )
  ]
  rendererClasses [&NiagaraSpriteRendererProperties, &NiagaraMeshRendererProperties]
  renderers [
   {rendererClass=&NiagaraSpriteRendererProperties}
   {rendererClass=&NiagaraMeshRendererProperties, rendererIndex=1}
  ]
 }
}
summary {
 systemName=NDCCollision_FountainDemo
 emitters [
  {
   bEnabled=false
   emitterName=SingleLoopingParticle
   rendererClasses [&NiagaraSpriteRendererProperties, &NiagaraMeshRendererProperties]
  }
  {emitterName=Fountain, rendererClasses [&NiagaraSpriteRendererProperties, &NiagaraMeshRendererProperties]}
 ]
 userVariables [
  User.bShouldUseCustomWaterZ {type=NiagaraBool}
  User.CustomWaterZ {type=NiagaraFloat, defaultValue=0, description=""}
  User.Radius {type=NiagaraFloat, defaultValue=10, description=""}
  User.SpeedThresholdToUseHighFreq {type=NiagaraFloat, defaultValue=50, description=""}
  User.WaveAmpHigh {type=NiagaraFloat, defaultValue=10, description=""}
  User.WaveAmpLow {type=NiagaraFloat, defaultValue=1, description=""}
  User.WaveFreqHigh {type=NiagaraFloat, defaultValue=1, description=""}
  User.WaveFreqLow {type=NiagaraFloat, defaultValue=0.4, description=""}
 ]
}
