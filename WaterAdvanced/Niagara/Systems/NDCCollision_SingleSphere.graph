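# Niagara system that feeds one sphere collider at its owner's position into the shallow water Data Channel.
# A looping particle follows the owner, passes through a water-surface height step with optional custom Z, then calculates velocity.
# SphereWiggle takes low/high wave settings and a speed threshold, driving three offset waveforms that form a synthetic collision velocity.
# That velocity is written with position and radius each tick, and also scales the mesh.
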NiagaraSystem /WaterAdvanced/Niagara/Systems/NDCCollision_SingleSphere
dependencies {
 usedDataInterfaces [/Script/Niagara.NiagaraDataInterfaceDataChannelWrite]
 usedDynamicInputs []
 usedModules [&SystemState]
}
emitters {SingleLoopingParticle {
  emitterUpdateScript [
   &EmitterState()
   &SpawnBurst_Instantaneous(Use Spawn Probability=false, UseLoopCountLimit=false)
  ]
  particleSpawnScript [&InitializeParticle(
    Color Mode=::NewEnumerator0
    Lifetime=$Emitter.CurrentLoopDuration
    Mesh Scale Mode=::NewEnumerator3
    Mesh Uniform Scale=0.1
    Sprite Size Mode=::NewEnumerator3
    Uniform Sprite Size=5
    UsePositionOffset=false
    Write Material Random=true
   )]
  particleUpdateScript [
   &ParticleState()
   /SetVariables(Particles.Position=$Engine.Owner.Position) {isSetParameters=true}
   &self:SetWaterSurfacePosition(CustomWaterZ:NiagaraFloat=$User.CustomWaterZ, bShouldUseCustomWaterZ:NiagaraBool=$User.bShouldUseCustomWaterZ) {module=SetWaterSurfacePosition}
   &VectorNoiseForce(Module.Use Random Noise Time Interval=true)
   &CalculateAccurateVelocity()
   /WaterAdvanced/Niagara/Modules/NDCCollision/NDCCollision_SphereWiggle(
    AmpHigh:NiagaraFloat=$User.WaveAmpHigh
    AmpLow:NiagaraFloat=$User.WaveAmpLow
    FreqHigh:NiagaraFloat=$User.WaveFreqHigh
    FreqLow:NiagaraFloat=$User.WaveFreqLow
    SpeedThresholdToUseHigh:NiagaraFloat=$User.SpeedThresholdToUseHighFreq
   )
   /Niagara/Noise/Waveform(
    EnableExponent1=false
    EnableExponent2=false
    EnableExponent3=false
    Global Phase:NiagaraFloat=$StackContext.Phase
    GlobalClampMax=false
    GlobalClampMin=false
    UseInput1=false
    UseInput3=false
    UsePhase1=false
    UsePhase2=false
    UsePhase3=false
    Waveform Count:ENiagara_WaveformCount=::NewEnumerator2
    "[1] Amplitude Min/Max":Vector2f=<-1 1>
    "[2] Amplitude Min/Max":Vector2f=<-1 1>
    "[2] Frequency"=1.7
    "[2] Phase Scale"=0.8
    "[3] Amplitude Min/Max":Vector2f=<-1 1>
    "[3] Frequency"=2.9
    "[3] Phase Scale"=0.5
   )
   &Waveform(
    EnableExponent1=false
    EnableExponent2=false
    EnableExponent3=false
    Global Input Position=1234.5601
    Global Phase:NiagaraFloat=$StackContext.Phase
    GlobalClampMax=false
    GlobalClampMin=false
    UseInput1=false
    UseInput3=false
    UsePhase1=false
    UsePhase2=false
    UsePhase3=false
    Waveform Count:ENiagara_WaveformCount=::NewEnumerator2
    "[1] Amplitude Min/Max":Vector2f=<-1 1>
    "[2] Amplitude Min/Max":Vector2f=<-1 1>
    "[2] Frequency"=1.7
    "[2] Phase Scale"=0.8
    "[3] Amplitude Min/Max":Vector2f=<-1 1>
    "[3] Frequency"=2.9
    "[3] Phase Scale"=0.5
   ) {module=Waveform001}
   &Waveform(
    EnableExponent1=false
    EnableExponent2=false
    EnableExponent3=false
    Global Input Position=66666.6641
    Global Phase:NiagaraFloat=$StackContext.Phase
    GlobalClampMax=false
    GlobalClampMin=false
    UseInput1=false
    UseInput3=false
    UsePhase1=false
    UsePhase2=false
    UsePhase3=false
    Waveform Count:ENiagara_WaveformCount=::NewEnumerator2
    "[1] Amplitude Min/Max":Vector2f=<-1 1>
    "[2] Amplitude Min/Max":Vector2f=<-1 1>
    "[2] Frequency"=1.7
    "[2] Phase Scale"=0.8
    "[3] Amplitude Min/Max":Vector2f=<-1 1>
    "[3] Frequency"=2.9
    "[3] Phase Scale"=0.5
   ) {module=Waveform002}
   /SetVariables(Particles.WaveVector:Vector3f=&MakeVector(X=$Particles.Waveform.Waveform, Y=$Particles.Waveform001.Waveform, Z=$Particles.Waveform002.Waveform)) {isSetParameters=true}
   /SetVariables(Particles.CollisionVelocity:Vector3f=&Multiply_VectorByFloat(Vector=$Particles.WaveVector, Float=$StackContext.Amp)) {isSetParameters=true}
   &ScaleMeshSize(Scale Factor=&VectorFromFloat(Value=/Niagara/DynamicInputs/Math/Min_Float(A:NiagaraFloat=&VectorLength(Vector=$Particles.CollisionVelocity), B:NiagaraFloat=100)))
   /WaterAdvanced/Niagara/Modules/WriteNDC_SW_CollisionParticles(
    DataChannelWriter=DataInterface() {propertyValues {
      AllocationCount=1
      AllocationMode=Static
      Channel=/WaterAdvanced/Niagara/DataChannels/NDC_ShallowWater
      bOnlyWriteOnceOnSubticks=true
      bPublishToCPU=true
      bPublishToGPU=true
      bPublishToGame=false
      bUpdateDestinationDataEveryTick=true
      AccessContext {accessContext {}}
     }}
    Position:NiagaraPosition=$Particles.Position
    Radius:NiagaraFloat=$User.Radius
    Velocity:Vector3f=$Particles.CollisionVelocity
   )
  ]
  rendererClasses [&NiagaraSpriteRendererProperties, &NiagaraMeshRendererProperties]
  renderers [
   {rendererClass=&NiagaraSpriteRendererProperties}
   {rendererClass=&NiagaraMeshRendererProperties, rendererIndex=1}
  ]
 }}
summary {
 systemName=NDCCollision_SingleSphere
 emitters [{
   emitterName=SingleLoopingParticle
   rendererClasses [&NiagaraSpriteRendererProperties, &NiagaraMeshRendererProperties]
  }]
 userVariables [
  User.bShouldUseCustomWaterZ {type=NiagaraBool}
  User.CustomWaterZ {type=NiagaraFloat, defaultValue=0, description=""}
  User.Radius {type=NiagaraFloat, defaultValue=10, description=""}
  User.SpeedThresholdToUseHighFreq {type=NiagaraFloat, defaultValue=50, description=""}
  User.WaveAmpHigh {type=NiagaraFloat, defaultValue=10, description=""}
  User.WaveAmpLow {type=NiagaraFloat, defaultValue=1, description=""}
  User.WaveFreqHigh {type=NiagaraFloat, defaultValue=1, description=""}
  User.WaveFreqLow {type=NiagaraFloat, defaultValue=0.4, description=""}
 ]
}
