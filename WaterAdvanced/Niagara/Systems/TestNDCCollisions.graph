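# Niagara test system that feeds a moving particle into the shallow water Data Channel as a collision source.
# A single looping sprite particle swings along X around the owner's position on a cosine of emitter age, with velocity derived from that motion.
# Each update, WriteNDC_SW_CollisionParticles writes position, half-sprite-size radius and velocity to NDC_ShallowWater, published to CPU and GPU but not to game.
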NiagaraSystem /WaterAdvanced/Niagara/Systems/TestNDCCollisions
dependencies {
 usedDataInterfaces [/Script/Niagara.NiagaraDataInterfaceDataChannelWrite]
 usedDynamicInputs []
 usedModules [&SystemState]
}
emitters {SingleLoopingParticle {
  emitterUpdateScript [
   &EmitterState()
   &SpawnBurst_Instantaneous(Use Spawn Probability=false, UseLoopCountLimit=false)
  ]
  particleSpawnScript [&InitializeParticle(
    Color Mode=::NewEnumerator0
    Lifetime=$Emitter.CurrentLoopDuration
    Sprite Size Mode=::NewEnumerator3
    Uniform Sprite Size=100
    UsePositionOffset=false
    Write Material Random=true
   )]
  particleUpdateScript [
   &ParticleState()
   &SolveForcesAndVelocity(Clamp Velocity=false, Limit Acceleration=false)
   /SetVariables(Particles.Position=&Add_Vector(
     A=&ConvertPositionToVector(Input Position=$Engine.Owner.Position, Conversion Operation:ENiagara_LWCConvertPosToVec=::NewEnumerator0)
     B=&MakeVector(X=&Cosine(Normalized Angle=$Emitter.Age, Period=8, Scale=800, Bias=0), Y=0, Z=0)
    )) {isSetParameters=true}
   &CalculateAccurateVelocity()
   /WaterAdvanced/Niagara/Modules/WriteNDC_SW_CollisionParticles(
    DataChannelWriter=DataInterface() {propertyValues {
      AllocationCount=20
      AllocationMode=Static
      Channel=/WaterAdvanced/Niagara/DataChannels/NDC_ShallowWater
      bOnlyWriteOnceOnSubticks=true
      bPublishToCPU=true
      bPublishToGPU=true
      bPublishToGame=false
      bUpdateDestinationDataEveryTick=true
      AccessContext {accessContext {}}
     }}
    Position:NiagaraPosition=$Particles.Position
    Radius:NiagaraFloat=&Multiply_Float(A=&MakeFloatFromVector2d(Vector 2D=$Particles.SpriteSize), B=0.5)
    Velocity:Vector3f=$Particles.Velocity
   )
  ]
  rendererClasses [&NiagaraSpriteRendererProperties]
  renderers [{rendererClass=&NiagaraSpriteRendererProperties}]
 }}
summary {
 systemName=TestNDCCollisions
 emitters [{emitterName=SingleLoopingParticle, rendererClasses [&NiagaraSpriteRendererProperties]}]
 userVariables []
}
